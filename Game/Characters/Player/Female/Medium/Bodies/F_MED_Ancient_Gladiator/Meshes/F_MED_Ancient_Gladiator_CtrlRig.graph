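# Control Rig for the female medium Ancient Gladiator body that procedurally adjusts the left shoulder pad driver bone to follow arm pose.
# It aims the bone at the upper-arm twist, weighted by the larger of two remapped arm-to-clavicle Euler angles.
# Two additive local offsets follow, weighted by arm rotation and elbow proximity to helper bone dyn_arm_up.
# Strap, frill and right-side bones are not driven; a clamp and a remap node are unused.
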ControlRigBlueprint /Game/Characters/Player/Female/Medium/Bodies/F_MED_Ancient_Gladiator/Meshes/F_MED_Ancient_Gladiator_CtrlRig
graphs [
 RigVMFunctionLibrary {links [], nodes []}
 RigVMModel {nodes [
   n1 = &RigUnit_BeginExecution()
   RigUnit_GetRelativeBoneTransform = &RigUnit_GetRelativeBoneTransform(Bone=upperarm_l, Space=clavicle_l)
   n2 = &RigVMFunction_MathQuaternionToEuler(RotationOrder=XYZ, Value=@RigUnit_GetRelativeBoneTransform.Transform.Rotation)
   n3 = &RigVMFunction_MathQuaternionToEuler(RotationOrder=ZXY, Value=&RigUnit_GetRelativeBoneTransform(Bone=upperarm_l, Space=clavicle_l).Transform.Rotation)
   /Script/ControlRig.RigUnit_AimBone(
    Bone=dyn_driver_shldrpad_l
    DebugSettings=(Scale=10,WorldOffset=(Rotation=(X=0,Y=0,Z=0,W=1),Translation=(X=0,Y=0,Z=0),Scale3D=(X=1,Y=1,Z=1)))
    Primary=(Weight=1,Axis=(X=1,Y=0,Z=0),Target=(X=1,Y=0,Z=0),Kind=Location)
    Secondary=(Weight=0,Axis=(X=0,Y=0,Z=0),Target=(X=1,Y=0,Z=0),Kind=Location)
    bPropagateToChildren=True
    Weight=/Script/RigVM.RigVMFunction_MathFloatMax(
     A=&RigVMFunction_MathFloatRemap(SourceMaximum=10, SourceMinimum=50, TargetMaximum=0.7, TargetMinimum=0, bClamp=true, Value=@n2.Result.Y).Result
     B=&RigVMFunction_MathFloatRemap(SourceMaximum=90, SourceMinimum=55, TargetMaximum=0.4, TargetMinimum=0, bClamp=true, Value=@n3.Result.Z).Result
    ) {is=struct RigVMUnitNode, template="Maximum::Execute(in A,in B,out Result)"}.Result
    ExecutePin=@n1.ExecutePin
    Primary.Target=&RigUnit_GetBoneTransform(Bone=upperarm_twist_02_l).Transform.Translation
   ) {is=struct RigVMUnitNode}
   n4 = &RigUnit_GetRelativeBoneTransform(Bone=upperarm_l, Space=clavicle_l)
   > /Script/ControlRig.RigUnit_ModifyBoneTransforms(
    BoneToModify=((Bone="dyn_driver_shldrpad_l",Transform=(Rotation=(X=0,Y=0,Z=0,W=1),Translation=(X=3,Y=0,Z=0.5),Scale3D=(X=1,Y=1,Z=1))))
    Mode=AdditiveLocal
    WeightMaximum=1
    Weight=&RigVMFunction_MathFloatRemap(
     SourceMaximum=80
     SourceMinimum=0
     TargetMinimum=0
     bClamp=true
     Value=&RigVMFunction_MathQuaternionToEuler(RotationOrder=ZXY, Value=@n4.Transform.Rotation).Result.Z
    ).Result
    BoneToModify.0.Transform.Rotation=&RigVMFunction_MathQuaternionFromEuler(Euler=(X=15,Y=10,Z=11)).Result
   ) {is=struct RigVMUnitNode}
   RigUnit_GetBoneTransform = &RigUnit_GetBoneTransform(Bone=lowerarm_l)
   n5 = &RigVMFunction_MathVectorDistance(A=@RigUnit_GetBoneTransform.Transform.Translation, B=&RigUnit_GetBoneTransform(Bone=dyn_arm_up).Transform.Translation)
   > &RigUnit_ModifyBoneTransforms(
    BoneToModify=((Bone="dyn_driver_shldrpad_l",Transform=(Rotation=(X=0,Y=0,Z=0,W=1),Translation=(X=5,Y=0,Z=0),Scale3D=(X=1,Y=1,Z=1))))
    Mode=AdditiveLocal
    WeightMaximum=1
    Weight=&RigVMFunction_MathFloatRemap(SourceMaximum=11, SourceMinimum=25, bClamp=true, Value=@n5.Result).Result
   )
   /Script/RigVM.RigVMFunction_MathFloatClamp(Maximum=11, Minimum=0, Value=@n5.Result) {is=struct RigVMUnitNode, template="Clamp::Execute(in Value,in Minimum,in Maximum,out Result)"}
   &RigVMFunction_MathFloatRemap(
    SourceMaximum=-65
    SourceMinimum=0
    TargetMinimum=0
    bClamp=true
    Value=&RigVMFunction_MathQuaternionToEuler(RotationOrder=ZYX, Value=&RigUnit_GetRelativeBoneTransform(Bone=upperarm_l, Space=clavicle_l).Transform.Rotation).Result.Y
   )
  ]}
]
hierarchy [
 root {loc=<0 0 0>, rot=<0 0 0 1>}
 pelvis {parent=root, loc=<0 -0.5301 98.0571>, rot=<0.0127 0.707 -0.0127 -0.707>}
 spine_01 {parent=pelvis, loc=<6.1208 -0.9328 0>, rot=<0 0 0.034 -0.9994>}
 spine_02 {parent=spine_01, loc=<6.5996 -0.2095 0>, rot=<0 0 0.0238 -0.9997>}
 spine_03 {parent=spine_02, loc=<6.4914 -0.1237 0>, rot=<0 0 0.0149 -0.9999>}
 spine_04 {parent=spine_03, loc=<6.9598 0.0794 0>, rot=<0 0 -0.0132 -0.9999>}
 spine_05 {parent=spine_04, loc=<6.9987 0 0>, rot=<0 0 -0.0015 -1>}
 clavicle_l {parent=spine_05, loc=<5.4831 -0.1053 -3.2946>, rot=<0.0465 -0.7289 -0.009 -0.6829>}
 upperarm_l {parent=clavicle_l, loc=<9.8658 -1.2136 -0.4105>, rot=<0.0164 -0.3245 -0.0269 -0.9454>}
 lowerarm_l {parent=upperarm_l, loc=<26.465 -0.8345 0.259>, rot=<0.0036 0.0161 -0.2298 -0.9731>}
 hand_l {parent=lowerarm_l, loc=<22.7355 -0.3203 -0.6105>, rot=<-0.6784 -0.0223 0.0735 -0.7307>}
 index_metacarpal_l {parent=hand_l, loc=<1.6535 0.0183 -2.96>, rot=<0.0158 -0.0124 -0.0173 -0.9996>}
 index_01_l {parent=index_metacarpal_l, loc=<8.3008 0.0357 -0.0524>, rot=<-0.0044 0.0485 0.0924 -0.9945>}
 index_02_l {parent=index_01_l, loc=<3.3587 -0.3003 0.0452>, rot=<-0.002 0.0027 0.3225 -0.9466>}
 index_03_l {parent=index_02_l, loc=<2.5578 -0.123 -0.0252>, rot=<-0.002 0.0067 0.2558 -0.9667>}
 middle_metacarpal_l {parent=hand_l, loc=<2.1856 -0.1725 -0.5984>, rot=<0.006 0.0168 -0.0329 -0.9993>}
 middle_01_l {parent=middle_metacarpal_l, loc=<7.3275 -0.2447 0.0129>, rot=<-0.0208 0.0042 0.0858 -0.9961>}
 middle_02_l {parent=middle_01_l, loc=<4.0123 0.1401 0.1175>, rot=<-0.0218 0.0258 0.3512 -0.9357>}
 middle_03_l {parent=middle_02_l, loc=<2.6077 0.0534 0.1373>, rot=<0.0082 0.0188 0.1708 -0.9851>}
 pinky_metacarpal_l {parent=hand_l, loc=<1.7616 -0.9753 3.3814>, rot=<-0.0397 0.0814 -0.069 -0.9935>}
 pinky_01_l {parent=pinky_metacarpal_l, loc=<7.1709 -0.4831 -0.1094>, rot=<-0.0654 -0.0035 0.2242 -0.9723>}
 pinky_02_l {parent=pinky_01_l, loc=<2.8274 0.0892 -0.1043>, rot=<0.0037 0.0011 0.2854 -0.9584>}
 pinky_03_l {parent=pinky_02_l, loc=<2.2046 0.0308 -0.0078>, rot=<-0.0002 0.0098 0.1445 -0.9895>}
 ring_metacarpal_l {parent=hand_l, loc=<2.2074 -0.3625 1.3638>, rot=<-0.0284 0.0722 -0.0458 -0.9959>}
 ring_01_l {parent=ring_metacarpal_l, loc=<7.0295 -0.6197 -0.1446>, rot=<-0.0489 -0.0252 0.1892 -0.9804>}
 ring_02_l {parent=ring_01_l, loc=<3.6174 0.4647 -0.1027>, rot=<0 0 0.2806 -0.9598>}
 ring_03_l {parent=ring_02_l, loc=<2.4152 0.1141 0.0199>, rot=<0 0 0.1189 -0.9929>}
 thumb_01_l {parent=hand_l, loc=<1.3118 -0.3191 -2.459>, rot=<0.6546 -0.4042 0.0664 -0.6354>}
 thumb_02_l {parent=thumb_01_l, loc=<5.9336 -0.0448 0.0401>, rot=<-0.0037 0.0213 0.298 -0.9543>}
 thumb_03_l {parent=thumb_02_l, loc=<2.2223 0.3682 -0.0907>, rot=<0.0118 -0.013 0.0066 -0.9998>}
 lowerarm_twist_01_l {parent=lowerarm_l, loc=<18.316 0 0>, rot=<0 0 0 -1>}
 lowerarm_twist_02_l {parent=lowerarm_l, loc=<10.6707 0 0>, rot=<0 0 0 -1>}
 upperarm_twist_01_l {parent=upperarm_l, loc=<7.2135 -0.1134 0>, rot=<0 0 0 -1>}
 upperarm_twist_02_l {parent=upperarm_l, loc=<19.7998 -0.1134 0>, rot=<0 0 0 -1>}
 dyn_parent_shldrpad_l {parent=clavicle_l, loc=<10.3137 -1.1782 3.2868>, rot=<0.0008 0.0336 -0.0252 -0.9991>}
 dyn_driver_shldrpad_l {parent=dyn_parent_shldrpad_l, loc=<-1.8827 0.1949 -2.749>, rot=<0.0134 -0.3555 0.014 -0.9345>}
 dyn_shldrpad_01_l {parent=dyn_driver_shldrpad_l, loc=<-7.965 -0.2315 3.3757>, rot=<0.7009 0.0042 0.0067 -0.7133>}
 dyn_shldrpad_02_l {parent=dyn_shldrpad_01_l, loc=<13.2877 1.0083 0>, rot=<0 0 0 -1>}
 dyn_shldrpad_03_l {parent=dyn_shldrpad_02_l, loc=<-5.2707 -5.58 0>, rot=<0 0 0 -1>}
 dyn_arm_up {parent=clavicle_l, loc=<28.0371 -7.3264 39.3807>, rot=<-0.0008 -0.0336 0.0252 0.9991>}
 clavicle_r {parent=spine_05, loc=<5.4831 -0.1053 3.2946>, rot=<0.7289 0.0465 -0.6829 0.009>}
 upperarm_r {parent=clavicle_r, loc=<-9.8658 1.2136 0.4104>, rot=<0.0164 -0.3245 -0.0269 -0.9454>}
 lowerarm_r {parent=upperarm_r, loc=<-26.4651 0.8345 -0.259>, rot=<0.0036 0.0161 -0.2298 -0.9731>}
 hand_r {parent=lowerarm_r, loc=<-22.7356 0.3203 0.6106>, rot=<-0.6784 -0.0223 0.0735 -0.7307>}
 index_metacarpal_r {parent=hand_r, loc=<-1.6534 -0.0181 2.9601>, rot=<0.0158 -0.0124 -0.0173 -0.9996>}
 index_01_r {parent=index_metacarpal_r, loc=<-8.3008 -0.0357 0.0524>, rot=<-0.0044 0.0485 0.0924 -0.9945>}
 index_02_r {parent=index_01_r, loc=<-3.3589 0.3003 -0.0453>, rot=<-0.002 0.0027 0.3225 -0.9466>}
 index_03_r {parent=index_02_r, loc=<-2.5578 0.123 0.0252>, rot=<-0.002 0.0067 0.2558 -0.9667>}
 middle_metacarpal_r {parent=hand_r, loc=<-2.1856 0.1726 0.5984>, rot=<0.006 0.0168 -0.0329 -0.9993>}
 middle_01_r {parent=middle_metacarpal_r, loc=<-7.3273 0.2447 -0.0129>, rot=<-0.0208 0.0042 0.0858 -0.9961>}
 middle_02_r {parent=middle_01_r, loc=<-4.0124 -0.14 -0.1175>, rot=<-0.0218 0.0258 0.3512 -0.9357>}
 middle_03_r {parent=middle_02_r, loc=<-2.6077 -0.0533 -0.1373>, rot=<0.0082 0.0188 0.1708 -0.9851>}
 pinky_metacarpal_r {parent=hand_r, loc=<-1.7617 0.9753 -3.3814>, rot=<-0.0397 0.0814 -0.069 -0.9935>}
 pinky_01_r {parent=pinky_metacarpal_r, loc=<-7.1707 0.4832 0.1094>, rot=<-0.0654 -0.0035 0.2242 -0.9723>}
 pinky_02_r {parent=pinky_01_r, loc=<-2.8274 -0.0892 0.1043>, rot=<0.0037 0.0011 0.2854 -0.9584>}
 pinky_03_r {parent=pinky_02_r, loc=<-2.2045 -0.031 0.0078>, rot=<-0.0002 0.0098 0.1445 -0.9895>}
 ring_metacarpal_r {parent=hand_r, loc=<-2.2073 0.3627 -1.3638>, rot=<-0.0284 0.0722 -0.0458 -0.9959>}
 ring_01_r {parent=ring_metacarpal_r, loc=<-7.0297 0.6197 0.1446>, rot=<-0.0489 -0.0252 0.1892 -0.9804>}
 ring_02_r {parent=ring_01_r, loc=<-3.6172 -0.4648 0.1027>, rot=<0 0 0.2806 -0.9598>}
 ring_03_r {parent=ring_02_r, loc=<-2.4153 -0.1141 -0.0199>, rot=<0 0 0.1189 -0.9929>}
 thumb_01_r {parent=hand_r, loc=<-1.3119 0.3192 2.459>, rot=<0.6546 -0.4042 0.0664 -0.6354>}
 thumb_02_r {parent=thumb_01_r, loc=<-5.9336 0.0447 -0.0401>, rot=<-0.0037 0.0213 0.298 -0.9543>}
 thumb_03_r {parent=thumb_02_r, loc=<-2.2223 -0.3682 0.0907>, rot=<0.0118 -0.013 0.0066 -0.9998>}
 lowerarm_twist_01_r {parent=lowerarm_r, loc=<-18.3161 0 0>, rot=<0 0 0 1>}
 lowerarm_twist_02_r {parent=lowerarm_r, loc=<-10.6707 0 0>, rot=<0 0 0 1>}
 upperarm_twist_01_r {parent=upperarm_r, loc=<-7.2135 0.1134 0>, rot=<0 0 0 1>}
 upperarm_twist_02_r {parent=upperarm_r, loc=<-19.7998 0.1134 0>, rot=<0 0 0 1>}
 dyn_ponytail_aim {parent=spine_05, loc=<9.552 -6.9215 0.0017>, rot=<0.2062 -0.9785 0 0>}
 thigh_l {parent=pelvis, loc=<-1.0855 -0.2109 -8.7053>, rot=<0.0009 0.0355 0.0155 -0.9993>}
 calf_l {parent=thigh_l, loc=<-42.5284 -0.0019 -0.7297>, rot=<0.0006 -0.0161 -0.0391 -0.9991>}
 calf_twist_01_l {parent=calf_l, loc=<-35.8483 -1.9197 -0.5794>, rot=<0.0237 0.0095 -0.0035 -0.9997>}
 calf_twist_02_l {parent=calf_l, loc=<-19.1546 -1.7943 -0.2638>, rot=<0.0237 0.0095 -0.0035 -0.9997>}
 foot_l {parent=calf_l, loc=<-46.7401 -2.5014 -0.592>, rot=<0.0427 -0.0238 0.0353 -0.9982>}
 ball_l {parent=foot_l, loc=<-3.5658 10.8066 -0.1544>, rot=<0.0001 -0.0001 -0.7063 -0.7079>}
 thigh_twist_01_l {parent=thigh_l, loc=<-27.952 0 0>, rot=<0.0013 0 0 -1>}
 sub_thigh_l {parent=thigh_l, loc=<-2.972 0 0>, rot=<0 0 0 -1>}
 driver_frill_frt_l {parent=sub_thigh_l, loc=<3.6841 9.2434 -0.0995>, rot=<0 0 0 -1>}
 dyn_frill_frt_l {parent=driver_frill_frt_l, loc=<0 0 0>, rot=<0 0 0 -1>}
 dyn_parent_strap_B_01_l {parent=thigh_l, loc=<-2.568 -4.597 -6.4143>, rot=<-0.0115 -0.6544 0.7548 -0.0437>}
 dyn_driver_strap_B_01_l {parent=dyn_parent_strap_B_01_l, loc=<0.6461 0 0>, rot=<-0.0872 0 0 -0.9962>}
 dyn_strap_B_01_l {parent=dyn_driver_strap_B_01_l, loc=<0 0 0>, rot=<0 0 0 1>}
 dyn_strap_B_02_l {parent=dyn_strap_B_01_l, loc=<7.6417 0.153 0.027>, rot=<0 -0.0079 0.0449 -0.999>}
 dyn_strap_B_03_l {parent=dyn_strap_B_02_l, loc=<6.2423 -0.0002 0>, rot=<0 -0.0036 0.0204 -0.9998>}
 dyn_strap_B_04_l {parent=dyn_strap_B_03_l, loc=<5.4134 0.0002 0>, rot=<0 0 0 1>}
 dyn_strap_B_05_l {parent=dyn_strap_B_04_l, loc=<4.6933 -0.02 -0.0035>, rot=<0 0 0 1>}
 dyn_parent_strap_A_01_l {parent=thigh_l, loc=<-3.4464 4.7286 -6.4437>, rot=<-0.0348 -0.7857 0.6171 -0.0243>}
 dyn_driver_strap_A_01_l {parent=dyn_parent_strap_A_01_l, loc=<0.6461 0 0>, rot=<0 0 0 1>}
 dyn_strap_A_01_l {parent=dyn_driver_strap_A_01_l, loc=<0 0 0>, rot=<0 0 0 1>}
 dyn_strap_A_02_l {parent=dyn_strap_A_01_l, loc=<7.3791 0.1685 -0.0613>, rot=<0 0.0156 0.0429 -0.999>}
 dyn_strap_A_03_l {parent=dyn_strap_A_02_l, loc=<6.2423 -0.0002 0.0001>, rot=<0 0 0 1>}
 dyn_strap_A_04_l {parent=dyn_strap_A_03_l, loc=<5.4135 0.0002 -0.0001>, rot=<0 0 0 1>}
 dyn_strap_A_05_l {parent=dyn_strap_A_04_l, loc=<4.6933 -0.019 0.0069>, rot=<0 0 0 1>}
 thigh_r {parent=pelvis, loc=<-1.0853 -0.2109 8.7053>, rot=<0.0355 -0.0009 0.9993 0.0155>}
 calf_r {parent=thigh_r, loc=<42.5286 0.0019 0.7297>, rot=<0.0006 -0.0161 -0.0391 -0.9991>}
 calf_twist_01_r {parent=calf_r, loc=<35.8485 1.9197 0.5794>, rot=<0.0237 0.0095 -0.0035 -0.9997>}
 calf_twist_02_r {parent=calf_r, loc=<19.1546 1.7943 0.2638>, rot=<0.0237 0.0095 -0.0035 -0.9997>}
 foot_r {parent=calf_r, loc=<46.7402 2.5014 0.592>, rot=<0.0427 -0.0238 0.0353 -0.9982>}
 ball_r {parent=foot_r, loc=<3.5658 -10.8066 0.1545>, rot=<0.0001 -0.0001 -0.7063 -0.7079>}
 thigh_twist_01_r {parent=thigh_r, loc=<27.952 0 0>, rot=<0.0013 0 0 -1>}
 sub_thigh_r {parent=thigh_r, loc=<2.9722 0 0>, rot=<0 0 0 -1>}
 driver_frill_frt_r {parent=sub_thigh_r, loc=<-3.6841 -9.2434 0.0995>, rot=<0 0 0 -1>}
 dyn_frill_frt_r {parent=driver_frill_frt_r, loc=<0 0 0>, rot=<0 0 0 -1>}
 dyn_parent_strap_A_01_r {parent=thigh_r, loc=<2.9997 -4.7176 6.4073>, rot=<-0.7857 0.0348 -0.0243 -0.6171>}
 dyn_driver_strap_A_01_r {parent=dyn_parent_strap_A_01_r, loc=<0.6461 0 0>, rot=<0 0 0 1>}
 dyn_strap_A_01_r {parent=dyn_driver_strap_A_01_r, loc=<0 0 0>, rot=<0 0 0 1>}
 dyn_strap_A_02_r {parent=dyn_strap_A_01_r, loc=<7.379 0.1685 0.0614>, rot=<0 -0.0156 0.0429 -0.999>}
 dyn_strap_A_03_r {parent=dyn_strap_A_02_r, loc=<6.2423 -0.0002 -0.0001>, rot=<0 0 0 -1>}
 dyn_strap_A_04_r {parent=dyn_strap_A_03_r, loc=<5.4134 0.0002 0.0001>, rot=<0 0 0 -1>}
 dyn_strap_A_05_r {parent=dyn_strap_A_04_r, loc=<4.6933 -0.019 -0.0069>, rot=<0 0 0 -1>}
 dyn_parent_strap_B_01_r {parent=thigh_r, loc=<2.5682 4.597 6.4143>, rot=<-0.6544 0.0115 -0.0437 -0.7548>}
 dyn_driver_strap_B_01_r {parent=dyn_parent_strap_B_01_r, loc=<0.6461 0 0>, rot=<0.0872 0 0 -0.9962>}
 dyn_strap_B_01_r {parent=dyn_driver_strap_B_01_r, loc=<0 0 0>, rot=<0 0 0 1>}
 dyn_strap_B_02_r {parent=dyn_strap_B_01_r, loc=<7.6417 0.153 -0.027>, rot=<0 0.0079 0.0449 -0.999>}
 dyn_strap_B_03_r {parent=dyn_strap_B_02_r, loc=<6.2423 -0.0003 0>, rot=<0 0.0036 0.0204 -0.9998>}
 dyn_strap_B_04_r {parent=dyn_strap_B_03_r, loc=<5.4135 0.0003 0>, rot=<0 0 0 -1>}
 dyn_strap_B_05_r {parent=dyn_strap_B_04_r, loc=<4.6933 -0.02 0.0035>, rot=<0 0 0 -1>}
 dyn_parent_strap_01_r {parent=pelvis, loc=<2.391 0.0649 8.0313>, rot=<-0.0127 -0.707 0.0127 -0.707>}
 dyn_driver_strap_01_r {parent=dyn_parent_strap_01_r, loc=<0 0 0>, rot=<0 0 0 1>}
 dyn_strap_01_r {parent=dyn_driver_strap_01_r, loc=<0 0 -0.0003>, rot=<0 0.3763 0 -0.9265>}
 dyn_parent_strap_bck_01_r {parent=pelvis, loc=<2.9035 -11.1544 7.2016>, rot=<0.0449 -0.7057 0.0702 -0.7036>}
 dyn_driver_strap_bck_01_r {parent=dyn_parent_strap_bck_01_r, loc=<0 0 0>, rot=<0 0 0 1>}
 dyn_strap_bck_01_r {parent=dyn_driver_strap_bck_01_r, loc=<0 0 0>, rot=<0 0 0 1>}
 dyn_parent_strap_01_l {parent=pelvis, loc=<2.3907 0.0649 -8.0313>, rot=<0.707 -0.0127 -0.707 -0.0127>}
 dyn_driver_strap_01_l {parent=dyn_parent_strap_01_l, loc=<0 0 0>, rot=<0 0 0 1>}
 dyn_strap_01_l {parent=dyn_driver_strap_01_l, loc=<0 0 0>, rot=<0 0.3763 0 -0.9265>}
 dyn_parent_strap_bck_01_l {parent=pelvis, loc=<2.9032 -11.1543 -7.2016>, rot=<0.7057 0.0449 -0.7036 -0.0702>}
 dyn_driver_strap_bck_01_l {parent=dyn_parent_strap_bck_01_l, loc=<0 0 0>, rot=<0 0 0 -1>}
 dyn_strap_bck_01_l {parent=dyn_driver_strap_bck_01_l, loc=<0 0 0>, rot=<0 0 0 -1>}
 dyn_loinClothBack_01_Base {parent=pelvis, loc=<-3.1849 -12.0211 0.3209>, rot=<-0.0179 -0.9998 0 0>}
 dyn_loinClothBack_01 {parent=dyn_loinClothBack_01_Base, loc=<0 0 0>, rot=<0 0 0 -1>}
 dyn_loinClothBack_02 {parent=dyn_loinClothBack_01, loc=<7.8997 0 0>, rot=<0 0 0.028 -0.9996>}
 dyn_loinClothBack_03 {parent=dyn_loinClothBack_02, loc=<7.8749 0.0091 0>, rot=<0 0 0.0143 -0.9999>}
 dyn_loinClothBack_04 {parent=dyn_loinClothBack_03, loc=<7.5855 0 0>, rot=<0 0 0.0087 -1>}
 dyn_loinClothBack_05 {parent=dyn_loinClothBack_04, loc=<7.2348 0 0>, rot=<0 0 0 1>}
 dyn_parent_strap_01 {parent=pelvis, loc=<2.391 0.0649 8.0313>, rot=<-0.0127 -0.707 0.0127 -0.707>}
 dyn_driver_strap_01 {parent=dyn_parent_strap_01, loc=<0 0 0>, rot=<0 0 0 1>}
 dyn_strap_01 {parent=dyn_driver_strap_01, loc=<0 0 0>, rot=<0 0 0 1>}
 dyn_glute_r {parent=pelvis, loc=<-5.2839 -3.8044 7.094>, rot=<-0.0127 -0.707 0.0127 -0.707>}
 dyn_glute_l {parent=pelvis, loc=<-5.2839 -3.8044 -7.094>, rot=<-0.0127 -0.707 0.0127 -0.707>}
 Flap_Front_Parent_Space {parent=pelvis, loc=<-2.6394 10.9147 0>, rot=<-0.0127 -0.707 0.0127 -0.707>}
 Flap_Front_Driven_Bone {parent=Flap_Front_Parent_Space, loc=<0 -0.3116 0>, rot=<0 0 0 1>}
 dyn_frill_frt {parent=Flap_Front_Driven_Bone, loc=<0 0 0>, rot=<0 0 0 1>}
 dyn_pelvis_offset {parent=pelvis, loc=<-2.6394 10.9147 0>, rot=<-0.0127 -0.707 0.0127 -0.707>}
 dyn_parent_strap_frt_01_r {parent=dyn_pelvis_offset, loc=<-7.34 -1.7426 4.7086>, rot=<-0.0597 -0.0019 -0.0085 -0.9982>}
 dyn_driver_strap_frt_01_r {parent=dyn_parent_strap_frt_01_r, loc=<0 0 0>, rot=<0 0 0 1>}
 dyn_strap_frt_01_r {parent=dyn_driver_strap_frt_01_r, loc=<0 0 0>, rot=<0 0 0 1>}
 dyn_parent_strap_frt_01_l {parent=dyn_pelvis_offset, loc=<7.34 -1.7426 4.7086>, rot=<0.9982 -0.0085 0.0019 -0.0597>}
 dyn_driver_strap_frt_01_l {parent=dyn_parent_strap_frt_01_l, loc=<0 0 0>, rot=<0 0 0 1>}
 dyn_strap_frt_01_l {parent=dyn_driver_strap_frt_01_l, loc=<0 0 0>, rot=<0 0 0 1>}
 dyn_parent_frill_frt_01 {parent=dyn_pelvis_offset, loc=<0 -0.54 -0.8409>, rot=<0.0122 -0.707 0.0122 -0.707>}
 dyn_driver_frill_frt_01 {parent=dyn_parent_frill_frt_01, loc=<0.1547 0 0>, rot=<0 0 0 -1>}
 dyn_frill_frt_01 {parent=dyn_driver_frill_frt_01, loc=<7.7867 0 0>, rot=<0 0 0 -1>}
 dyn_frill_frt_02 {parent=dyn_frill_frt_01, loc=<5.7546 0.0063 0>, rot=<0 0 0 -1>}
 dyn_frill_frt_03 {parent=dyn_frill_frt_02, loc=<5.6804 0.0294 0>, rot=<0 0 0 -1>}
 dyn_frill_frt_04 {parent=dyn_frill_frt_03, loc=<5.183 0.0926 0>, rot=<0 0 0 -1>}
 dyn_frill_frt_05 {parent=dyn_frill_frt_04, loc=<5.5609 0.0993 0>, rot=<0 0 0 -1>}
 dyn_parent_frill_frtA_01_l {parent=dyn_pelvis_offset, loc=<5.1304 -0.8627 0.9832>, rot=<0.0041 -0.7071 0.0041 -0.7071>}
 dyn_driver_frill_frtA_01_l {parent=dyn_parent_frill_frtA_01_l, loc=<0.6461 0 0>, rot=<0 0 0 -1>}
 dyn_frill_frtA_01_l {parent=dyn_driver_frill_frtA_01_l, loc=<9.3392 0 0>, rot=<0 0 0 -1>}
 dyn_frill_frtA_02_l {parent=dyn_frill_frtA_01_l, loc=<5.9021 -0.0002 0>, rot=<0 0 0 -1>}
 dyn_frill_frtA_03_l {parent=dyn_frill_frtA_02_l, loc=<5.8541 0.0002 0>, rot=<0 0 0 -1>}
 dyn_frill_frtA_04_l {parent=dyn_frill_frtA_03_l, loc=<4.1077 -0.0203 0>, rot=<0 0 0 -1>}
 dyn_frill_frtA_05_l {parent=dyn_frill_frtA_04_l, loc=<3.9129 -0.0193 0>, rot=<0 0 0 -1>}
 dyn_parent_frill_frtA_01_r {parent=dyn_pelvis_offset, loc=<-4.7111 -0.9103 0.7083>, rot=<0.0042 -0.715 0.0043 -0.6991>}
 dyn_driver_frill_frtA_01_r {parent=dyn_parent_frill_frtA_01_r, loc=<0.6461 0 0>, rot=<0 0 0 -1>}
 dyn_frill_frtA_01_r {parent=dyn_driver_frill_frtA_01_r, loc=<8.9145 -0.0441 0.0438>, rot=<0 0 0 -1>}
 dyn_frill_frtA_02_r {parent=dyn_frill_frtA_01_r, loc=<5.7853 0 0>, rot=<0 0 0 -1>}
 dyn_frill_frtA_03_r {parent=dyn_frill_frtA_02_r, loc=<5.8125 -0.0019 0>, rot=<0 0 0 -1>}
 dyn_frill_frtA_04_r {parent=dyn_frill_frtA_03_r, loc=<3.8578 -0.0182 0>, rot=<0 0 0 -1>}
 dyn_frill_frtA_05_r {parent=dyn_frill_frtA_04_r, loc=<3.9571 -0.0186 0>, rot=<0 0 0 -1>}
 strap_aim_01_l {parent=pelvis, loc=<-3.7575 10.0673 -5.1304>, rot=<-0.0121 -0.9999 0 0>}
 strap_aim_02_l {parent=pelvis, loc=<-11.2306 10.2478 -5.1304>, rot=<-0.0121 -0.9999 0 0>}
 strap_aim_03_l {parent=pelvis, loc=<-17.5705 10.4006 -5.1304>, rot=<-0.012 -0.9999 0 0>}
 strap_aim_04_l {parent=pelvis, loc=<-23.4391 9.8614 -5.1304>, rot=<-0.012 -0.9999 0 0>}
 strap_aim_01_r {parent=pelvis, loc=<-4.0338 10.0291 4.7256>, rot=<-0.012 -0.9999 0 0>}
 strap_aim_02_r {parent=pelvis, loc=<-11.5217 10.1641 4.8815>, rot=<-0.012 -0.9999 0 0>}
 strap_aim_03_r {parent=pelvis, loc=<-17.3039 10.3024 5.0111>, rot=<-0.012 -0.9999 0 0>}
 strap_aim_04_r {parent=pelvis, loc=<-23.1281 9.8232 5.1413>, rot=<-0.012 -0.9999 0 0>}
 strap_aim_01_c {parent=pelvis, loc=<-4.8037 10.4329 0>, rot=<-0.012 -0.9999 0 0>}
 strap_aim_02_c {parent=pelvis, loc=<-11.4405 10.4159 0>, rot=<-0.012 -0.9999 0 0>}
 strap_aim_03_c {parent=pelvis, loc=<-17.1951 10.43 0>, rot=<-0.012 -0.9999 0 0>}
 strap_aim_04_c {parent=pelvis, loc=<-22.8883 9.9306 0>, rot=<-0.012 -0.9999 0 0>}
 NewControl {type=Control, loc=<0 0 0>, rot=<0 0 0 1>}
]
