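# Physics asset for the hip chains of the female medium Ancient Gladiator body, covering a left and right strap pair around the pelvis and thighs.
# Capsule bodies model the pelvis, both thighs, and a driver and strap bone on each side.
# Thighs and driver bones have angularly free constraints to the pelvis.
# Each strap hangs off its driver with narrow swing and very tight twist limits, slightly asymmetric between sides.
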PhysicsAsset /Game/Characters/Player/Female/Medium/Bodies/F_MED_Ancient_Gladiator/Meshes/physics/F_MED_Ancient_Gladiator_Chains_Hips
bodies {
 dyn_driver_strap_01_l [{shapeType=Capsule, length=2.5, radius=1.75, center=<-1 0 0>, rotation=<rot 90 0 0>}]
 dyn_driver_strap_01_r [{shapeType=Capsule, length=1, radius=1}]
 dyn_strap_01_l [{shapeType=Capsule, length=2.25, radius=2, rotation=<rot 90 0 0>}]
 dyn_strap_01_r [{shapeType=Capsule, length=2.5, radius=1.75, center=<-1 0 0>, rotation=<rot 90 0 0>}]
 pelvis [{shapeType=Capsule, length=6.4689, radius=3.0903}]
 thigh_l [{shapeType=Capsule, length=17, radius=8.4048, center=<-16.2315 -0.0332 0.1396>, rotation=<rot -0.7496 -91.2078 -92.9031>}]
 thigh_r [{shapeType=Capsule, length=17, radius=8.4048, center=<10.5433 0.9478 -0.4902>, rotation=<rot -0.4897 -88.6883 -95.9058>}]
}
constraints [
 {
  bone1Name=thigh_r
  bone2Name=pelvis
  swing1LimitDegrees=45
  swing1Motion=Free
  swing2LimitDegrees=45
  swing2Motion=Free
  twistLimitDegrees=45
  twistMotion=Free
 }
 {
  bone1Name=dyn_driver_strap_01_r
  bone2Name=pelvis
  swing1LimitDegrees=45
  swing1Motion=Free
  swing2LimitDegrees=45
  swing2Motion=Free
  twistLimitDegrees=45
  twistMotion=Free
 }
 {bone1Name=dyn_strap_01_r, bone2Name=dyn_driver_strap_01_r, swing1LimitDegrees=10, swing2LimitDegrees=25, twistLimitDegrees=2}
 {
  bone1Name=dyn_driver_strap_01_l
  bone2Name=pelvis
  swing1LimitDegrees=45
  swing1Motion=Free
  swing2LimitDegrees=45
  swing2Motion=Free
  twistLimitDegrees=45
  twistMotion=Free
 }
 {bone1Name=dyn_strap_01_l, bone2Name=dyn_driver_strap_01_l, swing1LimitDegrees=15, swing2LimitDegrees=25, twistLimitDegrees=2}
 {
  bone1Name=thigh_l
  bone2Name=pelvis
  swing1LimitDegrees=45
  swing1Motion=Free
  swing2LimitDegrees=45
  swing2Motion=Free
  twistLimitDegrees=45
  twistMotion=Free
 }
]
