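# Physics asset for the rear loincloth of the female medium Ancient Gladiator player body.
# Three box bodies form the cloth chain, hanging from a small base box that joins to the pelvis with free angular motion.
# The chain links share identical swing and much tighter twist limits.
# Capsules on the pelvis and both thighs, also joined by free angular constraints, give body shapes for the cloth to collide against.
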PhysicsAsset /Game/Characters/Player/Female/Medium/Bodies/F_MED_Ancient_Gladiator/Meshes/physics/F_MED_Ancient_Gladiator_Rear_Physics
bodies {
 dyn_loinClothBack_01 [{shapeType=Box, extentX=3.1431, extentY=9.134, extentZ=20.5999, center=<3.5044 0.0001 0>, rotation=<rot 0 90 0>}]
 dyn_loinClothBack_01_Base [{shapeType=Box, extentX=1.01, extentY=1.01, extentZ=1.01, rotation=<rot 0 90 0>}]
 dyn_loinClothBack_02 [{shapeType=Box, extentX=2.4797, extentY=10.688, extentZ=19.0518, center=<5.7318 0.0002 0>, rotation=<rot 0 90 0>}]
 dyn_loinClothBack_03 [{shapeType=Box, extentX=2.4395, extentY=11.5094, extentZ=17.529, center=<8.7738 0 0>, rotation=<rot 0 90 0>}]
 pelvis [{shapeType=Capsule, length=10.1327, radius=11.2063, center=<-5.6844 0.9709 0>}]
 thigh_l [{shapeType=Capsule, length=23.9255, radius=8.7894, center=<-16.5241 -0.0007 -0.0869>, rotation=<rot 90 0 180>}]
 thigh_r [{shapeType=Capsule, length=23.9255, radius=8.7894, center=<16.5241 0.0007 0.0869>, rotation=<rot 90 0 0>}]
}
constraints [
 {
  bone1Name=dyn_loinClothBack_01_Base
  bone2Name=pelvis
  swing1LimitDegrees=45
  swing1Motion=Free
  swing2LimitDegrees=45
  swing2Motion=Free
  twistLimitDegrees=45
  twistMotion=Free
 }
 {bone1Name=dyn_loinClothBack_01, bone2Name=dyn_loinClothBack_01_Base, swing1LimitDegrees=35, swing2LimitDegrees=25, twistLimitDegrees=6}
 {bone1Name=dyn_loinClothBack_02, bone2Name=dyn_loinClothBack_01, swing1LimitDegrees=35, swing2LimitDegrees=25, twistLimitDegrees=6}
 {bone1Name=dyn_loinClothBack_03, bone2Name=dyn_loinClothBack_02, swing1LimitDegrees=35, swing2LimitDegrees=25, twistLimitDegrees=6}
 {
  bone1Name=thigh_r
  bone2Name=pelvis
  swing1LimitDegrees=45
  swing1Motion=Free
  swing2LimitDegrees=45
  swing2Motion=Free
  twistLimitDegrees=45
  twistMotion=Free
 }
 {
  bone1Name=thigh_l
  bone2Name=pelvis
  swing1LimitDegrees=45
  swing1Motion=Free
  swing2LimitDegrees=45
  swing2Motion=Free
  twistLimitDegrees=45
  twistMotion=Free
 }
]
