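# Physics asset for the skirt of the female medium Ancient Gladiator player body, covering its front center, left and right flaps.
# Pelvis and thigh capsules form the body volumes, while boxes and capsules shape the dyn_frill flaps.
# Each flap has an angular constraint to a small driver body, tighter on the sides than the center.
# The center driver links to pelvis and the left to sub_thigh_l, but the right driver has no parent constraint.
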PhysicsAsset /Game/Characters/Player/Female/Medium/Bodies/F_MED_Ancient_Gladiator/Meshes/physics/F_MED_Ancient_Gladiator_Skirt
bodies {
 Flap_Front_Driven_Bone [{shapeType=Sphere, radius=0.505}]
 driver_frill_frt_l [{shapeType=Sphere, radius=1}]
 driver_frill_frt_r [{shapeType=Sphere, radius=0.505}]
 dyn_frill_frt [{shapeType=Box, extentX=13.5, extentY=4, extentZ=20.5, center=<0 1.15 -9>, rotation=<rot 0 0 -3.1066>}]
 dyn_frill_frt_l [
  {shapeType=Capsule, length=13, radius=4.5, center=<-28.4376 3.9978 -1.774>, rotation=<rot 4.1653 -0.5867 11.0184>}
  {shapeType=Box, extentX=19.5, extentY=14, extentZ=5.5, center=<-14.3367 3.4879 -1.1044>, rotation=<rot 2.5452 0.3899 -79.6815>}
 ]
 dyn_frill_frt_r [
  {shapeType=Capsule, length=11.75, radius=5.25, center=<28.747 -3.8386 -0.7062>, rotation=<rot 3.2469 0.451 -3.1687>}
  {shapeType=Box, extentX=23, extentY=14, extentZ=7, center=<12.2499 -2.6581 -0.2725>, rotation=<rot 2.6267 -0.0588 90.1883>}
 ]
 pelvis [{shapeType=Capsule, length=10.8103, radius=8.1851, center=<-3.9396 0 0>}]
 sub_thigh_l [{shapeType=Capsule, length=0.505, radius=0.505, rotation=<rot 0 -90 180>}]
 sub_thigh_r [{shapeType=Capsule, length=0.505, radius=0.505, rotation=<rot 0 -90 0>}]
 thigh_l [{shapeType=Capsule, length=17, radius=9.5, center=<-17.8393 0.1854 0.8317>, rotation=<rot 90 0 176.3396>}]
 thigh_r [{shapeType=Capsule, length=20.25, radius=9.75, center=<17.177 0.625 -0.1326>, rotation=<rot 90 0 -1.5027>}]
}
constraints [
 {
  bone1Name=sub_thigh_r
  bone2Name=thigh_r
  swing1LimitDegrees=45
  swing1Motion=Free
  swing2LimitDegrees=45
  swing2Motion=Free
  twistLimitDegrees=45
  twistMotion=Free
 }
 {
  bone1Name=sub_thigh_l
  bone2Name=thigh_l
  swing1LimitDegrees=45
  swing1Motion=Free
  swing2LimitDegrees=45
  swing2Motion=Free
  twistLimitDegrees=45
  twistMotion=Free
 }
 {bone1Name=dyn_frill_frt_l, bone2Name=driver_frill_frt_l, swing1LimitDegrees=15, swing2LimitDegrees=15, twistLimitDegrees=15}
 {bone1Name=dyn_frill_frt, bone2Name=Flap_Front_Driven_Bone, swing1LimitDegrees=45, swing2LimitDegrees=45, twistLimitDegrees=45}
 {
  bone1Name=thigh_r
  bone2Name=pelvis
  swing1LimitDegrees=45
  swing1Motion=Free
  swing2LimitDegrees=45
  swing2Motion=Free
  twistLimitDegrees=45
  twistMotion=Free
 }
 {
  bone1Name=Flap_Front_Driven_Bone
  bone2Name=pelvis
  swing1LimitDegrees=45
  swing1Motion=Free
  swing2LimitDegrees=45
  swing2Motion=Free
  twistLimitDegrees=45
  twistMotion=Free
 }
 {
  bone1Name=driver_frill_frt_l
  bone2Name=sub_thigh_l
  swing1LimitDegrees=45
  swing1Motion=Free
  swing2LimitDegrees=45
  swing2Motion=Free
  twistLimitDegrees=45
  twistMotion=Free
 }
 {bone1Name=thigh_l, bone2Name=pelvis, swing1LimitDegrees=45, swing2LimitDegrees=45, twistLimitDegrees=45}
 {bone1Name=dyn_frill_frt_r, bone2Name=driver_frill_frt_r, swing1LimitDegrees=20, swing2LimitDegrees=15, twistLimitDegrees=10}
]
