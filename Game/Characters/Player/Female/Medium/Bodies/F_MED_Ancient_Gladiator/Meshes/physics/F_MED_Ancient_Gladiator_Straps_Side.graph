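# Physics asset for the side straps of the female medium Ancient Gladiator body, with capsule bodies for the pelvis, thighs and strap bones.
# Four five-link chains (A and B on each side) each hang from a small driver body through swing- and twist-constrained joints.
# The setup is asymmetric: left drivers attach to the left thigh, right drivers to the pelvis, and the right chains have tighter twist limits.
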PhysicsAsset /Game/Characters/Player/Female/Medium/Bodies/F_MED_Ancient_Gladiator/Meshes/physics/F_MED_Ancient_Gladiator_Straps_Side
bodies {
 dyn_driver_strap_A_01_l [{shapeType=Capsule, length=0.505, radius=0.505, rotation=<rot 0 -90 0>}]
 dyn_driver_strap_A_01_r [{shapeType=Capsule, length=1, radius=1}]
 dyn_driver_strap_B_01_l [{shapeType=Capsule, length=0.505, radius=0.505, rotation=<rot 0 -90 0>}]
 dyn_driver_strap_B_01_r [{shapeType=Capsule, length=1, radius=1}]
 dyn_strap_A_01_l [{shapeType=Capsule, length=7.75, radius=7.3, center=<3 5.5 0>, rotation=<rot 90 0 0>}]
 dyn_strap_A_01_r [{shapeType=Capsule, length=7.75, radius=5.5, center=<3 5.125 0>, rotation=<rot 90 0 0>}]
 dyn_strap_A_02_l [{shapeType=Capsule, length=7.75, radius=7.25, center=<3 5.5 0>, rotation=<rot 90 0 0>}]
 dyn_strap_A_02_r [{shapeType=Capsule, length=5.25, radius=6.2, center=<3 5.5 0>, rotation=<rot 90 0 0>}]
 dyn_strap_A_03_l [{shapeType=Capsule, length=8, radius=7.35, center=<3 5.5 0>, rotation=<rot 90 0 0>}]
 dyn_strap_A_03_r [{shapeType=Capsule, length=4.75, radius=6.45, center=<3 5.5 0>, rotation=<rot 90 0 0>}]
 dyn_strap_A_04_l [{shapeType=Capsule, length=8, radius=7.25, center=<3 5.5 0>, rotation=<rot 90 0 0>}]
 dyn_strap_A_04_r [{shapeType=Capsule, length=7.5, radius=6.95, center=<3 5.5 0>, rotation=<rot 90 0 0>}]
 dyn_strap_A_05_l [{shapeType=Capsule, length=8, radius=7, center=<3 5.5 0>, rotation=<rot 90 0 0>}]
 dyn_strap_A_05_r [{shapeType=Capsule, length=8.25, radius=6.95, center=<3 5.5 0>, rotation=<rot 90 0 0>}]
 dyn_strap_B_01_l [{shapeType=Capsule, length=8, radius=6.5, center=<3 7 0>, rotation=<rot 90 0 0>}]
 dyn_strap_B_01_r [{shapeType=Capsule, length=6, radius=5, center=<3 4.5 0>, rotation=<rot 90 0 0>}]
 dyn_strap_B_02_l [{shapeType=Capsule, length=8, radius=6.5, center=<3 7 0>, rotation=<rot 90 0 0>}]
 dyn_strap_B_02_r [{shapeType=Capsule, length=5.75, radius=5, center=<3 4.5 0>, rotation=<rot 90 0 0>}]
 dyn_strap_B_03_l [{shapeType=Capsule, length=8, radius=6.5, center=<3 7 0>, rotation=<rot 90 0 0>}]
 dyn_strap_B_03_r [{shapeType=Capsule, length=9.25, radius=5, center=<3 4.5 0>, rotation=<rot 90 0 0>}]
 dyn_strap_B_04_l [{shapeType=Capsule, length=8, radius=6.5, center=<3 7 0>, rotation=<rot 90 0 0>}]
 dyn_strap_B_04_r [{shapeType=Capsule, length=9.25, radius=5, center=<3 4.5 0>, rotation=<rot 90 0 0>}]
 dyn_strap_B_05_l [{shapeType=Capsule, length=8, radius=6.5, center=<3 7 0>, rotation=<rot 90 0 0>}]
 dyn_strap_B_05_r [{shapeType=Capsule, length=9.25, radius=5, center=<3 4.5 0>, rotation=<rot 90 0 0>}]
 pelvis [{shapeType=Capsule, length=0, radius=15.1851, center=<-4.5229 -7.4088 -0.0001>}]
 thigh_l [
  {shapeType=Capsule, length=24.75, radius=9.85, center=<-18.0428 0.0274 1.1898>, rotation=<rot 88.7521 170.8146 -9.1542>}
  {shapeType=Capsule, length=10, radius=9.85, center=<-10.3344 -1.633 1.412>, rotation=<rot 88.7499 170.8127 4.8243>}
 ]
 thigh_r [
  {shapeType=Capsule, length=20, radius=8.6, center=<15.6228 -1.145 -0.7379>, rotation=<rot 88.6013 174.5988 -9.7961>}
  {shapeType=Capsule, length=10.75, radius=8.85, center=<11.0242 -0.7745 -0.421>, rotation=<rot 88.071 -17.0729 164.6373>}
 ]
}
constraints [
 {
  bone1Name=thigh_r
  bone2Name=pelvis
  swing1LimitDegrees=45
  swing1Motion=Free
  swing2LimitDegrees=45
  swing2Motion=Free
  twistLimitDegrees=45
  twistMotion=Free
 }
 {bone1Name=thigh_l, bone2Name=pelvis, swing1LimitDegrees=45, swing2LimitDegrees=45, twistLimitDegrees=45}
 {
  bone1Name=dyn_driver_strap_A_01_l
  bone2Name=thigh_l
  swing1LimitDegrees=45
  swing1Motion=Free
  swing2LimitDegrees=45
  swing2Motion=Free
  twistLimitDegrees=45
  twistMotion=Free
 }
 {
  bone1Name=dyn_driver_strap_B_01_l
  bone2Name=thigh_l
  swing1LimitDegrees=45
  swing1Motion=Free
  swing2LimitDegrees=45
  swing2Motion=Free
  twistLimitDegrees=45
  twistMotion=Free
 }
 {bone1Name=dyn_strap_B_01_l, bone2Name=dyn_driver_strap_B_01_l, swing1LimitDegrees=80, swing2LimitDegrees=80, twistLimitDegrees=45}
 {bone1Name=dyn_strap_B_02_l, bone2Name=dyn_strap_B_01_l, swing1LimitDegrees=80, swing2LimitDegrees=80, twistLimitDegrees=45}
 {bone1Name=dyn_strap_B_03_l, bone2Name=dyn_strap_B_02_l, swing1LimitDegrees=80, swing2LimitDegrees=80, twistLimitDegrees=45}
 {bone1Name=dyn_strap_B_04_l, bone2Name=dyn_strap_B_03_l, swing1LimitDegrees=80, swing2LimitDegrees=80, twistLimitDegrees=45}
 {bone1Name=dyn_strap_B_05_l, bone2Name=dyn_strap_B_04_l, swing1LimitDegrees=80, swing2LimitDegrees=80, twistLimitDegrees=45}
 {bone1Name=dyn_strap_A_01_l, bone2Name=dyn_driver_strap_A_01_l, swing1LimitDegrees=85, swing2LimitDegrees=80, twistLimitDegrees=45}
 {bone1Name=dyn_strap_A_02_l, bone2Name=dyn_strap_A_01_l, swing1LimitDegrees=85, swing2LimitDegrees=80, twistLimitDegrees=45}
 {bone1Name=dyn_strap_A_03_l, bone2Name=dyn_strap_A_02_l, swing1LimitDegrees=85, swing2LimitDegrees=80, twistLimitDegrees=45}
 {bone1Name=dyn_strap_A_04_l, bone2Name=dyn_strap_A_03_l, swing1LimitDegrees=85, swing2LimitDegrees=80, twistLimitDegrees=45}
 {bone1Name=dyn_strap_A_05_l, bone2Name=dyn_strap_A_04_l, swing1LimitDegrees=85, swing2LimitDegrees=80, twistLimitDegrees=45}
 {
  bone1Name=dyn_driver_strap_A_01_r
  bone2Name=pelvis
  swing1LimitDegrees=45
  swing1Motion=Free
  swing2LimitDegrees=45
  swing2Motion=Free
  twistLimitDegrees=45
  twistMotion=Free
 }
 {
  bone1Name=dyn_driver_strap_B_01_r
  bone2Name=pelvis
  swing1LimitDegrees=45
  swing1Motion=Free
  swing2LimitDegrees=45
  swing2Motion=Free
  twistLimitDegrees=45
  twistMotion=Free
 }
 {bone1Name=dyn_strap_A_01_r, bone2Name=dyn_driver_strap_A_01_r, swing1LimitDegrees=80, swing2LimitDegrees=80, twistLimitDegrees=15}
 {bone1Name=dyn_strap_A_02_r, bone2Name=dyn_strap_A_01_r, swing1LimitDegrees=80, swing2LimitDegrees=80, twistLimitDegrees=15}
 {bone1Name=dyn_strap_A_03_r, bone2Name=dyn_strap_A_02_r, swing1LimitDegrees=80, swing2LimitDegrees=80, twistLimitDegrees=15}
 {bone1Name=dyn_strap_A_04_r, bone2Name=dyn_strap_A_03_r, swing1LimitDegrees=80, swing2LimitDegrees=80, twistLimitDegrees=15}
 {bone1Name=dyn_strap_A_05_r, bone2Name=dyn_strap_A_04_r, swing1LimitDegrees=80, swing2LimitDegrees=80, twistLimitDegrees=15}
 {bone1Name=dyn_strap_B_01_r, bone2Name=dyn_driver_strap_B_01_r, swing1LimitDegrees=80, swing2LimitDegrees=80, twistLimitDegrees=15}
 {bone1Name=dyn_strap_B_02_r, bone2Name=dyn_strap_B_01_r, swing1LimitDegrees=80, swing2LimitDegrees=80, twistLimitDegrees=15}
 {bone1Name=dyn_strap_B_03_r, bone2Name=dyn_strap_B_02_r, swing1LimitDegrees=80, swing2LimitDegrees=80, twistLimitDegrees=15}
 {bone1Name=dyn_strap_B_04_r, bone2Name=dyn_strap_B_03_r, swing1LimitDegrees=80, swing2LimitDegrees=80, twistLimitDegrees=15}
 {bone1Name=dyn_strap_B_05_r, bone2Name=dyn_strap_B_04_r, swing1LimitDegrees=80, swing2LimitDegrees=80, twistLimitDegrees=15}
]
