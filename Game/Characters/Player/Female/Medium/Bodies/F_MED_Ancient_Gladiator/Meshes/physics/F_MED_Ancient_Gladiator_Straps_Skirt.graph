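# Physics asset for the strap skirt on the medium female Ancient Gladiator player body.
# Three front strap chains of capsules hang from a central flap body, linked to the pelvis through a driven bone, with limited swing and twist.
# Cross-constraints tie the side chains' first two segments to the center chain.
# Box side flaps hang from driver bodies, only the left constrained to its thigh, alongside pelvis, thigh and strap-aim capsules.
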PhysicsAsset /Game/Characters/Player/Female/Medium/Bodies/F_MED_Ancient_Gladiator/Meshes/physics/F_MED_Ancient_Gladiator_Straps_Skirt
bodies {
 Flap_Front_Driven_Bone [{shapeType=Sphere, radius=0.505}]
 driver_frill_frt_l [{shapeType=Sphere, radius=1}]
 driver_frill_frt_r [{shapeType=Sphere, radius=0.505}]
 dyn_driver_frill_frtA_01_l [{shapeType=Capsule, length=9.25, radius=5.95, center=<3 7 0>, rotation=<rot 90 0 0>}]
 dyn_driver_frill_frtA_01_r [{shapeType=Capsule, length=9, radius=6.5, center=<3 7 0>, rotation=<rot 90 0 0>}]
 dyn_driver_frill_frt_01 [{shapeType=Capsule, length=7.25, radius=6.5, center=<1.2251 7.0002 0>, rotation=<rot 90 0 0>}]
 dyn_frill_frt [{shapeType=Capsule, length=25.75, radius=22.5, center=<0 -23.1316 -10.0828>, rotation=<rot 0 0 2.4212>}]
 dyn_frill_frtA_01_l [{shapeType=Capsule, length=9.5, radius=5.95, center=<3 6.5 0>, rotation=<rot 90 0 0>}]
 dyn_frill_frtA_01_r [{shapeType=Capsule, length=9, radius=6.5, center=<3 7 0>, rotation=<rot 90 0 0>}]
 dyn_frill_frtA_02_l [{shapeType=Capsule, length=9.5, radius=7, center=<3 6.5 0>, rotation=<rot 90 0 0>}]
 dyn_frill_frtA_02_r [{shapeType=Capsule, length=9, radius=6.5, center=<3 7 0>, rotation=<rot 90 0 0>}]
 dyn_frill_frtA_03_l [{shapeType=Capsule, length=9.25, radius=6.5, center=<3 6.5 0>, rotation=<rot 90 0 0>}]
 dyn_frill_frtA_03_r [{shapeType=Capsule, length=9, radius=6.5, center=<3 7 0>, rotation=<rot 90 0 0>}]
 dyn_frill_frtA_04_l [{shapeType=Capsule, length=9.25, radius=6.5, center=<3 6.5 0>, rotation=<rot 90 0 0>}]
 dyn_frill_frtA_04_r [{shapeType=Capsule, length=9, radius=6.5, center=<3 7 0>, rotation=<rot 90 0 0>}]
 dyn_frill_frtA_05_l [{shapeType=Capsule, length=9.25, radius=6.5, center=<3 6.5 0>, rotation=<rot 90 0 0>}]
 dyn_frill_frtA_05_r [{shapeType=Capsule, length=9, radius=6.5, center=<3 7 0>, rotation=<rot 90 0 0>}]
 dyn_frill_frt_01 [{shapeType=Capsule, length=7.25, radius=6.5, center=<3 7 0>, rotation=<rot 90 0 0>}]
 dyn_frill_frt_02 [{shapeType=Capsule, length=7.25, radius=6.5, center=<3 7 0>, rotation=<rot 90 0 0>}]
 dyn_frill_frt_03 [{shapeType=Capsule, length=7.25, radius=6.5, center=<3 7 0>, rotation=<rot 90 0 0>}]
 dyn_frill_frt_04 [{shapeType=Capsule, length=7.25, radius=6.5, center=<3 7 0>, rotation=<rot 90 0 0>}]
 dyn_frill_frt_05 [{shapeType=Capsule, length=7.25, radius=6.5, center=<3 7 0>, rotation=<rot 90 0 0>}]
 dyn_frill_frt_l [{shapeType=Box, extentX=25, extentY=14, extentZ=2.5, center=<-11.655 -1.8783 2.5283>, rotation=<rot -0.6258 1.4133 -84.4248>}]
 dyn_frill_frt_r [{shapeType=Box, extentX=25, extentY=14, extentZ=1.5, center=<12.3723 0.2811 -2.8738>, rotation=<rot -3.7035 -0.0796 90.1882>}]
 dyn_parent_frill_frtA_01_l [{shapeType=Capsule, length=0.505, radius=0.505, rotation=<rot 0 -90 0>}]
 dyn_parent_frill_frtA_01_r [{shapeType=Capsule, length=0.505, radius=0.505, rotation=<rot 0 -90 0>}]
 dyn_parent_frill_frt_01 [{shapeType=Capsule, length=0.505, radius=0.505, rotation=<rot 0 -90 0>}]
 pelvis [{shapeType=Capsule, length=8.5603, radius=17.6851, center=<-4.5229 -7.4088 -0.0001>}]
 strap_aim_01_c [{shapeType=Capsule, length=8, radius=7.5, center=<0 -6.5 0>, rotation=<rot 90 0 0>}]
 strap_aim_01_l [{shapeType=Capsule, length=8, radius=7.5, center=<0 -6.5 0>, rotation=<rot 90 0 0>}]
 strap_aim_01_r [{shapeType=Capsule, length=8, radius=7.5, center=<0 -6.5 0>, rotation=<rot 90 0 0>}]
 strap_aim_02_c [{shapeType=Capsule, length=8, radius=7.5, center=<0 -6.5 0>, rotation=<rot 90 0 0>}]
 strap_aim_02_l [{shapeType=Capsule, length=8, radius=7.5, center=<0 -6.5 0>, rotation=<rot 90 0 0>}]
 strap_aim_02_r [{shapeType=Capsule, length=8, radius=7.5, center=<0 -6.5 0>, rotation=<rot 90 0 0>}]
 thigh_l [{shapeType=Capsule, length=23, radius=9.25, center=<-17.7245 0.7443 1.4808>, rotation=<rot 89.852 -89.8388 89.4568>}]
 thigh_r [{shapeType=Capsule, length=21.75, radius=9.5, center=<18.8601 0.6166 -1.2773>, rotation=<rot 90 0 -2.3948>}]
}
constraints [
 {bone1Name=dyn_frill_frt_l, bone2Name=driver_frill_frt_l, swing1LimitDegrees=15, swing2LimitDegrees=15, twistLimitDegrees=15}
 {bone1Name=dyn_frill_frt, bone2Name=Flap_Front_Driven_Bone, swing1LimitDegrees=45, swing2LimitDegrees=45, twistLimitDegrees=45}
 {
  bone1Name=thigh_r
  bone2Name=pelvis
  swing1LimitDegrees=45
  swing1Motion=Free
  swing2LimitDegrees=45
  swing2Motion=Free
  twistLimitDegrees=45
  twistMotion=Free
 }
 {
  bone1Name=Flap_Front_Driven_Bone
  bone2Name=pelvis
  swing1LimitDegrees=45
  swing1Motion=Free
  swing2LimitDegrees=45
  swing2Motion=Free
  twistLimitDegrees=45
  twistMotion=Free
 }
 {
  bone1Name=driver_frill_frt_l
  bone2Name=thigh_l
  swing1LimitDegrees=45
  swing1Motion=Free
  swing2LimitDegrees=45
  swing2Motion=Free
  twistLimitDegrees=45
  twistMotion=Free
 }
 {bone1Name=thigh_l, bone2Name=pelvis, swing1LimitDegrees=45, swing2LimitDegrees=45, twistLimitDegrees=45}
 {bone1Name=dyn_frill_frt_r, bone2Name=driver_frill_frt_r, swing1LimitDegrees=20, swing2LimitDegrees=15, twistLimitDegrees=10}
 {bone1Name=dyn_driver_frill_frt_01, bone2Name=dyn_parent_frill_frt_01, swing1LimitDegrees=45, swing2LimitDegrees=40, twistLimitDegrees=30}
 {bone1Name=dyn_frill_frt_01, bone2Name=dyn_driver_frill_frt_01, swing1LimitDegrees=45, swing2LimitDegrees=40, twistLimitDegrees=30}
 {bone1Name=dyn_frill_frt_02, bone2Name=dyn_frill_frt_01, swing1LimitDegrees=30, swing2LimitDegrees=5, twistLimitDegrees=30}
 {bone1Name=dyn_frill_frt_03, bone2Name=dyn_frill_frt_02, swing1LimitDegrees=20, swing2LimitDegrees=15, twistLimitDegrees=30}
 {bone1Name=dyn_frill_frt_04, bone2Name=dyn_frill_frt_03, swing1LimitDegrees=20, swing2LimitDegrees=15, twistLimitDegrees=30}
 {bone1Name=dyn_frill_frt_05, bone2Name=dyn_frill_frt_04, swing1LimitDegrees=20, swing2LimitDegrees=15, twistLimitDegrees=30}
 {
  bone1Name=dyn_parent_frill_frt_01
  bone2Name=dyn_frill_frt
  swing1LimitDegrees=45
  swing1Motion=Free
  swing2LimitDegrees=45
  swing2Motion=Free
  twistLimitDegrees=45
  twistMotion=Free
 }
 {
  bone1Name=dyn_parent_frill_frtA_01_l
  bone2Name=dyn_frill_frt
  swing1LimitDegrees=45
  swing1Motion=Free
  swing2LimitDegrees=45
  swing2Motion=Free
  twistLimitDegrees=45
  twistMotion=Free
 }
 {
  bone1Name=dyn_driver_frill_frtA_01_l
  bone2Name=dyn_parent_frill_frtA_01_l
  swing1LimitDegrees=45
  swing2LimitDegrees=40
  twistLimitDegrees=30
 }
 {bone1Name=dyn_frill_frtA_01_l, bone2Name=dyn_driver_frill_frtA_01_l, swing1LimitDegrees=45, swing2LimitDegrees=40, twistLimitDegrees=30}
 {bone1Name=dyn_frill_frtA_02_l, bone2Name=dyn_frill_frtA_01_l, swing1LimitDegrees=30, swing2LimitDegrees=5, twistLimitDegrees=30}
 {bone1Name=dyn_frill_frtA_03_l, bone2Name=dyn_frill_frtA_02_l, swing1LimitDegrees=20, swing2LimitDegrees=15, twistLimitDegrees=30}
 {bone1Name=dyn_frill_frtA_04_l, bone2Name=dyn_frill_frtA_03_l, swing1LimitDegrees=20, swing2LimitDegrees=15, twistLimitDegrees=30}
 {bone1Name=dyn_frill_frtA_05_l, bone2Name=dyn_frill_frtA_04_l, swing1LimitDegrees=20, swing2LimitDegrees=15, twistLimitDegrees=30}
 {
  bone1Name=dyn_parent_frill_frtA_01_r
  bone2Name=dyn_frill_frt
  swing1LimitDegrees=45
  swing1Motion=Free
  swing2LimitDegrees=45
  swing2Motion=Free
  twistLimitDegrees=45
  twistMotion=Free
 }
 {
  bone1Name=dyn_driver_frill_frtA_01_r
  bone2Name=dyn_parent_frill_frtA_01_r
  swing1LimitDegrees=45
  swing2LimitDegrees=40
  twistLimitDegrees=30
 }
 {bone1Name=dyn_frill_frtA_01_r, bone2Name=dyn_driver_frill_frtA_01_r, swing1LimitDegrees=45, swing2LimitDegrees=40, twistLimitDegrees=30}
 {bone1Name=dyn_frill_frtA_02_r, bone2Name=dyn_frill_frtA_01_r, swing1LimitDegrees=30, swing2LimitDegrees=5, twistLimitDegrees=30}
 {bone1Name=dyn_frill_frtA_03_r, bone2Name=dyn_frill_frtA_02_r, swing1LimitDegrees=20, swing2LimitDegrees=15, twistLimitDegrees=30}
 {bone1Name=dyn_frill_frtA_04_r, bone2Name=dyn_frill_frtA_03_r, swing1LimitDegrees=20, swing2LimitDegrees=15, twistLimitDegrees=30}
 {bone1Name=dyn_frill_frtA_05_r, bone2Name=dyn_frill_frtA_04_r, swing1LimitDegrees=20, swing2LimitDegrees=15, twistLimitDegrees=30}
 {bone1Name=dyn_frill_frtA_02_r, bone2Name=dyn_frill_frt_02, swing1LimitDegrees=45, swing2LimitDegrees=45, twistLimitDegrees=45}
 {bone1Name=dyn_frill_frtA_02_l, bone2Name=dyn_frill_frt_02, swing1LimitDegrees=45, swing2LimitDegrees=45, twistLimitDegrees=45}
 {bone1Name=dyn_frill_frtA_01_r, bone2Name=dyn_frill_frt_01, swing1LimitDegrees=45, swing2LimitDegrees=45, twistLimitDegrees=45}
 {bone1Name=dyn_frill_frtA_01_l, bone2Name=dyn_frill_frt_01, swing1LimitDegrees=45, swing2LimitDegrees=45, twistLimitDegrees=45}
 {
  bone1Name=strap_aim_01_l
  bone2Name=pelvis
  swing1LimitDegrees=45
  swing1Motion=Free
  swing2LimitDegrees=45
  swing2Motion=Free
  twistLimitDegrees=45
  twistMotion=Free
 }
 {
  bone1Name=strap_aim_02_l
  bone2Name=pelvis
  swing1LimitDegrees=45
  swing1Motion=Free
  swing2LimitDegrees=45
  swing2Motion=Free
  twistLimitDegrees=45
  twistMotion=Free
 }
 {
  bone1Name=strap_aim_01_r
  bone2Name=pelvis
  swing1LimitDegrees=45
  swing1Motion=Free
  swing2LimitDegrees=45
  swing2Motion=Free
  twistLimitDegrees=45
  twistMotion=Free
 }
 {
  bone1Name=strap_aim_02_r
  bone2Name=pelvis
  swing1LimitDegrees=45
  swing1Motion=Free
  swing2LimitDegrees=45
  swing2Motion=Free
  twistLimitDegrees=45
  twistMotion=Free
 }
 {
  bone1Name=strap_aim_01_c
  bone2Name=pelvis
  swing1LimitDegrees=45
  swing1Motion=Free
  swing2LimitDegrees=45
  swing2Motion=Free
  twistLimitDegrees=45
  twistMotion=Free
 }
 {
  bone1Name=strap_aim_02_c
  bone2Name=pelvis
  swing1LimitDegrees=45
  swing1Motion=Free
  swing2LimitDegrees=45
  swing2Motion=Free
  twistLimitDegrees=45
  twistMotion=Free
 }
]
