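# Animation Blueprint that drives the skirt frills and straps of the female medium Ancient Gladiator player body.
# It copies the parent mesh pose through a Control Rig, then applies weighted bone constraints and alpha-driven corrective offsets.
# Straps swing via Anim Dynamics, some with thigh-attached spherical limits, while the crouch branch adds RigidBody simulation with crouch-specific physics assets.
# The output selects crouching, front-end or adjusted standing poses using bIsCrouching and bIsInFrontEnd.
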AnimBlueprint /Game/Characters/Player/Female/Medium/Bodies/F_MED_Ancient_Gladiator/Meshes/F_MED_Ancient_Gladiator_Skirt_AnimBP
target_skeleton=&/F_MED_Ancient_Gladiator_Skeleton
graphs [
 AnimGraph {nodes [
   Constraint!(ComponentPose, Alpha=1) -> (Pose) {alphaBoolBlend {blendOption=Linear}, boneToModify=~}
   "Constraint - dyn_driver_frill_frt_01 (PPPPPP)"!(
    ComponentPose
    Alpha=1
    ConstraintWeights_0=0.8358
    ConstraintWeights_1:float=0.0083
    ConstraintWeights_2:float=0.0339
    ConstraintWeights_3:float=0.0928
    ConstraintWeights_4:float=0.0089
    ConstraintWeights_5:float=0.0203
   ) -> (Pose) {
    alphaBoolBlend {blendOption=Linear}
    boneToModify {boneName=dyn_driver_frill_frt_01}
    constraintSetup [
     {targetBone {boneName=dyn_pelvis_offset}}
     {targetBone {boneName=dyn_frill_frt_l}}
     {targetBone {boneName=thigh_twist_01_r}}
     {targetBone {boneName=thigh_twist_01_l}}
     {targetBone {boneName=dyn_frill_frt_r}}
     {targetBone {boneName=dyn_frill_frt}}
    ]
    constraintWeights [1,1,1,1,1,1]
   }
   n1 = Get rban_straps_look_alpha() -> (rban_straps_look_alpha:double)
   n2 = "Constraint - dyn_frill_frt_01 (PPPPPP)"!(
    ComponentPose
    Alpha=@n1
    ConstraintWeights_0=0.2042
    ConstraintWeights_1:float=0.3549
    ConstraintWeights_2:float=0.0149
    ConstraintWeights_3:float=0.0138
    ConstraintWeights_4:float=0.3188
    ConstraintWeights_5:float=0.0933
   ) -> (Pose) {
    alphaBoolBlend {blendOption=Linear}
    boneToModify {boneName=dyn_frill_frt_01}
    constraintSetup [
     {targetBone {boneName=dyn_pelvis_offset}}
     {targetBone {boneName=dyn_frill_frt_l}}
     {targetBone {boneName=thigh_twist_01_r}}
     {targetBone {boneName=thigh_twist_01_l}}
     {targetBone {boneName=dyn_frill_frt_r}}
     {targetBone {boneName=dyn_frill_frt}}
    ]
    constraintWeights [1,1,1,1,1,1]
   }
   n3 = "Constraint - dyn_frill_frt_02 (PPPPPP)"!(
    ComponentPose=@n2
    Alpha=@n1
    ConstraintWeights_0=0.0236
    ConstraintWeights_1:float=0.4105
    ConstraintWeights_2:float=0.0084
    ConstraintWeights_3:float=0.0072
    ConstraintWeights_4:float=0.3666
    ConstraintWeights_5:float=0.1837
   ) -> (Pose) {
    alphaBoolBlend {blendOption=Linear}
    boneToModify {boneName=dyn_frill_frt_02}
    constraintSetup [
     {targetBone {boneName=dyn_pelvis_offset}}
     {targetBone {boneName=dyn_frill_frt_l}}
     {targetBone {boneName=thigh_twist_01_r}}
     {targetBone {boneName=thigh_twist_01_l}}
     {targetBone {boneName=dyn_frill_frt_r}}
     {targetBone {boneName=dyn_frill_frt}}
    ]
    constraintWeights [1,1,1,1,1,1]
   }
   n4 = "Constraint - dyn_frill_frt_03 (PPPPPP)"!(
    ComponentPose=@n3
    Alpha=@n1
    ConstraintWeights_0=0.0028
    ConstraintWeights_1:float=0.3777
    ConstraintWeights_2:float=0.0127
    ConstraintWeights_3:float=0.0097
    ConstraintWeights_4:float=0.3214
    ConstraintWeights_5:float=0.2757
   ) -> (Pose) {
    alphaBoolBlend {blendOption=Linear}
    boneToModify {boneName=dyn_frill_frt_03}
    constraintSetup [
     {targetBone {boneName=dyn_pelvis_offset}}
     {targetBone {boneName=dyn_frill_frt_l}}
     {targetBone {boneName=thigh_twist_01_r}}
     {targetBone {boneName=thigh_twist_01_l}}
     {targetBone {boneName=dyn_frill_frt_r}}
     {targetBone {boneName=dyn_frill_frt}}
    ]
    constraintWeights [1,1,1,1,1,1]
   }
   "Constraint - dyn_frill_frt_04 (PPPPPP)"!(
    ComponentPose=@n4
    Alpha=@n1
    ConstraintWeights_0=0
    ConstraintWeights_1:float=0.2706
    ConstraintWeights_2:float=0.0345
    ConstraintWeights_3:float=0.0269
    ConstraintWeights_4:float=0.2465
    ConstraintWeights_5:float=0.4215
   ) -> (Pose) {
    alphaBoolBlend {blendOption=Linear}
    boneToModify {boneName=dyn_frill_frt_04}
    constraintSetup [
     {targetBone {boneName=dyn_pelvis_offset}}
     {targetBone {boneName=dyn_frill_frt_l}}
     {targetBone {boneName=thigh_twist_01_r}}
     {targetBone {boneName=thigh_twist_01_l}}
     {targetBone {boneName=dyn_frill_frt_r}}
     {targetBone {boneName=dyn_frill_frt}}
    ]
    constraintWeights [1,1,1,1,1,1]
   }
   n5 = "Constraint - dyn_driver_frill_frtA_01_r (P)"(
    ComponentPose=Reroute Node(InputPin:Component Space Pose Link Structure) -> (OutputPin:Component Space Pose Link Structure)
    Alpha=1
    ConstraintWeights_0=1
   ) {
    alphaBoolBlend {blendOption=Linear}
    boneToModify {boneName=dyn_driver_frill_frtA_01_r}
    constraintSetup [{targetBone {boneName=dyn_frill_frt_r}}]
    constraintWeights [1]
   }
   n6 = "Constraint - dyn_driver_frill_frtA_01_l (P)"(ComponentPose=@n5, Alpha=1, ConstraintWeights_0=1) {
    alphaBoolBlend {blendOption=Linear}
    boneToModify {boneName=dyn_driver_frill_frtA_01_l}
    constraintSetup [{targetBone {boneName=dyn_frill_frt_l}}]
    constraintWeights [1]
   }
   "Constraint - dyn_driver_frill_frt_01 (PP)"!(ComponentPose=@n6, Alpha=1, ConstraintWeights_0=0.5, ConstraintWeights_1:float=0.5) -> (Pose) {
    alphaBoolBlend {blendOption=Linear}
    boneToModify {boneName=dyn_driver_frill_frt_01}
    constraintSetup [{targetBone {boneName=dyn_frill_frt_r}}, {targetBone {boneName=dyn_frill_frt_l}}]
    constraintWeights [1,1]
   }
   n7 = Look At - Bone: dyn_parent_frill_frtA_01_r(Alpha=1) {
    alphaBoolBlend {blendOption=Linear}
    boneToModify {boneName=dyn_parent_frill_frtA_01_r}
    lookAtLocation=<50 -2 -10>
    lookAtTarget {boneReference {boneName=strap_aim_04_r}}
    lookAt_Axis {axis=<1 0>}
   }
   n8 = Look At - Bone: dyn_parent_frill_frt_01(ComponentPose=@n7, Alpha=1) {
    alphaBoolBlend {blendOption=Linear}
    boneToModify {boneName=dyn_parent_frill_frt_01}
    lookAtLocation=<50 1>
    lookAtTarget {boneReference {boneName=strap_aim_04_c}}
    lookAt_Axis {axis=<1 0>}
   }
   Look At - Bone: dyn_parent_frill_frtA_01_l(ComponentPose=@n8, Alpha=1) {
    alphaBoolBlend {blendOption=Linear}
    boneToModify {boneName=dyn_parent_frill_frtA_01_l}
    lookAtLocation=<50 3 7>
    lookAtTarget {boneReference {boneName=strap_aim_04_l}}
    lookAt_Axis {axis=<1 0>}
   }
   n9 = "Transform (Modify) Bone - Bone: dyn_driver_strap_A_01_r"(Alpha=Get side_straps_crouch() -> (side_straps_crouch:double), Translation="0,1,0", Rotation="0,5,20", Scale="1,1,1") {
    rotationMode=BMM_Additive
    rotationSpace=BCS_BoneSpace
    scaleSpace=BCS_BoneSpace
    translationMode=BMM_Additive
    translationSpace=BCS_BoneSpace
    alphaBoolBlend {blendOption=Linear}
    boneToModify {boneName=dyn_driver_strap_A_01_r}
    rotation {roll=20, yaw=5}
    translation {y=1}
   }
   "Transform (Modify) Bone - Bone: dyn_driver_strap_B_01_r"(ComponentPose=@n9, Alpha=1, Translation="0,1,0", Rotation="0,-5,-10", Scale="1,1,1") {
    rotationMode=BMM_Additive
    rotationSpace=BCS_BoneSpace
    scaleSpace=BCS_BoneSpace
    translationMode=BMM_Additive
    translationSpace=BCS_BoneSpace
    alphaBoolBlend {blendOption=Linear}
    boneToModify {boneName=dyn_driver_strap_B_01_r}
    rotation {roll=-10, yaw=-5}
    translation {y=1}
   }
   n10 = Get rban_straps_alpha() -> (rban_straps_alpha:double)
   n11 = Get rban_straps_alpha()
   n12 = "Transform (Modify) Bone - Bone: dyn_driver_strap_01_r"(Alpha=@n11, Translation="0,0,0", Rotation="-3,0,0", Scale="1,1,1") {
    rotationMode=BMM_Additive
    rotationSpace=BCS_BoneSpace
    alphaBoolBlend {blendOption=Linear}
    boneToModify {boneName=dyn_driver_strap_01_r}
    rotation {pitch=-3}
   }
   n13 = "Transform (Modify) Bone - Bone: dyn_driver_strap_01_l"(ComponentPose=@n12, Alpha=@n11, Translation="0,0,0", Rotation="-1,0,0", Scale="1,1,1") {
    rotationMode=BMM_Additive
    rotationSpace=BCS_BoneSpace
    alphaBoolBlend {blendOption=Linear}
    boneToModify {boneName=dyn_driver_strap_01_l}
    rotation {pitch=-1}
   }
   n14 = "Transform (Modify) Bone - Bone: dyn_parent_strap_A_01_r"(ComponentPose=@n13, Alpha=@n10, Translation="0,0,0", Rotation="0,2,15", Scale="1,1,1") {
    rotationMode=BMM_Additive
    rotationSpace=BCS_BoneSpace
    alphaBoolBlend {blendOption=Linear}
    boneToModify {boneName=dyn_parent_strap_A_01_r}
    rotation {roll=15, yaw=2}
   }
   "Transform (Modify) Bone - Bone: dyn_parent_strap_A_01_r"(ComponentPose=@n14, Alpha=@n10, Translation="0,0,0", Rotation="-5,-6.5,0", Scale="1,1,1") {
    rotationMode=BMM_Additive
    rotationSpace=BCS_BoneSpace
    alphaBoolBlend {blendOption=Linear}
    boneToModify {boneName=dyn_parent_strap_A_01_r}
    rotation {pitch=-5, yaw=-6.5}
   }
   RigidBody!(ComponentPose, Alpha=Get side_straps_alpha() -> (side_straps_alpha:double), OverrideWorldGravity:Vector="0,0,-980", ExternalForce) -> (Pose) {
    bClampLinearTranslationLimitToRefPose=true
    bOverrideWorldGravity=true
    overridePhysicsAsset=/Game/Characters/Player/Female/Medium/Bodies/F_MED_Ancient_Gladiator/Meshes/physics/F_MED_Ancient_Gladiator_Straps_Side
    simulationSpace=BaseBoneSpace
    alphaBoolBlend {blendOption=Linear}
    baseBoneRef {boneName=root}
    componentLinearAccScale=<0.5 0.5 0.35>
    componentLinearVelScale=<0.5 0.5 0.35>
    overrideWorldGravity {z=-980}
   }
   n15 = Get fe_alpha() -> (fe_alpha:double)
   n16 = "Transform (Modify) Bone - Bone: dyn_driver_strap_A_01_l"(
    ComponentPose=Local To Component(LocalPose=Use cached pose 'Standing_Cache'())
    Alpha=@n15
    Translation="0,0,0"
    Rotation="8,-8,-6"
    Scale="1,1,1"
   ) {
    rotationMode=BMM_Additive
    rotationSpace=BCS_BoneSpace
    translationMode=BMM_Additive
    translationSpace=BCS_BoneSpace
    alphaBoolBlend {blendOption=Linear}
    boneToModify {boneName=dyn_driver_strap_A_01_l}
    rotation=<rot 8 -8 -6>
   }
   n17 = "Transform (Modify) Bone - Bone: dyn_strap_B_01_l"(ComponentPose=@n16, Alpha=@n15, Translation="0,0,0", Rotation="8,0,0", Scale="1,1,1") {
    rotationMode=BMM_Additive
    rotationSpace=BCS_BoneSpace
    translationMode=BMM_Additive
    translationSpace=BCS_BoneSpace
    alphaBoolBlend {blendOption=Linear}
    boneToModify {boneName=dyn_strap_B_01_l}
    rotation {pitch=8}
   }
   n18 = Get fe_alpha()
   n19 = "Transform (Modify) Bone - Bone: dyn_parent_frill_frtA_01_r"(ComponentPose=@n17, Alpha=@n18, Translation="0,1,0", Rotation="0,0,-5", Scale="1,1,1") {
    rotationMode=BMM_Additive
    rotationSpace=BCS_BoneSpace
    translationMode=BMM_Additive
    translationSpace=BCS_BoneSpace
    alphaBoolBlend {blendOption=Linear}
    boneToModify {boneName=dyn_parent_frill_frtA_01_r}
    rotation {roll=-5}
    translation {y=1}
   }
   n20 = "Transform (Modify) Bone - Bone: dyn_frill_frtA_02_r"(ComponentPose=@n19, Alpha=@n18, Translation="0,0,0", Rotation="0.000,2,0.000", Scale="1,1,1") {
    rotationMode=BMM_Additive
    rotationSpace=BCS_BoneSpace
    translationMode=BMM_Additive
    translationSpace=BCS_BoneSpace
    alphaBoolBlend {blendOption=Linear}
    boneToModify {boneName=dyn_frill_frtA_02_r}
    rotation {yaw=2}
   }
   n21 = "Transform (Modify) Bone - Bone: dyn_parent_frill_frt_01"(ComponentPose=@n20, Alpha=@n18, Translation="0,0.75,0", Rotation="0.000,-1,0.000", Scale="1,1,1") {
    rotationMode=BMM_Additive
    rotationSpace=BCS_BoneSpace
    translationMode=BMM_Additive
    translationSpace=BCS_BoneSpace
    alphaBoolBlend {blendOption=Linear}
    boneToModify {boneName=dyn_parent_frill_frt_01}
    rotation {yaw=-1}
    translation {y=0.75}
   }
   n22 = "Transform (Modify) Bone - Bone: dyn_frill_frt_01"(ComponentPose=@n21, Alpha=@n18, Translation="0,0,0", Rotation="0.000,2,5", Scale="1,1,1") {
    rotationMode=BMM_Additive
    rotationSpace=BCS_BoneSpace
    translationMode=BMM_Additive
    translationSpace=BCS_BoneSpace
    alphaBoolBlend {blendOption=Linear}
    boneToModify {boneName=dyn_frill_frt_01}
    rotation {roll=5, yaw=2}
   }
   n23 = "Transform (Modify) Bone - Bone: dyn_parent_frill_frtA_01_l"(ComponentPose=@n22, Alpha=@n18, Translation="0,0.25,0", Rotation="0.000,1.5,7", Scale="1,1,1") {
    rotationMode=BMM_Additive
    rotationSpace=BCS_BoneSpace
    translationMode=BMM_Additive
    translationSpace=BCS_BoneSpace
    alphaBoolBlend {blendOption=Linear}
    boneToModify {boneName=dyn_parent_frill_frtA_01_l}
    rotation {roll=7, yaw=1.5}
    translation {y=0.25}
   }
   n24 = "Transform (Modify) Bone - Bone: dyn_driver_strap_A_01_r"(ComponentPose=@n23, Alpha=1, Translation="0,1,0", Rotation="0,0,0", Scale="1,1,1") {
    rotationMode=BMM_Additive
    rotationSpace=BCS_BoneSpace
    translationMode=BMM_Additive
    translationSpace=BCS_BoneSpace
    alphaBoolBlend {blendOption=Linear}
    boneToModify {boneName=dyn_driver_strap_A_01_r}
    translation {y=1}
   }
   n25 = "Transform (Modify) Bone - Bone: dyn_driver_strap_A_01_l"(ComponentPose=@n24, Alpha=1, Translation="0,2,0", Rotation="-15,5,0", Scale="1,1,1") {
    rotationMode=BMM_Additive
    rotationSpace=BCS_BoneSpace
    translationMode=BMM_Additive
    translationSpace=BCS_BoneSpace
    alphaBoolBlend {blendOption=Linear}
    boneToModify {boneName=dyn_driver_strap_A_01_l}
    rotation {pitch=-15, yaw=5}
    translation {y=2}
   }
   n26 = "Transform (Modify) Bone - Bone: dyn_driver_strap_A_01_r"(ComponentPose=@n25, Alpha=1, Translation="0,1.2,0", Rotation="20,5,0", Scale="1,1,1") {
    rotationMode=BMM_Additive
    rotationSpace=BCS_BoneSpace
    translationMode=BMM_Additive
    translationSpace=BCS_BoneSpace
    alphaBoolBlend {blendOption=Linear}
    boneToModify {boneName=dyn_driver_strap_A_01_r}
    rotation {pitch=20, yaw=5}
    translation {y=1.2}
   }
   n27 = Component To Local(ComponentPose=@n26)
   n28 = Use cached pose 'Standing_Cache'()
   n29 = Get bIsInFrontEnd()
   n30 = "Transform (Modify) Bone - Bone: dyn_pelvis_offset"(
    ComponentPose=Local To Component(LocalPose=Use cached pose 'CrouchCache'())
    Alpha=1
    Translation="0,1.55,5"
    Rotation="0,0,3"
    Scale="1,1,1"
   ) {
    rotationMode=BMM_Additive
    rotationSpace=BCS_BoneSpace
    translationMode=BMM_Additive
    translationSpace=BCS_BoneSpace
    alphaBoolBlend {blendOption=Linear}
    boneToModify {boneName=dyn_pelvis_offset}
    rotation {roll=3}
    translation {y=1.55, z=5}
   }
   n31 = RigidBody!(ComponentPose=@n30, Alpha, ExternalForce) -> (Pose) {
    bClampLinearTranslationLimitToRefPose=true
    bOverrideWorldGravity=true
    overridePhysicsAsset=/Game/Characters/Player/Female/Medium/Bodies/F_MED_Ancient_Gladiator/Meshes/physics/F_MED_Ancient_Gladiator_Skirt_Crouch
    simulationSpace=BaseBoneSpace
    alphaBoolBlend {blendOption=Linear}
    baseBoneRef {boneName=root}
    componentLinearAccScale=<0.5 0.5 0.35>
    componentLinearVelScale=<0.5 0.5 0.35>
    overrideWorldGravity {z=-980}
   }
   n32 = "Transform (Modify) Bone - Bone: dyn_frill_frt"(ComponentPose=@n31, Alpha=1, Translation="-1,0,10", Rotation="0,0,0", Scale="0.25,0.25,0.25") {
    scaleMode=BMM_Additive
    scaleSpace=BCS_BoneSpace
    alphaBoolBlend {blendOption=Linear}
    boneToModify {boneName=dyn_frill_frt}
    scale=<0.25 0.25 0.25>
    translation {x=-1, z=10}
   }
   n33 = "Transform (Modify) Bone - Bone: dyn_frill_frt_r"(ComponentPose=@n32, Alpha=Get legUp_Frill_R() -> (legUp_Frill_R:double), Translation="0,1,0", Rotation="-10,8,-12", Scale="1,1,1") {
    rotationMode=BMM_Additive
    rotationSpace=BCS_BoneSpace
    translationMode=BMM_Additive
    translationSpace=BCS_BoneSpace
    alphaBoolBlend {blendOption=Linear}
    boneToModify {boneName=dyn_frill_frt_r}
    rotation=<rot -10 8 -12>
    translation {y=1}
   }
   n34 = "Transform (Modify) Bone - Bone: dyn_frill_frt_l"(ComponentPose=@n33, Alpha=1, Translation="0,-0.7,4", Rotation="5.25,0,-35", Scale="0.85,0.85,0.85") {
    rotationMode=BMM_Additive
    rotationSpace=BCS_BoneSpace
    scaleMode=BMM_Additive
    scaleSpace=BCS_BoneSpace
    translationMode=BMM_Additive
    translationSpace=BCS_BoneSpace
    alphaBoolBlend {blendOption=Linear}
    boneToModify {boneName=dyn_frill_frt_l}
    rotation {pitch=5.25, roll=-35}
    scale=<0.85 0.85 0.85>
    translation {y=-0.7, z=4}
   }
   n35 = "Constraint - strap_aim_01_l (PPPPPP)"!(
    ComponentPose=@n34
    Alpha=1
    ConstraintWeights_0=0.7021
    ConstraintWeights_1:float=0.2395
    ConstraintWeights_2:float=0.0003
    ConstraintWeights_3:float=0.0576
    ConstraintWeights_4:float=0.0001
    ConstraintWeights_5:float=0.0003
   ) -> (Pose) {
    alphaBoolBlend {blendOption=Linear}
    boneToModify {boneName=strap_aim_01_l}
    constraintSetup [
     {targetBone {boneName=dyn_pelvis_offset}}
     {targetBone {boneName=dyn_frill_frt_l}}
     {targetBone {boneName=thigh_twist_01_r}}
     {targetBone {boneName=thigh_twist_01_l}}
     {targetBone {boneName=dyn_frill_frt_r}}
     {targetBone {boneName=dyn_frill_frt}}
    ]
    constraintWeights [1,1,1,1,1,1]
   }
   n36 = "Constraint - strap_aim_02_l (PPPPPP)"!(
    ComponentPose=@n35
    Alpha=1
    ConstraintWeights_0=0.0321
    ConstraintWeights_1:float=0.8622
    ConstraintWeights_2:float=0.0005
    ConstraintWeights_3:float=0.0005
    ConstraintWeights_4:float=0.0172
    ConstraintWeights_5:float=0.0876
   ) -> (Pose) {
    alphaBoolBlend {blendOption=Linear}
    boneToModify {boneName=strap_aim_02_l}
    constraintSetup [
     {targetBone {boneName=dyn_pelvis_offset}}
     {targetBone {boneName=dyn_frill_frt_l}}
     {targetBone {boneName=thigh_twist_01_r}}
     {targetBone {boneName=thigh_twist_01_l}}
     {targetBone {boneName=dyn_frill_frt_r}}
     {targetBone {boneName=dyn_frill_frt}}
    ]
    constraintWeights [1,1,1,1,1,1]
   }
   n37 = "Constraint - strap_aim_03_l (PPPPPP)"!(
    ComponentPose=@n36
    Alpha=1
    ConstraintWeights_0=0.001
    ConstraintWeights_1:float=0.836
    ConstraintWeights_2:float=0.0004
    ConstraintWeights_3:float=0.0004
    ConstraintWeights_4:float=0.0116
    ConstraintWeights_5:float=0.1505
   ) -> (Pose) {
    alphaBoolBlend {blendOption=Linear}
    boneToModify {boneName=strap_aim_03_l}
    constraintSetup [
     {targetBone {boneName=dyn_pelvis_offset}}
     {targetBone {boneName=dyn_frill_frt_l}}
     {targetBone {boneName=thigh_twist_01_r}}
     {targetBone {boneName=thigh_twist_01_l}}
     {targetBone {boneName=dyn_frill_frt_r}}
     {targetBone {boneName=dyn_frill_frt}}
    ]
    constraintWeights [1,1,1,1,1,1]
   }
   n38 = "Constraint - strap_aim_04_l (PPPPPP)"!(
    ComponentPose=@n37
    Alpha=1
    ConstraintWeights_0=0
    ConstraintWeights_1:float=0.777
    ConstraintWeights_2:float=0.0016
    ConstraintWeights_3:float=0.0005
    ConstraintWeights_4:float=0.0124
    ConstraintWeights_5:float=0.2085
   ) -> (Pose) {
    alphaBoolBlend {blendOption=Linear}
    boneToModify {boneName=strap_aim_04_l}
    constraintSetup [
     {targetBone {boneName=dyn_pelvis_offset}}
     {targetBone {boneName=dyn_frill_frt_l}}
     {targetBone {boneName=thigh_twist_01_r}}
     {targetBone {boneName=thigh_twist_01_l}}
     {targetBone {boneName=dyn_frill_frt_r}}
     {targetBone {boneName=dyn_frill_frt}}
    ]
    constraintWeights [1,1,1,1,1,1]
   }
   n39 = "Constraint - strap_aim_01_r (PPPPPP)"!(
    ComponentPose=@n38
    Alpha=1
    ConstraintWeights_0=0.724
    ConstraintWeights_1:float=0.0006
    ConstraintWeights_2:float=0.108
    ConstraintWeights_3:float=0.0026
    ConstraintWeights_4:float=0.1631
    ConstraintWeights_5:float=0.0017
   ) -> (Pose) {
    alphaBoolBlend {blendOption=Linear}
    boneToModify {boneName=strap_aim_01_r}
    constraintSetup [
     {targetBone {boneName=dyn_pelvis_offset}}
     {targetBone {boneName=dyn_frill_frt_l}}
     {targetBone {boneName=thigh_twist_01_r}}
     {targetBone {boneName=thigh_twist_01_l}}
     {targetBone {boneName=dyn_frill_frt_r}}
     {targetBone {boneName=dyn_frill_frt}}
    ]
    constraintWeights [1,1,1,1,1,1]
   }
   n40 = "Constraint - strap_aim_02_r (PPPPPP)"!(
    ComponentPose=@n39
    Alpha=1
    ConstraintWeights_0=0.0547
    ConstraintWeights_1:float=0.0585
    ConstraintWeights_2:float=0.0074
    ConstraintWeights_3:float=0.0052
    ConstraintWeights_4:float=0.7797
    ConstraintWeights_5:float=0.0945
   ) -> (Pose) {
    alphaBoolBlend {blendOption=Linear}
    boneToModify {boneName=strap_aim_02_r}
    constraintSetup [
     {targetBone {boneName=dyn_pelvis_offset}}
     {targetBone {boneName=dyn_frill_frt_l}}
     {targetBone {boneName=thigh_twist_01_r}}
     {targetBone {boneName=thigh_twist_01_l}}
     {targetBone {boneName=dyn_frill_frt_r}}
     {targetBone {boneName=dyn_frill_frt}}
    ]
    constraintWeights [1,1,1,1,1,1]
   }
   n41 = "Constraint - strap_aim_03_r (PPPPPP)"!(
    ComponentPose=@n40
    Alpha=1
    ConstraintWeights_0=0.0048
    ConstraintWeights_1:float=0.0812
    ConstraintWeights_2:float=0.0041
    ConstraintWeights_3:float=0.0036
    ConstraintWeights_4:float=0.7503
    ConstraintWeights_5:float=0.156
   ) -> (Pose) {
    alphaBoolBlend {blendOption=Linear}
    boneToModify {boneName=strap_aim_03_r}
    constraintSetup [
     {targetBone {boneName=dyn_pelvis_offset}}
     {targetBone {boneName=dyn_frill_frt_l}}
     {targetBone {boneName=thigh_twist_01_r}}
     {targetBone {boneName=thigh_twist_01_l}}
     {targetBone {boneName=dyn_frill_frt_r}}
     {targetBone {boneName=dyn_frill_frt}}
    ]
    constraintWeights [1,1,1,1,1,1]
   }
   n42 = "Constraint - strap_aim_04_r (PPPPPP)"!(
    ComponentPose=@n41
    Alpha=1
    ConstraintWeights_0=0
    ConstraintWeights_1:float=0.049
    ConstraintWeights_2:float=0.0022
    ConstraintWeights_3:float=0.0061
    ConstraintWeights_4:float=0.6967
    ConstraintWeights_5:float=0.246
   ) -> (Pose) {
    alphaBoolBlend {blendOption=Linear}
    boneToModify {boneName=strap_aim_04_r}
    constraintSetup [
     {targetBone {boneName=dyn_pelvis_offset}}
     {targetBone {boneName=dyn_frill_frt_l}}
     {targetBone {boneName=thigh_twist_01_r}}
     {targetBone {boneName=thigh_twist_01_l}}
     {targetBone {boneName=dyn_frill_frt_r}}
     {targetBone {boneName=dyn_frill_frt}}
    ]
    constraintWeights [1,1,1,1,1,1]
   }
   n43 = "Constraint - strap_aim_01_c (PPPPPP)"!(
    ComponentPose=@n42
    Alpha=1
    ConstraintWeights_0=0.7923
    ConstraintWeights_1:float=0.0242
    ConstraintWeights_2:float=0.0276
    ConstraintWeights_3:float=0.0824
    ConstraintWeights_4:float=0.0247
    ConstraintWeights_5:float=0.0488
   ) -> (Pose) {
    alphaBoolBlend {blendOption=Linear}
    boneToModify {boneName=strap_aim_01_c}
    constraintSetup [
     {targetBone {boneName=dyn_pelvis_offset}}
     {targetBone {boneName=dyn_frill_frt_l}}
     {targetBone {boneName=thigh_twist_01_r}}
     {targetBone {boneName=thigh_twist_01_l}}
     {targetBone {boneName=dyn_frill_frt_r}}
     {targetBone {boneName=dyn_frill_frt}}
    ]
    constraintWeights [1,1,1,1,1,1]
   }
   n44 = "Constraint - strap_aim_02_c (PPPPPP)"!(
    ComponentPose=@n43
    Alpha=1
    ConstraintWeights_0=0.2042
    ConstraintWeights_1:float=0.3549
    ConstraintWeights_2:float=0.0149
    ConstraintWeights_3:float=0.0138
    ConstraintWeights_4:float=0.3188
    ConstraintWeights_5:float=0.0933
   ) -> (Pose) {
    alphaBoolBlend {blendOption=Linear}
    boneToModify {boneName=strap_aim_02_c}
    constraintSetup [
     {targetBone {boneName=dyn_pelvis_offset}}
     {targetBone {boneName=dyn_frill_frt_l}}
     {targetBone {boneName=thigh_twist_01_r}}
     {targetBone {boneName=thigh_twist_01_l}}
     {targetBone {boneName=dyn_frill_frt_r}}
     {targetBone {boneName=dyn_frill_frt}}
    ]
    constraintWeights [1,1,1,1,1,1]
   }
   n45 = "Constraint - strap_aim_03_c (PPPPPP)"!(
    ComponentPose=@n44
    Alpha=1
    ConstraintWeights_0=0.0236
    ConstraintWeights_1:float=0.4105
    ConstraintWeights_2:float=0.0084
    ConstraintWeights_3:float=0.0072
    ConstraintWeights_4:float=0.3666
    ConstraintWeights_5:float=0.1837
   ) -> (Pose) {
    alphaBoolBlend {blendOption=Linear}
    boneToModify {boneName=strap_aim_03_c}
    constraintSetup [
     {targetBone {boneName=dyn_pelvis_offset}}
     {targetBone {boneName=dyn_frill_frt_l}}
     {targetBone {boneName=thigh_twist_01_r}}
     {targetBone {boneName=thigh_twist_01_l}}
     {targetBone {boneName=dyn_frill_frt_r}}
     {targetBone {boneName=dyn_frill_frt}}
    ]
    constraintWeights [1,1,1,1,1,1]
   }
   n46 = "Constraint - strap_aim_04_c (PPPPPP)"!(
    ComponentPose=@n45
    Alpha=1
    ConstraintWeights_0=0.0028
    ConstraintWeights_1:float=0.3777
    ConstraintWeights_2:float=0.0127
    ConstraintWeights_3:float=0.0097
    ConstraintWeights_4:float=0.3214
    ConstraintWeights_5:float=0.2757
   ) -> (Pose) {
    alphaBoolBlend {blendOption=Linear}
    boneToModify {boneName=strap_aim_04_c}
    constraintSetup [
     {targetBone {boneName=dyn_pelvis_offset}}
     {targetBone {boneName=dyn_frill_frt_l}}
     {targetBone {boneName=thigh_twist_01_r}}
     {targetBone {boneName=thigh_twist_01_l}}
     {targetBone {boneName=dyn_frill_frt_r}}
     {targetBone {boneName=dyn_frill_frt}}
    ]
    constraintWeights [1,1,1,1,1,1]
   }
   n47 = "Transform (Modify) Bone - Bone: dyn_driver_strap_A_01_l"(ComponentPose=@n46, Alpha=1, Translation="-5,-8,0", Rotation="0,0,0", Scale="0.1,0.1,0.1") {
    rotationMode=BMM_Additive
    rotationSpace=BCS_BoneSpace
    scaleMode=BMM_Additive
    scaleSpace=BCS_BoneSpace
    translationMode=BMM_Additive
    translationSpace=BCS_BoneSpace
    alphaBoolBlend {blendOption=Linear}
    boneToModify {boneName=dyn_driver_strap_A_01_l}
    scale=<0.1 0.1 0.1>
    translation=<-5 -8>
   }
   n48 = "Transform (Modify) Bone - Bone: dyn_driver_strap_B_01_l"(ComponentPose=@n47, Alpha=1, Translation="0,0,-1", Rotation="0,0,0", Scale="1,1,1") {
    rotationMode=BMM_Additive
    rotationSpace=BCS_BoneSpace
    scaleSpace=BCS_BoneSpace
    translationMode=BMM_Additive
    translationSpace=BCS_BoneSpace
    alphaBoolBlend {blendOption=Linear}
    boneToModify {boneName=dyn_driver_strap_B_01_l}
    translation {z=-1}
   }
   n49 = "Transform (Modify) Bone - Bone: driver_frill_frt_l"(ComponentPose=@n48, Alpha=1, Translation="0,0,0", Rotation="10,-5,-5", Scale="1,1,1") {
    rotationMode=BMM_Additive
    rotationSpace=BCS_BoneSpace
    scaleSpace=BCS_BoneSpace
    translationMode=BMM_Additive
    translationSpace=BCS_BoneSpace
    alphaBoolBlend {blendOption=Linear}
    boneToModify {boneName=driver_frill_frt_l}
    rotation=<rot 10 -5 -5>
   }
   n50 = RigidBody(ComponentPose=@n49) {
    bClampLinearTranslationLimitToRefPose=true
    bOverrideWorldGravity=true
    overridePhysicsAsset=/Game/Characters/Player/Female/Medium/Bodies/F_MED_Ancient_Gladiator/Meshes/physics/F_MED_Ancient_Gladiator_Straps_Skirt_Crouch
    simulationSpace=BaseBoneSpace
    alphaBoolBlend {blendOption=Linear}
    baseBoneRef {boneName=root}
    componentLinearAccScale=<0.5 0.5 0.35>
    componentLinearVelScale=<0.5 0.5 0.35>
    overrideWorldGravity {z=-980}
   }
   Output Pose(Result=Blend Poses by bool(
     BlendPose_0=Component To Local(ComponentPose=@n50)
     BlendPose_1=Blend Poses by bool(BlendPose_0=@n27, BlendPose_1=@n28, bActiveValue=@n29) {blendType=Linear}
     bActiveValue=Get bIsCrouching()
    ) {blendType=Linear}) {layerGroup=None}
   n51 = Get legUp_Strap_R() -> (legUp_Strap_R:double)
   n52 = Get rban_straps_alpha()
   n53 = Control Rig(Source=Copy Pose From Mesh() {bUseAttachedParent=true}) {
    alphaBoolBlend {blendOption=Linear}
    controlRigAssetReference {blueprintRigClass=&/F_MED_Ancient_Gladiator_CtrlRig.F_MED_Ancient_Gladiator_CtrlRig_C}
   }
   n54 = "Constraint - dyn_hoop (PPPPPPP)"!(
    ComponentPose=Local To Component(LocalPose=@n53)
    Alpha=1
    ConstraintWeights_0=0.0211
    ConstraintWeights_1:float=0.0239
    ConstraintWeights_2:float=0.022
    ConstraintWeights_3:float=0.0366
    ConstraintWeights_4:float=0.2575
    ConstraintWeights_5:float=0.4604
    ConstraintWeights_6:float=0.1784
   ) -> (Pose) {
    alphaBoolBlend {blendOption=Linear}
    boneToModify {boneName=dyn_hoop}
    constraintSetup [
     {targetBone {boneName=clavicle_l}}
     {targetBone {boneName=upperarm_twist_01_r}}
     {targetBone {boneName=spine_02}}
     {targetBone {boneName=spine_03}}
     {targetBone {boneName=spine_04}}
     {targetBone {boneName=spine_05}}
     {targetBone {boneName=clavicle_r}}
    ]
    constraintWeights [1,1,1,1,1,1,1]
   }
   n55 = Get rban_straps_alpha()
   n56 = "Constraint - sub_thigh_l (P)"(ComponentPose=@n54, Alpha=@n55, ConstraintWeights_0=1) {
    alphaBoolBlend {blendOption=Linear}
    boneToModify {boneName=sub_thigh_l}
    constraintSetup [{targetBone {boneName=thigh_twist_01_l}}]
    constraintWeights [1]
   }
   n57 = "Constraint - sub_thigh_r (P)"(ComponentPose=@n56, Alpha=@n55, ConstraintWeights_0=1) {
    alphaBoolBlend {blendOption=Linear}
    boneToModify {boneName=sub_thigh_r}
    constraintSetup [{targetBone {boneName=thigh_twist_01_r}}]
    constraintWeights [1]
   }
   n58 = "Constraint - Flap_Front_Driven_Bone (PP)"!(ComponentPose=@n57, Alpha=1, ConstraintWeights_0=0.5, ConstraintWeights_1:float=0.5) -> (Pose) {
    alphaBoolBlend {blendOption=Linear}
    boneToModify {boneName=Flap_Front_Driven_Bone}
    constraintSetup [{targetBone {boneName=dyn_frill_frt_l}}, {targetBone {boneName=dyn_frill_frt_r}}]
    constraintWeights [1,1]
   }
   n59 = Get clav_up_r() -> (clav_up_r:double)
   n60 = "Transform (Modify) Bone - Bone: clavicle_offset_r"(ComponentPose=@n58, Alpha=@n59, Translation="0,-1,-2.5", Rotation="0,0,0", Scale="1,1,1") {
    rotationSpace=BCS_BoneSpace
    translationMode=BMM_Additive
    translationSpace=BCS_BoneSpace
    alphaBoolBlend {blendOption=Linear}
    boneToModify {boneName=clavicle_offset_r}
    translation {y=-1, z=-2.5}
   }
   n61 = "Transform (Modify) Bone - Bone: upperarm_offset_r"(ComponentPose=@n60, Alpha=@n59, Translation="-3,0,-1", Rotation="-20,-45,40", Scale="0.4,0.4,0.4") {
    rotationMode=BMM_Additive
    rotationSpace=BCS_BoneSpace
    scaleMode=BMM_Additive
    scaleSpace=BCS_BoneSpace
    translationMode=BMM_Additive
    translationSpace=BCS_BoneSpace
    alphaBoolBlend {blendOption=Linear}
    boneToModify {boneName=upperarm_offset_r}
    rotation=<rot -20 -45 40>
    scale=<0.4 0.4 0.4>
    translation {x=-3, z=-1}
   }
   n62 = "Transform (Modify) Bone - Bone: dyn_hoop"(ComponentPose=@n61, Alpha=@n59, Translation="0,0.3,0", Rotation="0,-3,0", Scale="1,1,1") {
    rotationMode=BMM_Additive
    rotationSpace=BCS_BoneSpace
    scaleSpace=BCS_BoneSpace
    translationMode=BMM_Additive
    translationSpace=BCS_BoneSpace
    alphaBoolBlend {blendOption=Linear}
    boneToModify {boneName=dyn_hoop}
    rotation {yaw=-3}
    translation {y=0.3}
   }
   n63 = "Constraint - dyn_parent_strap_A_01_l (PPPPPPP)"!(
    ComponentPose=@n62
    Alpha=@n52
    ConstraintWeights_0=0
    ConstraintWeights_1:float=0
    ConstraintWeights_2:float=0
    ConstraintWeights_3:float=0
    ConstraintWeights_4:float=0.7164
    ConstraintWeights_5:float=0.181
    ConstraintWeights_6:float=0.1026
   ) -> (Pose) {
    alphaBoolBlend {blendOption=Linear}
    boneToModify {boneName=dyn_parent_strap_A_01_l}
    constraintSetup [
     {targetBone {boneName=thigh_r}}
     {targetBone {boneName=thigh_twist_01_r}}
     {targetBone {boneName=spine_01}}
     {targetBone {boneName=spine_02}}
     {targetBone {boneName=thigh_twist_01_l}}
     {targetBone {boneName=thigh_l}}
     {targetBone {boneName=pelvis}}
    ]
    constraintWeights [1,1,1,1,1,1,1]
   }
   n64 = "Constraint - dyn_parent_strap_B_01_l (PPP)"!(ComponentPose=@n63, Alpha=@n52, ConstraintWeights_0=0.1, ConstraintWeights_1:float=0.9, ConstraintWeights_2:float=0) -> (Pose) {
    alphaBoolBlend {blendOption=Linear}
    boneToModify {boneName=dyn_parent_strap_B_01_l}
    constraintSetup [{targetBone {boneName=pelvis}}, {targetBone {boneName=thigh_twist_01_l}}, {targetBone {boneName=thigh_l}}]
    constraintWeights [1,1,1]
   }
   n65 = "Constraint - dyn_parent_strap_A_01_r (PPP)"!(ComponentPose=@n64, Alpha=1, ConstraintWeights_0=0.0818, ConstraintWeights_1:float=0.1995, ConstraintWeights_2:float=0.7187) -> (Pose) {
    alphaBoolBlend {blendOption=Linear}
    boneToModify {boneName=dyn_parent_strap_A_01_r}
    constraintSetup [{targetBone {boneName=thigh_r}}, {targetBone {boneName=pelvis}}, {targetBone {boneName=thigh_twist_01_r}}]
    constraintWeights [1,1,1]
   }
   n66 = "Constraint - dyn_parent_strap_B_01_r (PPP)"!(ComponentPose=@n65, Alpha=1, ConstraintWeights_0=0, ConstraintWeights_1:float=0.1, ConstraintWeights_2:float=0.9) -> (Pose) {
    alphaBoolBlend {blendOption=Linear}
    boneToModify {boneName=dyn_parent_strap_B_01_r}
    constraintSetup [{targetBone {boneName=thigh_r}}, {targetBone {boneName=pelvis}}, {targetBone {boneName=thigh_twist_01_r}}]
    constraintWeights [1,1,1]
   }
   n67 = Get GravityOverride()
   n68 = Anim Dynamics - Bone: dyn_strap_B_01_l!(ComponentPose=@n66, Alpha=1, GravityOverride:Vector=@n67, ExternalForce="0,0,0") -> (Pose) {
    angularDampingOverride=0.9999
    angularSpringConstant=75
    bAngularSpring=true
    bGravityOverrideInSimSpace=true
    bOverrideAngularDamping=true
    bUseGravityOverride=true
    simulationSpace=RootRelative
    alphaBoolBlend {blendOption=Linear}
    boundBone {boneName=dyn_strap_B_01_l}
    componentAppliedLinearAccClamp=<3500 3500 3500>
    componentLinearAccScale=<0.9 0.9 0.3>
    componentLinearVelScale=<0.9 0.9 0.3>
    gravityOverride {z=-980}
    physicsBodyDefinitions [{
      sphereCollisionRadius=0
      boundBone {boneName=dyn_strap_B_01_l}
      boxExtents=<15 10 10>
      constraintSetup {angularLimitsMax=<1 35 25>, angularLimitsMin=<-1 -35 0>, angularTarget=<0 0 0>}
      localJointOffset=<2 0 0>
     }]
    relativeSpaceBone {boneName=pelvis}
   }
   n69 = Anim Dynamics - Bone: dyn_strap_B_02_l!(ComponentPose=@n68, Alpha=1, GravityOverride:Vector=@n67, ExternalForce="0,0,0") -> (Pose) {
    angularDampingOverride=0.9999
    angularSpringConstant=75
    bAngularSpring=true
    bGravityOverrideInSimSpace=true
    bOverrideAngularDamping=true
    bUseGravityOverride=true
    simulationSpace=RootRelative
    alphaBoolBlend {blendOption=Linear}
    boundBone {boneName=dyn_strap_B_02_l}
    componentAppliedLinearAccClamp=<3500 3500 3500>
    componentLinearAccScale=<0.9 0.9 0.3>
    componentLinearVelScale=<0.9 0.9 0.3>
    gravityOverride {z=-980}
    physicsBodyDefinitions [{
      sphereCollisionRadius=0
      boundBone {boneName=dyn_strap_B_02_l}
      boxExtents=<15 10 10>
      constraintSetup {angularLimitsMax=<1 35 25>, angularLimitsMin=<-1 -35 0>, angularTarget=<0 0 0>}
      localJointOffset=<2 0 0>
     }]
    relativeSpaceBone {boneName=pelvis}
   }
   n70 = Anim Dynamics - Bone: dyn_strap_B_03_l!(ComponentPose=@n69, Alpha=1, GravityOverride:Vector=@n67, ExternalForce="0,0,0") -> (Pose) {
    angularDampingOverride=0.9999
    angularSpringConstant=75
    bAngularSpring=true
    bGravityOverrideInSimSpace=true
    bOverrideAngularDamping=true
    bUseGravityOverride=true
    simulationSpace=RootRelative
    alphaBoolBlend {blendOption=Linear}
    boundBone {boneName=dyn_strap_B_03_l}
    componentAppliedLinearAccClamp=<3500 3500 3500>
    componentLinearAccScale=<0.9 0.9 0.3>
    componentLinearVelScale=<0.9 0.9 0.3>
    gravityOverride {z=-980}
    physicsBodyDefinitions [{
      sphereCollisionRadius=0
      boundBone {boneName=dyn_strap_B_03_l}
      boxExtents=<15 10 10>
      constraintSetup {angularLimitsMax=<1 35 25>, angularLimitsMin=<-1 -35 -5>, angularTarget=<0 0 0>}
      localJointOffset=<2 0 0>
     }]
    relativeSpaceBone {boneName=pelvis}
   }
   n71 = Anim Dynamics - Bone: dyn_strap_B_04_l!(ComponentPose=@n70, Alpha=1, GravityOverride:Vector=@n67, ExternalForce="0,0,0") -> (Pose) {
    angularDampingOverride=0.9999
    angularSpringConstant=75
    bAngularSpring=true
    bGravityOverrideInSimSpace=true
    bOverrideAngularDamping=true
    bUseGravityOverride=true
    simulationSpace=RootRelative
    alphaBoolBlend {blendOption=Linear}
    boundBone {boneName=dyn_strap_B_04_l}
    componentAppliedLinearAccClamp=<3500 3500 3500>
    componentLinearAccScale=<0.9 0.9 0.3>
    componentLinearVelScale=<0.9 0.9 0.3>
    gravityOverride {z=-980}
    physicsBodyDefinitions [{
      sphereCollisionRadius=0
      boundBone {boneName=dyn_strap_B_04_l}
      boxExtents=<15 10 10>
      constraintSetup {angularLimitsMax=<1 35 25>, angularLimitsMin=<-1 -35 -5>, angularTarget=<0 0 0>}
      localJointOffset=<2 0 0>
     }]
    relativeSpaceBone {boneName=pelvis}
   }
   n72 = Anim Dynamics - Bone: dyn_strap_B_05_l!(ComponentPose=@n71, Alpha=1, GravityOverride:Vector=@n67, ExternalForce="0,0,0") -> (Pose) {
    angularDampingOverride=0.9999
    angularSpringConstant=75
    bAngularSpring=true
    bGravityOverrideInSimSpace=true
    bOverrideAngularDamping=true
    bUseGravityOverride=true
    simulationSpace=RootRelative
    alphaBoolBlend {blendOption=Linear}
    boundBone {boneName=dyn_strap_B_05_l}
    componentAppliedLinearAccClamp=<3500 3500 3500>
    componentLinearAccScale=<0.9 0.9 0.3>
    componentLinearVelScale=<0.9 0.9 0.3>
    gravityOverride {z=-980}
    physicsBodyDefinitions [{
      sphereCollisionRadius=0
      boundBone {boneName=dyn_strap_B_05_l}
      boxExtents=<15 10 10>
      constraintSetup {angularLimitsMax=<1 35 25>, angularLimitsMin=<-1 -35 -5>, angularTarget=<0 0 0>}
      localJointOffset=<2 0 0>
     }]
    relativeSpaceBone {boneName=pelvis}
   }
   n73 = Anim Dynamics - Bone: dyn_strap_A_01_l!(ComponentPose=@n72, Alpha=1, GravityOverride:Vector=@n67, ExternalForce="0,0,0") -> (Pose) {
    angularDampingOverride=0.9999
    angularSpringConstant=75
    bAngularSpring=true
    bGravityOverrideInSimSpace=true
    bOverrideAngularDamping=true
    bUseGravityOverride=true
    bUseSphericalLimits=true
    simulationSpace=RootRelative
    alphaBoolBlend {blendOption=Linear}
    boundBone {boneName=dyn_strap_A_01_l}
    componentAppliedLinearAccClamp=<3500 3500 3500>
    componentLinearAccScale=<0.9 0.9 0.3>
    componentLinearVelScale=<0.9 0.9 0.3>
    gravityOverride {z=-980}
    physicsBodyDefinitions [{
      collisionType=CustomSphere
      sphereCollisionRadius=2
      boundBone {boneName=dyn_strap_A_01_l}
      boxExtents=<15 10 10>
      constraintSetup {angularLimitsMax=<1 35 25>, angularLimitsMin=<-1 -35 7>, angularTarget=<0 0 0>}
      localJointOffset=<2 0 0>
     }]
    relativeSpaceBone {boneName=pelvis}
    sphericalLimits [{limitRadius=16, drivingBone {boneName=thigh_l}, sphereLocalOffset=<-12 -3 9>}]
   }
   n74 = Anim Dynamics - Bone: dyn_strap_A_02_l!(ComponentPose=@n73, Alpha=1, GravityOverride:Vector=@n67, ExternalForce="0,0,0") -> (Pose) {
    angularDampingOverride=0.9999
    angularSpringConstant=75
    bAngularSpring=true
    bGravityOverrideInSimSpace=true
    bOverrideAngularDamping=true
    bUseGravityOverride=true
    bUseSphericalLimits=true
    simulationSpace=RootRelative
    alphaBoolBlend {blendOption=Linear}
    boundBone {boneName=dyn_strap_A_02_l}
    componentAppliedLinearAccClamp=<3500 3500 3500>
    componentLinearAccScale=<0.9 0.9 0.3>
    componentLinearVelScale=<0.9 0.9 0.3>
    gravityOverride {z=-980}
    physicsBodyDefinitions [{
      collisionType=CustomSphere
      sphereCollisionRadius=2
      boundBone {boneName=dyn_strap_A_02_l}
      boxExtents=<15 10 10>
      constraintSetup {angularLimitsMax=<1 35 25>, angularLimitsMin=<-1 -35 3>, angularTarget=<0 0 0>}
      localJointOffset=<2 0 0>
     }]
    relativeSpaceBone {boneName=pelvis}
    sphericalLimits [{limitRadius=25, drivingBone {boneName=thigh_l}, sphereLocalOffset=<-14 -4 18.5>}]
   }
   n75 = Anim Dynamics - Bone: dyn_strap_A_03_l!(ComponentPose=@n74, Alpha=1, GravityOverride:Vector=@n67, ExternalForce="0,0,0") -> (Pose) {
    angularDampingOverride=0.9999
    angularSpringConstant=75
    bAngularSpring=true
    bGravityOverrideInSimSpace=true
    bOverrideAngularDamping=true
    bUseGravityOverride=true
    bUseSphericalLimits=true
    simulationSpace=RootRelative
    alphaBoolBlend {blendOption=Linear}
    boundBone {boneName=dyn_strap_A_03_l}
    componentAppliedLinearAccClamp=<3500 3500 3500>
    componentLinearAccScale=<0.9 0.9 0.3>
    componentLinearVelScale=<0.9 0.9 0.3>
    gravityOverride {z=-980}
    physicsBodyDefinitions [{
      collisionType=CustomSphere
      sphereCollisionRadius=2
      boundBone {boneName=dyn_strap_A_03_l}
      boxExtents=<15 10 10>
      constraintSetup {angularLimitsMax=<1 35 25>, angularLimitsMin=<-1 -35 -5>, angularTarget=<0 0 0>}
      localJointOffset=<2 0 0>
     }]
    relativeSpaceBone {boneName=pelvis}
    sphericalLimits [{limitRadius=25, drivingBone {boneName=thigh_l}, sphereLocalOffset=<-20 -4 18>}]
   }
   n76 = Anim Dynamics - Bone: dyn_strap_A_04_l!(ComponentPose=@n75, Alpha=1, GravityOverride:Vector=@n67, ExternalForce="0,0,0") -> (Pose) {
    angularDampingOverride=0.9999
    angularSpringConstant=75
    bAngularSpring=true
    bGravityOverrideInSimSpace=true
    bOverrideAngularDamping=true
    bUseGravityOverride=true
    bUseSphericalLimits=true
    simulationSpace=RootRelative
    alphaBoolBlend {blendOption=Linear}
    boundBone {boneName=dyn_strap_A_04_l}
    componentAppliedLinearAccClamp=<3500 3500 3500>
    componentLinearAccScale=<0.9 0.9 0.3>
    componentLinearVelScale=<0.9 0.9 0.3>
    gravityOverride {z=-980}
    physicsBodyDefinitions [{
      collisionType=CustomSphere
      sphereCollisionRadius=2
      boundBone {boneName=dyn_strap_A_04_l}
      boxExtents=<15 10 10>
      constraintSetup {angularLimitsMax=<1 35 25>, angularLimitsMin=<-1 -35 -5>, angularTarget=<0 0 0>}
      localJointOffset=<2 0 0>
     }]
    relativeSpaceBone {boneName=pelvis}
    sphericalLimits [{limitRadius=25, drivingBone {boneName=thigh_l}, sphereLocalOffset=<-25 -4 20>}]
   }
   n77 = Anim Dynamics - Bone: dyn_strap_A_05_l!(ComponentPose=@n76, Alpha=1, GravityOverride:Vector=@n67, ExternalForce="0,0,0") -> (Pose) {
    angularDampingOverride=0.9999
    angularSpringConstant=75
    bAngularSpring=true
    bGravityOverrideInSimSpace=true
    bOverrideAngularDamping=true
    bUseGravityOverride=true
    simulationSpace=RootRelative
    alphaBoolBlend {blendOption=Linear}
    boundBone {boneName=dyn_strap_A_05_l}
    componentAppliedLinearAccClamp=<3500 3500 3500>
    componentLinearAccScale=<0.9 0.9 0.3>
    componentLinearVelScale=<0.9 0.9 0.3>
    gravityOverride {z=-980}
    physicsBodyDefinitions [{
      sphereCollisionRadius=0
      boundBone {boneName=dyn_strap_A_05_l}
      boxExtents=<15 10 10>
      constraintSetup {angularLimitsMax=<1 35 25>, angularLimitsMin=<-1 -35 0>, angularTarget=<0 0 0>}
      localJointOffset=<2 0 0>
     }]
    relativeSpaceBone {boneName=pelvis}
   }
   n78 = Anim Dynamics - Bone: dyn_strap_B_01_r!(ComponentPose=@n77, Alpha=1, GravityOverride:Vector=@n67, ExternalForce="0,0,0") -> (Pose) {
    angularDampingOverride=0.9999
    angularSpringConstant=75
    bAngularSpring=true
    bGravityOverrideInSimSpace=true
    bOverrideAngularDamping=true
    bUseGravityOverride=true
    simulationSpace=RootRelative
    alphaBoolBlend {blendOption=Linear}
    boundBone {boneName=dyn_strap_B_01_r}
    componentAppliedLinearAccClamp=<3500 3500 3500>
    componentLinearAccScale=<0.9 0.9 0.3>
    componentLinearVelScale=<0.9 0.9 0.3>
    gravityOverride {z=-980}
    physicsBodyDefinitions [{
      sphereCollisionRadius=0
      boundBone {boneName=dyn_strap_B_01_r}
      boxExtents=<15 10 10>
      constraintSetup {angularLimitsMax=<1 35 25>, angularLimitsMin=<-1 -35 0>, angularTarget=<0 0 0>}
      localJointOffset=<2 0 0>
     }]
    relativeSpaceBone {boneName=pelvis}
   }
   n79 = Anim Dynamics - Bone: dyn_strap_B_02_r!(ComponentPose=@n78, Alpha=1, GravityOverride:Vector=@n67, ExternalForce="0,0,0") -> (Pose) {
    angularDampingOverride=0.9999
    angularSpringConstant=75
    bAngularSpring=true
    bGravityOverrideInSimSpace=true
    bOverrideAngularDamping=true
    bUseGravityOverride=true
    simulationSpace=RootRelative
    alphaBoolBlend {blendOption=Linear}
    boundBone {boneName=dyn_strap_B_02_r}
    componentAppliedLinearAccClamp=<3500 3500 3500>
    componentLinearAccScale=<0.9 0.9 0.3>
    componentLinearVelScale=<0.9 0.9 0.3>
    gravityOverride {z=-980}
    physicsBodyDefinitions [{
      sphereCollisionRadius=0
      boundBone {boneName=dyn_strap_B_02_r}
      boxExtents=<15 10 10>
      constraintSetup {angularLimitsMax=<1 35 25>, angularLimitsMin=<-1 -35 0>, angularTarget=<0 0 0>}
      localJointOffset=<2 0 0>
     }]
    relativeSpaceBone {boneName=pelvis}
   }
   n80 = Anim Dynamics - Bone: dyn_strap_B_03_r!(ComponentPose=@n79, Alpha=1, GravityOverride:Vector=@n67, ExternalForce="0,0,0") -> (Pose) {
    angularDampingOverride=0.9999
    angularSpringConstant=75
    bAngularSpring=true
    bGravityOverrideInSimSpace=true
    bOverrideAngularDamping=true
    bUseGravityOverride=true
    simulationSpace=RootRelative
    alphaBoolBlend {blendOption=Linear}
    boundBone {boneName=dyn_strap_B_03_r}
    componentAppliedLinearAccClamp=<3500 3500 3500>
    componentLinearAccScale=<0.9 0.9 0.3>
    componentLinearVelScale=<0.9 0.9 0.3>
    gravityOverride {z=-980}
    physicsBodyDefinitions [{
      sphereCollisionRadius=0
      boundBone {boneName=dyn_strap_B_03_r}
      boxExtents=<15 10 10>
      constraintSetup {angularLimitsMax=<1 35 25>, angularLimitsMin=<-1 -35 -5>, angularTarget=<0 0 0>}
      localJointOffset=<2 0 0>
     }]
    relativeSpaceBone {boneName=pelvis}
   }
   n81 = Anim Dynamics - Bone: dyn_strap_B_04_r!(ComponentPose=@n80, Alpha=1, GravityOverride:Vector=@n67, ExternalForce="0,0,0") -> (Pose) {
    angularDampingOverride=0.9999
    angularSpringConstant=75
    bAngularSpring=true
    bGravityOverrideInSimSpace=true
    bOverrideAngularDamping=true
    bUseGravityOverride=true
    bUseSphericalLimits=true
    simulationSpace=RootRelative
    alphaBoolBlend {blendOption=Linear}
    boundBone {boneName=dyn_strap_B_04_r}
    componentAppliedLinearAccClamp=<3500 3500 3500>
    componentLinearAccScale=<0.9 0.9 0.3>
    componentLinearVelScale=<0.9 0.9 0.3>
    gravityOverride {z=-980}
    physicsBodyDefinitions [{
      collisionType=CustomSphere
      sphereCollisionRadius=2
      boundBone {boneName=dyn_strap_B_04_r}
      boxExtents=<15 10 10>
      constraintSetup {angularLimitsMax=<1 35 25>, angularLimitsMin=<-1 -35 -5>, angularTarget=<0 0 0>}
      localJointOffset=<2 0 0>
     }]
    relativeSpaceBone {boneName=pelvis}
    sphericalLimits [{limitRadius=12, drivingBone {boneName=calf_twist_02_r}, sphereLocalOffset=<0 0 -9>}]
   }
   n82 = Anim Dynamics - Bone: dyn_strap_B_05_r!(ComponentPose=@n81, Alpha=1, GravityOverride:Vector=@n67, ExternalForce="0,0,0") -> (Pose) {
    angularDampingOverride=0.9999
    angularSpringConstant=75
    bAngularSpring=true
    bGravityOverrideInSimSpace=true
    bOverrideAngularDamping=true
    bUseGravityOverride=true
    bUseSphericalLimits=true
    simulationSpace=RootRelative
    alphaBoolBlend {blendOption=Linear}
    boundBone {boneName=dyn_strap_B_05_r}
    componentAppliedLinearAccClamp=<3500 3500 3500>
    componentLinearAccScale=<0.9 0.9 0.3>
    componentLinearVelScale=<0.9 0.9 0.3>
    gravityOverride {z=-980}
    physicsBodyDefinitions [{
      collisionType=CustomSphere
      sphereCollisionRadius=2
      boundBone {boneName=dyn_strap_B_05_r}
      boxExtents=<15 10 10>
      constraintSetup {angularLimitsMax=<1 35 25>, angularLimitsMin=<-1 -35 -5>, angularTarget=<0 0 0>}
      localJointOffset=<2 0 0>
     }]
    relativeSpaceBone {boneName=pelvis}
    sphericalLimits [{limitRadius=12, drivingBone {boneName=calf_twist_02_r}, sphereLocalOffset=<0 0 -9>}]
   }
   n83 = Anim Dynamics - Bone: dyn_strap_A_01_r!(ComponentPose=@n82, Alpha=1, GravityOverride:Vector=@n67, ExternalForce="0,0,0") -> (Pose) {
    angularDampingOverride=0.9999
    angularSpringConstant=75
    bAngularSpring=true
    bGravityOverrideInSimSpace=true
    bOverrideAngularDamping=true
    bUseGravityOverride=true
    simulationSpace=RootRelative
    alphaBoolBlend {blendOption=Linear}
    boundBone {boneName=dyn_strap_A_01_r}
    componentAppliedLinearAccClamp=<3500 3500 3500>
    componentLinearAccScale=<0.9 0.9 0.3>
    componentLinearVelScale=<0.9 0.9 0.3>
    gravityOverride {z=-980}
    physicsBodyDefinitions [{
      sphereCollisionRadius=0
      boundBone {boneName=dyn_strap_A_01_r}
      boxExtents=<15 10 10>
      constraintSetup {angularLimitsMax=<1 35 25>, angularLimitsMin=<-1 -35 0>, angularTarget=<0 0 0>}
      localJointOffset=<2 0 0>
     }]
    relativeSpaceBone {boneName=pelvis}
   }
   n84 = Anim Dynamics - Bone: dyn_strap_A_02_r!(ComponentPose=@n83, Alpha=1, GravityOverride:Vector=@n67, ExternalForce="0,0,0") -> (Pose) {
    angularDampingOverride=0.9999
    angularSpringConstant=75
    bAngularSpring=true
    bGravityOverrideInSimSpace=true
    bOverrideAngularDamping=true
    bUseGravityOverride=true
    simulationSpace=RootRelative
    alphaBoolBlend {blendOption=Linear}
    boundBone {boneName=dyn_strap_A_02_r}
    componentAppliedLinearAccClamp=<3500 3500 3500>
    componentLinearAccScale=<0.9 0.9 0.3>
    componentLinearVelScale=<0.9 0.9 0.3>
    gravityOverride {z=-980}
    physicsBodyDefinitions [{
      sphereCollisionRadius=0
      boundBone {boneName=dyn_strap_A_02_r}
      boxExtents=<15 10 10>
      constraintSetup {angularLimitsMax=<1 35 25>, angularLimitsMin=<-1 -35 0>, angularTarget=<0 0 0>}
      localJointOffset=<2 0 0>
     }]
    relativeSpaceBone {boneName=pelvis}
   }
   n85 = Anim Dynamics - Bone: dyn_strap_A_03_r!(ComponentPose=@n84, Alpha=1, GravityOverride:Vector=@n67, ExternalForce="0,0,0") -> (Pose) {
    angularDampingOverride=0.9999
    angularSpringConstant=75
    bAngularSpring=true
    bGravityOverrideInSimSpace=true
    bOverrideAngularDamping=true
    bUseGravityOverride=true
    simulationSpace=RootRelative
    alphaBoolBlend {blendOption=Linear}
    boundBone {boneName=dyn_strap_A_03_r}
    componentAppliedLinearAccClamp=<3500 3500 3500>
    componentLinearAccScale=<0.9 0.9 0.3>
    componentLinearVelScale=<0.9 0.9 0.3>
    gravityOverride {z=-980}
    physicsBodyDefinitions [{
      sphereCollisionRadius=0
      boundBone {boneName=dyn_strap_A_03_r}
      boxExtents=<15 10 10>
      constraintSetup {angularLimitsMax=<1 35 25>, angularLimitsMin=<-1 -35 -5>, angularTarget=<0 0 0>}
      localJointOffset=<2 0 0>
     }]
    relativeSpaceBone {boneName=pelvis}
   }
   n86 = Anim Dynamics - Bone: dyn_strap_A_04_r!(ComponentPose=@n85, Alpha=1, GravityOverride:Vector=@n67, ExternalForce="0,0,0") -> (Pose) {
    angularDampingOverride=0.9999
    angularSpringConstant=75
    bAngularSpring=true
    bGravityOverrideInSimSpace=true
    bOverrideAngularDamping=true
    bUseGravityOverride=true
    simulationSpace=RootRelative
    alphaBoolBlend {blendOption=Linear}
    boundBone {boneName=dyn_strap_A_04_r}
    componentAppliedLinearAccClamp=<3500 3500 3500>
    componentLinearAccScale=<0.9 0.9 0.3>
    componentLinearVelScale=<0.9 0.9 0.3>
    gravityOverride {z=-980}
    physicsBodyDefinitions [{
      sphereCollisionRadius=0
      boundBone {boneName=dyn_strap_A_04_r}
      boxExtents=<15 10 10>
      constraintSetup {angularLimitsMax=<1 35 25>, angularLimitsMin=<-1 -35 -5>, angularTarget=<0 0 0>}
      localJointOffset=<2 0 0>
     }]
    relativeSpaceBone {boneName=pelvis}
   }
   n87 = Anim Dynamics - Bone: dyn_strap_A_05_r!(ComponentPose=@n86, Alpha=1, GravityOverride:Vector=@n67, ExternalForce="0,0,0") -> (Pose) {
    angularDampingOverride=0.9999
    angularSpringConstant=75
    bAngularSpring=true
    bGravityOverrideInSimSpace=true
    bOverrideAngularDamping=true
    bUseGravityOverride=true
    simulationSpace=RootRelative
    alphaBoolBlend {blendOption=Linear}
    boundBone {boneName=dyn_strap_A_05_r}
    componentAppliedLinearAccClamp=<3500 3500 3500>
    componentLinearAccScale=<0.9 0.9 0.3>
    componentLinearVelScale=<0.9 0.9 0.3>
    gravityOverride {z=-980}
    physicsBodyDefinitions [{
      sphereCollisionRadius=0
      boundBone {boneName=dyn_strap_A_05_r}
      boxExtents=<15 10 10>
      constraintSetup {angularLimitsMax=<1 35 25>, angularLimitsMin=<-1 -35 -5>, angularTarget=<0 0 0>}
      localJointOffset=<2 0 0>
     }]
    relativeSpaceBone {boneName=pelvis}
   }
   n88 = "Transform (Modify) Bone - Bone: dyn_driver_strap_A_01_l"(ComponentPose=@n87, Alpha=Get rban_straps_alpha(), Translation="0,0.5,0", Rotation="0,1,-15", Scale="1,1,1") {
    rotationMode=BMM_Additive
    rotationSpace=BCS_BoneSpace
    translationMode=BMM_Additive
    translationSpace=BCS_BoneSpace
    alphaBoolBlend {blendOption=Linear}
    boneToModify {boneName=dyn_driver_strap_A_01_l}
    rotation {roll=-15, yaw=1}
    translation {y=0.5}
   }
   n89 = "Transform (Modify) Bone - Bone: dyn_driver_strap_01_r"(ComponentPose=@n88, Alpha=Get legUp_Strap_R(), Translation="3.5,-3.5,5", Rotation="-20,0,-30", Scale="1,1,1") {
    rotationMode=BMM_Additive
    rotationSpace=BCS_BoneSpace
    translationMode=BMM_Additive
    translationSpace=BCS_BoneSpace
    alphaBoolBlend {blendOption=Linear}
    boneToModify {boneName=dyn_driver_strap_01_r}
    rotation {pitch=-20, roll=-30}
    translation=<3.5 -3.5 5>
   }
   n90 = "Transform (Modify) Bone - Bone: dyn_strap_frt_01_r"(ComponentPose=@n89, Alpha=@n51, Translation="0,0,0", Rotation="-70,0,0", Scale="1,1,1") {
    rotationMode=BMM_Additive
    rotationSpace=BCS_BoneSpace
    translationMode=BMM_Additive
    translationSpace=BCS_BoneSpace
    alphaBoolBlend {blendOption=Linear}
    boneToModify {boneName=dyn_strap_frt_01_r}
    rotation {pitch=-70}
   }
   n91 = "Transform (Modify) Bone - Bone: dyn_driver_strap_A_01_r"(ComponentPose=@n90, Alpha=@n51, Translation="0,1,0", Rotation="0,7,0", Scale="1,1,1") {
    rotationMode=BMM_Additive
    rotationSpace=BCS_BoneSpace
    translationMode=BMM_Additive
    translationSpace=BCS_BoneSpace
    alphaBoolBlend {blendOption=Linear}
    boneToModify {boneName=dyn_driver_strap_A_01_r}
    rotation {yaw=7}
    translation {y=1}
   }
   n92 = "Transform (Modify) Bone - Bone: dyn_driver_strap_B_01_r"(ComponentPose=@n91, Alpha=@n51, Translation="0,0.75,0", Rotation="0,5,0", Scale="1,1,1") {
    rotationMode=BMM_Additive
    rotationSpace=BCS_BoneSpace
    translationMode=BMM_Additive
    translationSpace=BCS_BoneSpace
    alphaBoolBlend {blendOption=Linear}
    boneToModify {boneName=dyn_driver_strap_B_01_r}
    rotation {yaw=5}
    translation {y=0.75}
   }
   n93 = Get legOut_R() -> (legOut_R:double)
   n94 = "Transform (Modify) Bone - Bone: dyn_driver_strap_01_r"(ComponentPose=@n92, Alpha=@n93, Translation="0,0,0", Rotation="-15,0,0", Scale="1,1,1") {
    rotationMode=BMM_Additive
    rotationSpace=BCS_BoneSpace
    translationMode=BMM_Additive
    translationSpace=BCS_BoneSpace
    alphaBoolBlend {blendOption=Linear}
    boneToModify {boneName=dyn_driver_strap_01_r}
    rotation {pitch=-15}
   }
   n95 = "Transform (Modify) Bone - Bone: dyn_driver_strap_A_01_r"(ComponentPose=@n94, Alpha=@n93, Translation="0,2,0", Rotation="0,15,0", Scale="1,1,1") {
    rotationMode=BMM_Additive
    rotationSpace=BCS_BoneSpace
    translationMode=BMM_Additive
    translationSpace=BCS_BoneSpace
    alphaBoolBlend {blendOption=Linear}
    boneToModify {boneName=dyn_driver_strap_A_01_r}
    rotation {yaw=15}
    translation {y=2}
   }
   n96 = "Transform (Modify) Bone - Bone: dyn_driver_strap_B_01_r"(ComponentPose=@n95, Alpha=@n93, Translation="0,0.75,0", Rotation="0,15,0", Scale="1,1,1") {
    rotationMode=BMM_Additive
    rotationSpace=BCS_BoneSpace
    translationMode=BMM_Additive
    translationSpace=BCS_BoneSpace
    alphaBoolBlend {blendOption=Linear}
    boneToModify {boneName=dyn_driver_strap_B_01_r}
    rotation {yaw=15}
    translation {y=0.75}
   }
   n97 = "Transform (Modify) Bone - Bone: dyn_driver_strap_01_l"(ComponentPose=@n96, Alpha=Get legUp_Strap_L() -> (legUp_Strap_L:double), Translation="-3.5,3.5,-5", Rotation="-20,0,-30", Scale="1,1,1") {
    rotationMode=BMM_Additive
    rotationSpace=BCS_BoneSpace
    translationMode=BMM_Additive
    translationSpace=BCS_BoneSpace
    alphaBoolBlend {blendOption=Linear}
    alphaScaleBiasClamp {bInterpResult=true, interpSpeedDecreasing=15, interpSpeedIncreasing=15}
    boneToModify {boneName=dyn_driver_strap_01_l}
    rotation {pitch=-20, roll=-30}
    translation=<-3.5 3.5 -5>
   }
   n98 = "Transform (Modify) Bone - Bone: dyn_strap_frt_01_l"(ComponentPose=@n97, Alpha=Get legUp_Strap_L(), Translation="0,6,0", Rotation="-90,0,0", Scale="1,1,1") {
    rotationMode=BMM_Additive
    rotationSpace=BCS_BoneSpace
    translationMode=BMM_Additive
    translationSpace=BCS_BoneSpace
    alphaBoolBlend {blendOption=Linear}
    alphaScaleBiasClamp {bInterpResult=true, interpSpeedDecreasing=15, interpSpeedIncreasing=15}
    boneToModify {boneName=dyn_strap_frt_01_l}
    rotation {pitch=-90}
    translation {y=6}
   }
   n99 = "Transform (Modify) Bone - Bone: dyn_driver_strap_01_l"(ComponentPose=@n98, Alpha=Get legOut_L() -> (legOut_L:double), Translation="0,0,0", Rotation="-5,0,-10", Scale="1,1,1") {
    rotationMode=BMM_Additive
    rotationSpace=BCS_BoneSpace
    translationMode=BMM_Additive
    translationSpace=BCS_BoneSpace
    alphaBoolBlend {blendOption=Linear}
    boneToModify {boneName=dyn_driver_strap_01_l}
    rotation {pitch=-5, roll=-10}
   }
   n100 = RigidBody!(ComponentPose=@n99, Alpha=Get cloth_rear_alpha() -> (cloth_rear_alpha:double), OverrideWorldGravity:Vector="0,0,-980", ExternalForce) -> (Pose) {
    bClampLinearTranslationLimitToRefPose=true
    bOverrideWorldGravity=true
    overridePhysicsAsset=/Game/Characters/Player/Female/Medium/Bodies/F_MED_Ancient_Gladiator/Meshes/physics/F_MED_Ancient_Gladiator_Rear_Physics
    simulationSpace=BaseBoneSpace
    alphaBoolBlend {blendOption=Linear}
    baseBoneRef {boneName=root}
    componentAppliedLinearAccClamp=<1000 1000 2000>
    componentLinearAccScale=<0.35 0.35 0.4>
    componentLinearVelScale=<0.1 0.1 0.25>
    overrideWorldGravity {z=-980}
   }
   n101 = RigidBody(ComponentPose=@n100, Alpha=Get hip_chains_alpha() -> (hip_chains_alpha:double), OverrideWorldGravity="0,0,-980") {
    bOverrideWorldGravity=true
    overridePhysicsAsset=/Game/Characters/Player/Female/Medium/Bodies/F_MED_Ancient_Gladiator/Meshes/physics/F_MED_Ancient_Gladiator_Chains_Hips
    alphaBoolBlend {blendOption=Linear}
    baseBoneRef {boneName=root}
    componentAppliedLinearAccClamp=<1000 1000 1000>
    overrideWorldGravity {z=-980}
    simSpaceSettings {
     maxAngularVelocity=1000
     maxLinearAcceleration=8000
     maxLinearVelocity=100000
     velocityScaleZ=0.35
     worldAlpha=1
     externalLinearDragV {z=0.3}
    }
   }
   n102 = "Transform (Modify) Bone - Bone: dyn_driver_shldrpad_l"(ComponentPose=@n101, Alpha=Get fe_pad_alpha() -> (fe_pad_alpha:double), Translation="0,-2,0", Rotation="0,-5,0", Scale="1,1,1") {
    rotationMode=BMM_Additive
    rotationSpace=BCS_BoneSpace
    translationMode=BMM_Additive
    translationSpace=BCS_BoneSpace
    alphaBoolBlend {blendOption=Linear}
    boneToModify {boneName=dyn_driver_shldrpad_l}
    rotation {yaw=-5}
    translation {y=-2}
   }
   Save cached pose 'CrouchCache'(Pose=Component To Local(ComponentPose=@n102)) {cacheName=CrouchCache}
   n103 = "Transform (Modify) Bone - Bone: dyn_pelvis_offset"(ComponentPose=Local To Component(LocalPose=Use cached pose 'CrouchCache'()), Alpha=1, Translation="0,0,1.5", Rotation="0,0,0", Scale="1,1,1") {
    rotationMode=BMM_Additive
    rotationSpace=BCS_BoneSpace
    translationMode=BMM_Additive
    translationSpace=BCS_BoneSpace
    alphaBoolBlend {blendOption=Linear}
    boneToModify {boneName=dyn_pelvis_offset}
    translation {z=1.5}
   }
   n104 = RigidBody(ComponentPose=@n103, OverrideWorldGravity="0,0,-980") {
    bClampLinearTranslationLimitToRefPose=true
    bOverrideWorldGravity=true
    overridePhysicsAsset=/Game/Characters/Player/Female/Medium/Bodies/F_MED_Ancient_Gladiator/Meshes/physics/F_MED_Ancient_Gladiator_Skirt
    simulationSpace=BaseBoneSpace
    alphaBoolBlend {blendOption=Linear}
    baseBoneRef {boneName=root}
    componentLinearAccScale=<0.5 0.5 0.35>
    componentLinearVelScale=<0.5 0.5 0.35>
    overrideWorldGravity {z=-980}
   }
   n105 = "Constraint - strap_aim_01_l (PPPPPP)"(
    ComponentPose=@n104
    Alpha=1
    ConstraintWeights_0=0.7021
    ConstraintWeights_1=0.2395
    ConstraintWeights_2=0.0003
    ConstraintWeights_3=0.0576
    ConstraintWeights_4=0.0001
    ConstraintWeights_5=0.0003
   ) {
    alphaBoolBlend {blendOption=Linear}
    boneToModify {boneName=strap_aim_01_l}
    constraintSetup [
     {targetBone {boneName=dyn_pelvis_offset}}
     {targetBone {boneName=dyn_frill_frt_l}}
     {targetBone {boneName=thigh_twist_01_r}}
     {targetBone {boneName=thigh_twist_01_l}}
     {targetBone {boneName=dyn_frill_frt_r}}
     {targetBone {boneName=dyn_frill_frt}}
    ]
    constraintWeights [1,1,1,1,1,1]
   }
   n106 = "Constraint - strap_aim_02_l (PPPPPP)"(
    ComponentPose=@n105
    Alpha=1
    ConstraintWeights_0=0.0321
    ConstraintWeights_1=0.8622
    ConstraintWeights_2=0.0005
    ConstraintWeights_3=0.0005
    ConstraintWeights_4=0.0172
    ConstraintWeights_5=0.0876
   ) {
    alphaBoolBlend {blendOption=Linear}
    boneToModify {boneName=strap_aim_02_l}
    constraintSetup [
     {targetBone {boneName=dyn_pelvis_offset}}
     {targetBone {boneName=dyn_frill_frt_l}}
     {targetBone {boneName=thigh_twist_01_r}}
     {targetBone {boneName=thigh_twist_01_l}}
     {targetBone {boneName=dyn_frill_frt_r}}
     {targetBone {boneName=dyn_frill_frt}}
    ]
    constraintWeights [1,1,1,1,1,1]
   }
   n107 = "Constraint - strap_aim_03_l (PPPPPP)"(
    ComponentPose=@n106
    Alpha=1
    ConstraintWeights_0=0.001
    ConstraintWeights_1=0.836
    ConstraintWeights_2=0.0004
    ConstraintWeights_3=0.0004
    ConstraintWeights_4=0.0116
    ConstraintWeights_5=0.1505
   ) {
    alphaBoolBlend {blendOption=Linear}
    boneToModify {boneName=strap_aim_03_l}
    constraintSetup [
     {targetBone {boneName=dyn_pelvis_offset}}
     {targetBone {boneName=dyn_frill_frt_l}}
     {targetBone {boneName=thigh_twist_01_r}}
     {targetBone {boneName=thigh_twist_01_l}}
     {targetBone {boneName=dyn_frill_frt_r}}
     {targetBone {boneName=dyn_frill_frt}}
    ]
    constraintWeights [1,1,1,1,1,1]
   }
   n108 = "Constraint - strap_aim_04_l (PPPPPP)"(
    ComponentPose=@n107
    Alpha=1
    ConstraintWeights_0=0
    ConstraintWeights_1=0.777
    ConstraintWeights_2=0.0016
    ConstraintWeights_3=0.0005
    ConstraintWeights_4=0.0124
    ConstraintWeights_5=0.2085
   ) {
    alphaBoolBlend {blendOption=Linear}
    boneToModify {boneName=strap_aim_04_l}
    constraintSetup [
     {targetBone {boneName=dyn_pelvis_offset}}
     {targetBone {boneName=dyn_frill_frt_l}}
     {targetBone {boneName=thigh_twist_01_r}}
     {targetBone {boneName=thigh_twist_01_l}}
     {targetBone {boneName=dyn_frill_frt_r}}
     {targetBone {boneName=dyn_frill_frt}}
    ]
    constraintWeights [1,1,1,1,1,1]
   }
   n109 = "Constraint - strap_aim_01_r (PPPPPP)"(
    ComponentPose=@n108
    Alpha=1
    ConstraintWeights_0=0.724
    ConstraintWeights_1=0.0006
    ConstraintWeights_2=0.108
    ConstraintWeights_3=0.0026
    ConstraintWeights_4=0.1631
    ConstraintWeights_5=0.0017
   ) {
    alphaBoolBlend {blendOption=Linear}
    boneToModify {boneName=strap_aim_01_r}
    constraintSetup [
     {targetBone {boneName=dyn_pelvis_offset}}
     {targetBone {boneName=dyn_frill_frt_l}}
     {targetBone {boneName=thigh_twist_01_r}}
     {targetBone {boneName=thigh_twist_01_l}}
     {targetBone {boneName=dyn_frill_frt_r}}
     {targetBone {boneName=dyn_frill_frt}}
    ]
    constraintWeights [1,1,1,1,1,1]
   }
   n110 = "Constraint - strap_aim_02_r (PPPPPP)"(
    ComponentPose=@n109
    Alpha=1
    ConstraintWeights_0=0.0547
    ConstraintWeights_1=0.0585
    ConstraintWeights_2=0.0074
    ConstraintWeights_3=0.0052
    ConstraintWeights_4=0.7797
    ConstraintWeights_5=0.0945
   ) {
    alphaBoolBlend {blendOption=Linear}
    boneToModify {boneName=strap_aim_02_r}
    constraintSetup [
     {targetBone {boneName=dyn_pelvis_offset}}
     {targetBone {boneName=dyn_frill_frt_l}}
     {targetBone {boneName=thigh_twist_01_r}}
     {targetBone {boneName=thigh_twist_01_l}}
     {targetBone {boneName=dyn_frill_frt_r}}
     {targetBone {boneName=dyn_frill_frt}}
    ]
    constraintWeights [1,1,1,1,1,1]
   }
   n111 = "Constraint - strap_aim_03_r (PPPPPP)"(
    ComponentPose=@n110
    Alpha=1
    ConstraintWeights_0=0.0048
    ConstraintWeights_1=0.0812
    ConstraintWeights_2=0.0041
    ConstraintWeights_3=0.0036
    ConstraintWeights_4=0.7503
    ConstraintWeights_5=0.156
   ) {
    alphaBoolBlend {blendOption=Linear}
    boneToModify {boneName=strap_aim_03_r}
    constraintSetup [
     {targetBone {boneName=dyn_pelvis_offset}}
     {targetBone {boneName=dyn_frill_frt_l}}
     {targetBone {boneName=thigh_twist_01_r}}
     {targetBone {boneName=thigh_twist_01_l}}
     {targetBone {boneName=dyn_frill_frt_r}}
     {targetBone {boneName=dyn_frill_frt}}
    ]
    constraintWeights [1,1,1,1,1,1]
   }
   n112 = "Constraint - strap_aim_04_r (PPPPPP)"(
    ComponentPose=@n111
    Alpha=1
    ConstraintWeights_0=0
    ConstraintWeights_1=0.049
    ConstraintWeights_2=0.0022
    ConstraintWeights_3=0.0061
    ConstraintWeights_4=0.6967
    ConstraintWeights_5=0.246
   ) {
    alphaBoolBlend {blendOption=Linear}
    boneToModify {boneName=strap_aim_04_r}
    constraintSetup [
     {targetBone {boneName=dyn_pelvis_offset}}
     {targetBone {boneName=dyn_frill_frt_l}}
     {targetBone {boneName=thigh_twist_01_r}}
     {targetBone {boneName=thigh_twist_01_l}}
     {targetBone {boneName=dyn_frill_frt_r}}
     {targetBone {boneName=dyn_frill_frt}}
    ]
    constraintWeights [1,1,1,1,1,1]
   }
   n113 = "Constraint - strap_aim_01_c (PPPPPP)"(
    ComponentPose=@n112
    Alpha=1
    ConstraintWeights_0=0.8358
    ConstraintWeights_1=0.0083
    ConstraintWeights_2=0.0339
    ConstraintWeights_3=0.0928
    ConstraintWeights_4=0.0089
    ConstraintWeights_5=0.0203
   ) {
    alphaBoolBlend {blendOption=Linear}
    boneToModify {boneName=strap_aim_01_c}
    constraintSetup [
     {targetBone {boneName=dyn_pelvis_offset}}
     {targetBone {boneName=dyn_frill_frt_l}}
     {targetBone {boneName=thigh_twist_01_r}}
     {targetBone {boneName=thigh_twist_01_l}}
     {targetBone {boneName=dyn_frill_frt_r}}
     {targetBone {boneName=dyn_frill_frt}}
    ]
    constraintWeights [1,1,1,1,1,1]
   }
   n114 = "Constraint - strap_aim_02_c (PPPPPP)"(
    ComponentPose=@n113
    Alpha=1
    ConstraintWeights_0=0.3241
    ConstraintWeights_1=0.2855
    ConstraintWeights_2=0.0205
    ConstraintWeights_3=0.0216
    ConstraintWeights_4=0.2548
    ConstraintWeights_5=0.0936
   ) {
    alphaBoolBlend {blendOption=Linear}
    boneToModify {boneName=strap_aim_02_c}
    constraintSetup [
     {targetBone {boneName=dyn_pelvis_offset}}
     {targetBone {boneName=dyn_frill_frt_l}}
     {targetBone {boneName=thigh_twist_01_r}}
     {targetBone {boneName=thigh_twist_01_l}}
     {targetBone {boneName=dyn_frill_frt_r}}
     {targetBone {boneName=dyn_frill_frt}}
    ]
    constraintWeights [1,1,1,1,1,1]
   }
   n115 = "Constraint - strap_aim_03_c (PPPPPP)"(
    ComponentPose=@n114
    Alpha=1
    ConstraintWeights_0=0.0562
    ConstraintWeights_1=0.4083
    ConstraintWeights_2=0.0097
    ConstraintWeights_3=0.0084
    ConstraintWeights_4=0.3701
    ConstraintWeights_5=0.1472
   ) {
    alphaBoolBlend {blendOption=Linear}
    boneToModify {boneName=strap_aim_03_c}
    constraintSetup [
     {targetBone {boneName=dyn_pelvis_offset}}
     {targetBone {boneName=dyn_frill_frt_l}}
     {targetBone {boneName=thigh_twist_01_r}}
     {targetBone {boneName=thigh_twist_01_l}}
     {targetBone {boneName=dyn_frill_frt_r}}
     {targetBone {boneName=dyn_frill_frt}}
    ]
    constraintWeights [1,1,1,1,1,1]
   }
   n116 = "Constraint - strap_aim_04_c (PPPPPP)"(
    ComponentPose=@n115
    Alpha=1
    ConstraintWeights_0=0.0028
    ConstraintWeights_1=0.3777
    ConstraintWeights_2=0.0127
    ConstraintWeights_3=0.0097
    ConstraintWeights_4=0.3214
    ConstraintWeights_5=0.2757
   ) {
    alphaBoolBlend {blendOption=Linear}
    boneToModify {boneName=strap_aim_04_c}
    constraintSetup [
     {targetBone {boneName=dyn_pelvis_offset}}
     {targetBone {boneName=dyn_frill_frt_l}}
     {targetBone {boneName=thigh_twist_01_r}}
     {targetBone {boneName=thigh_twist_01_l}}
     {targetBone {boneName=dyn_frill_frt_r}}
     {targetBone {boneName=dyn_frill_frt}}
    ]
    constraintWeights [1,1,1,1,1,1]
   }
   n117 = RigidBody(ComponentPose=@n116, OverrideWorldGravity="0,0,-980") {
    bClampLinearTranslationLimitToRefPose=true
    bOverrideWorldGravity=true
    overridePhysicsAsset=/Game/Characters/Player/Female/Medium/Bodies/F_MED_Ancient_Gladiator/Meshes/physics/F_MED_Ancient_Gladiator_Straps_Skirt
    simulationSpace=BaseBoneSpace
    alphaBoolBlend {blendOption=Linear}
    baseBoneRef {boneName=root}
    componentLinearAccScale=<0.5 0.5 0.35>
    componentLinearVelScale=<0.5 0.5 0.35>
    overrideWorldGravity {z=-980}
   }
   n118 = "Transform (Modify) Bone - Bone: dyn_parent_frill_frtA_01_r"(ComponentPose=@n117, Alpha=Get legUp_R() -> (legUp_R:double), Translation="0,1,0", Rotation="-3,10,-22", Scale="1,1,1") {
    rotationMode=BMM_Additive
    rotationSpace=BCS_BoneSpace
    scaleSpace=BCS_BoneSpace
    translationMode=BMM_Additive
    translationSpace=BCS_BoneSpace
    alphaBoolBlend {blendOption=Linear}
    alphaScaleBiasClamp {bInterpResult=true, interpSpeedDecreasing=25, interpSpeedIncreasing=25}
    boneToModify {boneName=dyn_parent_frill_frtA_01_r}
    rotation=<rot -3 10 -22>
    translation {y=1}
   }
   n119 = "Transform (Modify) Bone - Bone: dyn_parent_frill_frtA_01_l"(ComponentPose=@n118, Alpha=Get legUp_L() -> (legUp_L:double), Translation="0,5,0", Rotation="3,26,28", Scale="1,1,1") {
    rotationMode=BMM_Additive
    rotationSpace=BCS_BoneSpace
    scaleSpace=BCS_BoneSpace
    translationMode=BMM_Additive
    translationSpace=BCS_BoneSpace
    alphaBoolBlend {blendOption=Linear}
    alphaScaleBiasClamp {bInterpResult=true, interpSpeedDecreasing=25, interpSpeedIncreasing=25}
    boneToModify {boneName=dyn_parent_frill_frtA_01_l}
    rotation=<rot 3 26 28>
    translation {y=5}
   }
   Save cached pose 'Standing_Cache'(Pose=Component To Local(ComponentPose=@n119)) {cacheName=Standing_Cache}
   Reroute Node(InputPin:double=Get rban_straps_look_alpha()) -> (OutputPin:double)
   Get stand_alpha() -> (stand_alpha:double)
   Get legUp_Frill_L() -> (legUp_Frill_L:double)
   Get legUp_L()
  ]}
 EventGraph {nodes [
   Event OnEndTargeting() {is=title K2Node_Event}
   > Set isAiming(isAiming:double="To Float (Boolean)"(InBool=Get bIsTargeting())) -> (Output_Get:double)
   Event OnBeginSkydiving()
   n1 = "To Float (Boolean)"(InBool=Get bIsSkyDiving())
   > Set isSkydiving(isSkydiving:double=@n1) -> (Output_Get:double)
   > Set isSkydivingINV(isSkydivingINV:double=float - float(B=@n1) {is=title K2Node_CallFunction}) -> (Output_Get:double)
   > Set skydive_bool(skydive_bool:bool=true) -> (Output_Get:bool)
   Event OnEndPlayingEmote() {is=title K2Node_Event}
   > n2 = Set emote_active(emote_active:bool=false) -> (Output_Get:bool)
   Delay(execute=[@n3.then, @n2.then], Duration=0.1)
   > ResetDynamics(InTeleportType=TeleportPhysics)
   > ResetDynamics(InTeleportType=ResetPhysics)
   > Set crouch_alpha(crouch_alpha:double=1) -> (Output_Get:double)
   > Set crouch_bool(crouch_bool:bool=false) -> (Output_Get:bool)
   > Set crouch_float(crouch_float:double=1) -> (Output_Get:double)
   Event OnBeginTargeting() {is=title K2Node_Event}
   > Set isAiming(isAiming="To Float (Boolean)"(InBool=Get bIsTargeting()))
   Event OnEndCrouching()
   > n3 = Set isCrouching(isCrouching:double="To Float (Boolean)"(InBool=Get bIsCrouching())) -> (Output_Get:double)
   Event OnEndSkydiving()
   n4 = "To Float (Boolean)"(InBool=Get bIsSkyDiving())
   > Set isSkydiving(isSkydiving=@n4)
   > Set isSkydivingINV(isSkydivingINV=float - float(B=@n4))
   > Set skydive_bool(skydive_bool=false)
   Event BlueprintInitializeAnimation()
   > Set Body_Mesh(Body_Mesh=GetPartSkeletalMeshComponent())
   Event OnBeginCrouching()
   > Set isCrouching(isCrouching="To Float (Boolean)"(InBool=Get bIsCrouching()))
   > Set crouch_alpha(crouch_alpha=0.7)
   > Set crouch_bool(crouch_bool=true)
   > Set crouch_float(crouch_float=1)
   Event OnBeginPlayingEmote() {is=title K2Node_Event}
   > Set emote_active(emote_active=true)
   n5 = Event BlueprintUpdateAnimation()
   > Get!(execute:Exec) -> (then:Exec, else:Exec, Body_Mesh:Skeletal Mesh Component Object Reference)
   > Set delta_time(delta_time:double=@n5.DeltaTimeX) -> (Output_Get:double)
   > n6 = Sequence!(execute) -> (then_0, then_1, then_2:Exec, then_3:Exec, then_4:Exec, then_5:Exec, then_6:Exec, then_7:Exec, then_8:Exec)
   > Set isSprinting(isSprinting:double="To Float (Boolean)"(InBool=Get bIsSprinting())) -> (Output_Get:double)
   > Set isGliding(isGliding:double="To Float (Boolean)"(InBool=Get bIsParachuteOpened())) -> (Output_Get:double)
   n7 = Branch(execute=@n6.then_5, Condition=Get bIsInFrontEnd())
   > n8 = Branch(Condition=Get bIsPlayingEmote())
   > Set fe_alpha(fe_alpha:double=0.2) -> (Output_Get:double)
   > Set fe_pad_alpha(fe_pad_alpha:double=1) -> (Output_Get:double)
   Set fe_alpha(execute=@n8.else, fe_alpha=1)
   > Set fe_pad_alpha(fe_pad_alpha=0.25)
   Set fe_alpha(execute=@n7.else, fe_alpha=0)
   n9 = Branch(execute=@n6.then_6, Condition=Get bRigidBodyEnabled())
   > Set rban_straps_alpha(rban_straps_alpha:double=1) -> (Output_Get:double)
   > Set rban_straps_look_alpha(rban_straps_look_alpha:double=0) -> (Output_Get:double)
   Set rban_straps_alpha(execute=@n9.else, rban_straps_alpha=0)
   > Set legUp_Strap_R(legUp_Strap_R:double=0) -> (Output_Get:double)
   > Set fe_alpha(fe_alpha=0)
   > Set rban_straps_look_alpha(rban_straps_look_alpha=1)
   n10 = Gravity Override State Function DEPRECATED(
    execute=@n6.then_8
    CharacterPartAnimInstance=Self-Reference()
    GameTimeInSeconds=GetGameTimeInSeconds()
    GravityOverrideParams=Get Gravity Override Params()
   )
   > Set GravityOverride(GravityOverride=@n10)
   Set clav_up_r(
    execute=@n6.then_4
    clav_up_r:double=GetAlphaFromReferenceTransformDelta(
     SkeletalMeshComponent=Get Body_Mesh()
     BoneName=upperarm_r
     BaseName=clavicle_r
     DeltaComponentType=QuaternionTwist
     InputRange=(X=0,Y=-120)
     TwistAxis="0,0,1"
    )
   ) -> (Output_Get:double)
   n11 = Branch(execute=@n6.then_2, Condition="float > float"(A=Get PawnSpeed2D(), B=150) {is=title K2Node_CallFunction})
   > Set cloth_rear_alpha(cloth_rear_alpha:double=1) -> (Output_Get:double)
   > Set hip_chains_alpha(hip_chains_alpha:double=0.55) -> (Output_Get:double)
   > Set side_straps_alpha(side_straps_alpha:double=1) -> (Output_Get:double)
   > Set stand_alpha(stand_alpha:double=0) -> (Output_Get:double)
   > Set side_straps_crouch(side_straps_crouch:double=0) -> (Output_Get:double)
   Set cloth_rear_alpha(execute=@n11.else, cloth_rear_alpha=0.75)
   > Set hip_chains_alpha(hip_chains_alpha=0.25)
   > Set side_straps_alpha(side_straps_alpha=1)
   > Set stand_alpha(stand_alpha=1)
   > Set side_straps_crouch(side_straps_crouch=1)
   n12 = Get Body_Mesh()
   Set legUp_L(
    execute=@n6.then_1
    legUp_L:double=GetAlphaFromReferenceTransformDelta(
     SkeletalMeshComponent=@n12
     BoneName=thigh_l
     BaseName=pelvis
     DeltaComponentType=QuaternionTwist
     InputRange=(X=15,Y=100)
     TwistAxis="0,0,1"
    )
   ) -> (Output_Get:double)
   > Set legUp_Strap_L(legUp_Strap_L:double=GetAlphaFromReferenceTransformDelta(SkeletalMeshComponent=@n12, BoneName=thigh_l, BaseName=pelvis, DeltaComponentType=QuaternionTwist, InputRange=(X=-5,Y=95), TwistAxis="0,0,1")) -> (Output_Get:double)
   > Set legUp_R(legUp_R:double=GetAlphaFromReferenceTransformDelta(
     SkeletalMeshComponent=Get Body_Mesh()
     BoneName=thigh_r
     BaseName=pelvis
     DeltaComponentType=QuaternionTwist
     InputRange=(X=-5,Y=70)
     TwistAxis="0,0,1"
    )) -> (Output_Get:double)
   n13 = Get Body_Mesh()
   > Set legUp_Strap_R(legUp_Strap_R=GetAlphaFromReferenceTransformDelta(SkeletalMeshComponent=@n13, BoneName=thigh_r, BaseName=pelvis, DeltaComponentType=QuaternionTwist, InputRange=(X=-5,Y=95), TwistAxis="0,0,1"))
   > Set legUp_Frill_R(legUp_Frill_R:double=GetAlphaFromReferenceTransformDelta(
     SkeletalMeshComponent=@n13
     BoneName=thigh_r
     BaseName=pelvis
     DeltaComponentType=QuaternionTwist
     InputRange=(X=30,Y=105)
     TwistAxis="0,0,1"
    )) -> (Output_Get:double)
   n14 = Get Body_Mesh()
   > Set legOut_R(legOut_R:double=GetAlphaFromReferenceTransformDelta(SkeletalMeshComponent=@n14, BoneName=thigh_r, BaseName=pelvis, DeltaComponentType=EulerY, InputRange=(X=0,Y=40), TwistAxis="0,1,0")) -> (Output_Get:double)
   > Set legOut_L(legOut_L:double=GetAlphaFromReferenceTransformDelta(
     SkeletalMeshComponent=@n14
     BoneName=thigh_l
     BaseName=pelvis
     DeltaComponentType=EulerY
     InputRange=(X=0,Y=40)
     OutputRange=(X=0,Y=0.75)
     TwistAxis="0,1,0"
    )) -> (Output_Get:double)
   > Set legUp_Frill_L(legUp_Frill_L:double=GetAlphaFromReferenceTransformDelta(
     SkeletalMeshComponent=@n14
     BoneName=thigh_l
     BaseName=pelvis
     DeltaComponentType=QuaternionTwist
     InputRange=(X=30,Y=105)
     TwistAxis="0,0,1"
    )) -> (Output_Get:double)
   > Branch(Condition=Get bIsJumping())
   n15 = Get legUp_Strap_L()
   > Set legUp_Strap_L(legUp_Strap_L=FInterpTo(Current=@n15, Target=float * float(A=@n15, B=0.1), DeltaTime=Get delta_time() -> (delta_time:double), InterpSpeed=0.1))
   > Set hip_chains_alpha(hip_chains_alpha=0.1)
   n16 = Branch(execute=@n6.then_3, Condition=Get bIsCrouching())
   > Set legUp_L(legUp_L=1)
   > Set legUp_R(legUp_R=1)
   > Set sub_leg_l(sub_leg_l:double=0) -> (Output_Get:double)
   > Set sub_leg_r(sub_leg_r:double=0) -> (Output_Get:double)
   > Set stand_alpha(stand_alpha=0)
   Set sub_leg_l(execute=@n16.else, sub_leg_l=1)
   > Set sub_leg_r(sub_leg_r=1)
   n17 = Gravity Override State Function DEPRECATED(
    execute=@n6.then_7
    CharacterPartAnimInstance=Self-Reference()
    GameTimeInSeconds=GetGameTimeInSeconds()
    GravityOverrideParams=Get Gravity Override Params()
   )
   > n18 = Set gravityVector(gravityVector:Vector=@n17) -> (Output_Get:Vector)
   GetOwnerSkeletalMeshComponent()
   DrawVectorFromJointFunction!(
    execute
    SkelMeshComponent:Scene Component Object Reference
    JointName:Name=head
    LineEnd:Vector=@n18
    JointOffset:Vector="0, 0, 0"
    LineColor:Linear Color Structure=(R=1,G=0,B=0,A=1)
    Thickness:double=2
    Foreground:EDrawDebugSceneDepthPriorityGroup Enum=Foreground
   ) -> (then)
   "Max (Float)"(
    A="Max (Float)"(
     A=0
     B=GetAlphaFromReferenceTransformDelta(BoneName=thigh_r, BaseName=pelvis, DeltaComponentType=EulerY, InputRange=(X=0,Y=60), OutputRange=(X=0,Y=0.7), TwistAxis="0,1,0")
    )
    B=GetAlphaFromReferenceTransformDelta(BoneName=thigh_r, BaseName=pelvis, DeltaComponentType=EulerX, InputRange=(X=0,Y=150), TwistAxis="1,0,0")
   )
   Set Anim_Inst(Anim_Inst:Anim Instance Object Reference=GetAnimInstance()) -> (Output_Get:Anim Instance Object Reference)
  ]}
]
variables [
 Anim_Inst {type=AnimInstance, default=null}
 AnimInst {type=AnimInstance, default=null}
 ar_firingPosition {type=boolean}
 Body_Mesh {type=SkeletalMeshComponent}
 clav_up_r {type=number}
 cloth_rear_alpha {type=number}
 crouch_alpha {type=number, default=1}
 crouch_bool {type=boolean}
 crouch_float {type=number}
 delta_time {type=number}
 emote_active {type=boolean}
 fe_alpha {type=number}
 fe_pad_alpha {type=number}
 Gravity Override Params {
  type=GravityOverrideParamsStruct
  default {
   skydive_DivingDown {gravityOverride=<0 0 0>}
   skydive_FloatingDown {gravityOverride=<0 0 0>}
   skydive_GoingUp {gravityOverride=<0 -300 -100>}
   skydive_Parachuting {gravityOverride=<0 0 -920>}
  }
 }
 GravityOverride {type=Vector}
 gravityVector {type=Vector}
 hip_chains_alpha {type=number}
 isAiming {type=number}
 isCrouching {type=number}
 isGliding {type=number}
 isSkydiving {type=number}
 isSkydivingINV {type=number}
 isSprinting {type=number}
 leg_rot_r {type=number}
 legOut_L {type=number}
 legOut_R {type=number}
 legUp_Frill_L {type=number}
 legUp_Frill_R {type=number}
 legUp_L {type=number}
 legUp_R {type=number}
 legUp_Strap_L {type=number}
 legUp_Strap_R {type=number}
 legUpMin_L {type=number}
 Pawn {type=FortPawn, default=null}
 pickaxe_active {type=boolean}
 pistol_active {type=boolean}
 rban_straps_alpha {type=number}
 rban_straps_look_alpha {type=number}
 rifle_active {type=boolean}
 shldr_l_alpha {type=number}
 shldr_l_roll_alpha {type=number}
 shoulder_pad_l_aim_rotator {type=Rotator}
 shoulder_pad_l_aim_trans {type=Vector}
 shoulder_pad_l_roll_rotator {type=Rotator}
 shoulder_pad_l_roll_trans {type=Vector}
 shoulder_pad_l_rotator {type=Rotator}
 shoulder_pad_l_trans {type=Vector}
 shoulder_pad_r_aim_rotator {type="array<Rotator>", default []}
 shoulder_pad_r_aim_trans {type=Vector}
 shoulder_pad_r_pack_trans {type=Vector}
 shoulder_pad_r_rotator {type=Rotator}
 shoulder_pad_r_trans {type=Vector}
 side_straps_alpha {type=number}
 side_straps_crouch {type=number}
 skydive_bool {type=boolean}
 stand_alpha {type=number}
 sub_leg_l {type=number, default=1}
 sub_leg_r {type=number, default=1}
]
