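# Physics asset for the dynamic ponytail on the female medium Ancient Gladiator character's hair mesh.
# A capsule chain runs from dyn_parent_ponytail at the head through five ponytail bones, with angular limits tight at the root and much wider swing toward the tip.
# Capsules on the head, neck, clavicles, spine and pelvis provide body shapes around the ponytail, linked by constraints with free swing and twist.
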PhysicsAsset /Game/Characters/Player/Female/Medium/Bodies/F_MED_Ancient_Gladiator/Meshes/Parts/F_MED_Ancient_Gladiator_Hair_Dyn
bodies {
 clavicle_l [{shapeType=Capsule, length=6.2188, radius=6.7188, shapeName=clavicle_l_capsule, center=<4.2524 0.0001 -1.5232>, rotation=<rot 84.6895 0 0>}]
 clavicle_r [{shapeType=Capsule, length=6.2188, radius=6.7188, shapeName=clavicle_r_capsule, center=<-4.2524 -0.0001 1.5232>, rotation=<rot 84.6895 0 0>}]
 dyn_parent_ponytail [{shapeType=Capsule, length=0.505, radius=0.505, rotation=<rot 0 -90 0>}]
 dyn_ponytail_02 [{shapeType=Capsule, length=2, radius=1.6562, center=<3.534 0 0>, rotation=<rot 90 0 0>}]
 dyn_ponytail_03 [{shapeType=Capsule, length=3, radius=3.5, center=<3.9425 0 0>, rotation=<rot 90 0 0>}]
 dyn_ponytail_04 [{shapeType=Capsule, length=0, radius=4.5938, center=<5.038 0 0.0458>, rotation=<rot 88.8739 -179.9999 -179.9999>}]
 dyn_ponytail_05 [{shapeType=Capsule, length=5.3438, radius=3.375, center=<7.8753 0 0>, rotation=<rot 90 0 0>}]
 dyn_ponytail_MSTR [{shapeType=Capsule, length=2.3261, radius=1.2027, center=<2.417 0.0001 0>, rotation=<rot 90 0 0>}]
 head [{shapeType=Capsule, length=7.4688, radius=5.4375, center=<0.5445 3.4826 0>, rotation=<rot 90 0 0>}]
 neck_01 [{shapeType=Capsule, length=6.6875, radius=4.5, center=<2.6233 1.1555 0.0002>, rotation=<rot 90 0 0>}]
 pelvis [{shapeType=Capsule, length=8.1562, radius=12.1562, center=<-2.3989 -0.4861 0>}]
 spine_01 [{shapeType=Capsule, length=7.7188, radius=6.8438, center=<2.6802 1.5656 0>}]
 spine_03 [{shapeType=Capsule, length=2.7188, radius=9.4062, center=<1.3445 3.246 0>}]
 spine_04 [{shapeType=Capsule, length=5.125, radius=10.0625, center=<1.3926 2.539 0>}]
 spine_05 [
  {shapeType=Capsule, length=8, radius=9.1875, center=<1.148 2.0978 0>, rotation=<rot 90 9.9992 0>}
  {shapeType=Capsule, length=7.625, radius=7.25, center=<8.5862 3.1041 -0.0001>}
 ]
}
constraints [
 {bone1Name=dyn_ponytail_MSTR, bone2Name=dyn_parent_ponytail, swing1LimitDegrees=2, swing2LimitDegrees=5, twistLimitDegrees=1}
 {bone1Name=dyn_ponytail_02, bone2Name=dyn_ponytail_MSTR, swing1LimitDegrees=5, swing2LimitDegrees=5, twistLimitDegrees=5}
 {bone1Name=dyn_ponytail_03, bone2Name=dyn_ponytail_02, swing1LimitDegrees=40, swing2LimitDegrees=45, twistLimitDegrees=5}
 {bone1Name=dyn_ponytail_04, bone2Name=dyn_ponytail_03, swing1LimitDegrees=45, swing2LimitDegrees=80, twistLimitDegrees=5}
 {bone1Name=dyn_ponytail_05, bone2Name=dyn_ponytail_04, swing1LimitDegrees=45, swing2LimitDegrees=80, twistLimitDegrees=5}
 {
  bone1Name=head
  bone2Name=neck_01
  swing1LimitDegrees=45
  swing1Motion=Free
  swing2LimitDegrees=45
  swing2Motion=Free
  twistLimitDegrees=45
  twistMotion=Free
 }
 {
  bone1Name=neck_01
  bone2Name=spine_05
  swing1LimitDegrees=45
  swing1Motion=Free
  swing2LimitDegrees=45
  swing2Motion=Free
  twistLimitDegrees=45
  twistMotion=Free
 }
 {
  bone1Name=spine_05
  bone2Name=spine_04
  swing1LimitDegrees=45
  swing1Motion=Free
  swing2LimitDegrees=45
  swing2Motion=Free
  twistLimitDegrees=45
  twistMotion=Free
 }
 {
  bone1Name=spine_04
  bone2Name=spine_03
  swing1LimitDegrees=45
  swing1Motion=Free
  swing2LimitDegrees=45
  swing2Motion=Free
  twistLimitDegrees=45
  twistMotion=Free
 }
 {
  bone1Name=spine_03
  bone2Name=spine_01
  swing1LimitDegrees=45
  swing1Motion=Free
  swing2LimitDegrees=45
  swing2Motion=Free
  twistLimitDegrees=45
  twistMotion=Free
 }
 {
  bone1Name=spine_01
  bone2Name=pelvis
  swing1LimitDegrees=45
  swing1Motion=Free
  swing2LimitDegrees=45
  swing2Motion=Free
  twistLimitDegrees=45
  twistMotion=Free
 }
 {
  bone1Name=dyn_parent_ponytail
  bone2Name=head
  swing1LimitDegrees=45
  swing1Motion=Free
  swing2LimitDegrees=45
  swing2Motion=Free
  twistLimitDegrees=45
  twistMotion=Free
 }
 {
  bone1Name=clavicle_l
  bone2Name=spine_05
  swing1LimitDegrees=45
  swing1Motion=Free
  swing2LimitDegrees=45
  swing2Motion=Free
  twistLimitDegrees=45
  twistMotion=Free
 }
 {
  bone1Name=clavicle_r
  bone2Name=spine_05
  swing1LimitDegrees=45
  swing1Motion=Free
  swing2LimitDegrees=45
  swing2Motion=Free
  twistLimitDegrees=45
  twistMotion=Free
 }
]
