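# Physics asset for the hair mesh of the female medium Ancient Gladiator player character, defining bodies for hair motion.
# Capsules cover the front, left-side and tip hair bones, plus one larger capsule on the head.
# Most hair bodies are constrained directly to the head, while frtA, frtB and SideC form two-link chains rooted there.
# Narrow swing limits and very tight twist limits keep each lock close to the modeled style.
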PhysicsAsset /Game/Characters/Player/Female/Medium/Bodies/F_MED_Ancient_Gladiator/Meshes/Parts/F_MED_Ancient_Gladiator_Hair_Physics
bodies {
 dyn_hair_SideA_01_l [{shapeType=Capsule, length=2.25, radius=2.5, center=<1 0 0>, rotation=<rot 90 0 0>}]
 dyn_hair_SideB_01_l [{shapeType=Capsule, length=2, radius=2.25, center=<0.7 0 0>, rotation=<rot 90 0 0>}]
 dyn_hair_SideC_01_l [{shapeType=Capsule, length=2.5, radius=3.25, center=<1 0 0>, rotation=<rot 90 0 0>}]
 dyn_hair_SideC_02_l [{shapeType=Capsule, length=1.75, radius=2.75, center=<1 0 0>, rotation=<rot 90 0 0>}]
 dyn_hair_SideD_01_l [{shapeType=Capsule, length=1.25, radius=1.25, center=<0.5 0 0>, rotation=<rot 90 0 0>}]
 dyn_hair_frtA_01 [{shapeType=Capsule, length=1.75, radius=2.25, center=<1 0 0>, rotation=<rot 90 0 0>}]
 dyn_hair_frtA_02 [{shapeType=Capsule, length=2, radius=1.5, center=<1 0 0>, rotation=<rot 90 0 0>}]
 dyn_hair_frtB_01 [{shapeType=Capsule, length=2, radius=2.25, center=<0.5 0 0>, rotation=<rot 83.8402 -171.2099 -179.9999>}]
 dyn_hair_frtB_02 [{shapeType=Capsule, length=2, radius=2, center=<0.75 0 0>, rotation=<rot 83.8391 -171.2098 -179.9999>}]
 dyn_hair_tipA [{shapeType=Capsule, length=2.7269, radius=1.5246, center=<0.5 0 0>, rotation=<rot 90 0 0>}]
 dyn_hair_tipB [{shapeType=Capsule, length=1.9435, radius=1.1918, center=<0.5427 -0.5523 0>, rotation=<rot 90 0 47.5514>}]
 dyn_hair_tipC [{shapeType=Capsule, length=1.5, radius=1.005, center=<1 0 0>, rotation=<rot 90 0 0>}]
 dyn_hair_tipD [{shapeType=Capsule, length=2.0619, radius=1.75, center=<1 0 0>, rotation=<rot 90 0 0>}]
 head [{shapeType=Capsule, length=6.0717, radius=7.3858, center=<1.2697 0.6299 -0.2853>, rotation=<rot 0.4958 -69.6541 -88.7278>}]
}
constraints [
 {bone1Name=dyn_hair_tipA, bone2Name=head, swing1LimitDegrees=30, swing2LimitDegrees=30, twistLimitDegrees=3}
 {bone1Name=dyn_hair_tipB, bone2Name=head, swing1LimitDegrees=15, swing2LimitDegrees=15, twistLimitDegrees=3}
 {bone1Name=dyn_hair_tipC, bone2Name=head, swing1LimitDegrees=25, swing2LimitDegrees=35, twistLimitDegrees=3}
 {bone1Name=dyn_hair_tipD, bone2Name=head, swing1LimitDegrees=30, swing2LimitDegrees=20, twistLimitDegrees=3}
 {bone1Name=dyn_hair_frtA_01, bone2Name=head, swing1LimitDegrees=10, swing2LimitDegrees=10, twistLimitDegrees=5}
 {bone1Name=dyn_hair_frtA_02, bone2Name=dyn_hair_frtA_01, swing1LimitDegrees=20, swing2LimitDegrees=20, twistLimitDegrees=5}
 {bone1Name=dyn_hair_frtB_01, bone2Name=head, swing1LimitDegrees=20, swing2LimitDegrees=20, twistLimitDegrees=3}
 {bone1Name=dyn_hair_frtB_02, bone2Name=dyn_hair_frtB_01, swing1LimitDegrees=20, swing2LimitDegrees=20, twistLimitDegrees=3}
 {bone1Name=dyn_hair_SideA_01_l, bone2Name=head, swing1LimitDegrees=15, swing2LimitDegrees=15, twistLimitDegrees=3}
 {bone1Name=dyn_hair_SideB_01_l, bone2Name=head, swing1LimitDegrees=15, swing2LimitDegrees=15, twistLimitDegrees=3}
 {bone1Name=dyn_hair_SideC_01_l, bone2Name=head, swing1LimitDegrees=10, swing2LimitDegrees=15, twistLimitDegrees=5}
 {bone1Name=dyn_hair_SideC_02_l, bone2Name=dyn_hair_SideC_01_l, swing1LimitDegrees=10, swing2LimitDegrees=15, twistLimitDegrees=5}
 {bone1Name=dyn_hair_SideD_01_l, bone2Name=head, swing1LimitDegrees=25, swing2LimitDegrees=25, twistLimitDegrees=5}
]
