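# Animation Blueprint for the Ancient Gladiator female medium character's hair, adding secondary motion to the ponytail.
# It copies the attached parent's pose, applies HeadDelta-driven bone offsets and a RigidBody simulation, then caches the result.
# A backpack bool selects between weighted head, neck, spine and clavicle constraints with AnimDynamics tips, or a second RigidBody pass.
# The event graph updates GravityVector through a deprecated gravity-override function.
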AnimBlueprint /Game/Characters/Player/Female/Medium/Bodies/F_MED_Ancient_Gladiator/Meshes/Parts/F_MED_Ancient_Gladiator_Hair_AnimBP
target_skeleton=&/F_MED_Ancient_Gladiator_Hair_Skeleton
graphs [
 AnimGraph {nodes [
   n1 = "Constraint - dyn_con_hair_tail_01 (PP)"!(
    ComponentPose=Local To Component(LocalPose=Use cached pose 'baseCache'())
    Alpha=1
    ConstraintWeights_0=0.9526
    ConstraintWeights_1:float=0.0474
   ) -> (Pose) {
    alphaBoolBlend {blendOption=Linear}
    boneToModify {boneName=dyn_con_hair_tail_01}
    constraintSetup [{targetBone {boneName=head}}, {targetBone {boneName=neck_02}}]
    constraintWeights [1,1]
   }
   n2 = "Constraint - dyn_con_hair_tail_02 (PP)"!(ComponentPose=@n1, Alpha=1, ConstraintWeights_0=0.9526, ConstraintWeights_1:float=0.0474) -> (Pose) {
    alphaBoolBlend {blendOption=Linear}
    boneToModify {boneName=dyn_con_hair_tail_02}
    constraintSetup [{targetBone {boneName=head}}, {targetBone {boneName=neck_02}}]
    constraintWeights [1,1]
   }
   n3 = "Constraint - dyn_con_hair_tail_03 (PP)"!(ComponentPose=@n2, Alpha=1, ConstraintWeights_0=0.8651, ConstraintWeights_1:float=0.1349) -> (Pose) {
    alphaBoolBlend {blendOption=Linear}
    boneToModify {boneName=dyn_con_hair_tail_03}
    constraintSetup [{targetBone {boneName=head}}, {targetBone {boneName=neck_02}}]
    constraintWeights [1,1]
   }
   n4 = "Constraint - dyn_con_hair_tail_04 (PPP)"!(ComponentPose=@n3, Alpha=1, ConstraintWeights_0=0.5708, ConstraintWeights_1:float=0.4078, ConstraintWeights_2:float=0.0214) -> (Pose) {
    alphaBoolBlend {blendOption=Linear}
    boneToModify {boneName=dyn_con_hair_tail_04}
    constraintSetup [{targetBone {boneName=head}}, {targetBone {boneName=neck_02}}, {targetBone {boneName=neck_01}}]
    constraintWeights [1,1,1]
   }
   n5 = "Constraint - dyn_con_hair_tail_05 (PP)"!(ComponentPose=@n4, Alpha=1, ConstraintWeights_0=0.73, ConstraintWeights_1:float=0.27) -> (Pose) {
    alphaBoolBlend {blendOption=Linear}
    boneToModify {boneName=dyn_con_hair_tail_05}
    constraintSetup [{targetBone {boneName=spine_05}}, {targetBone {boneName=clavicle_l}}]
    constraintWeights [1,1]
   }
   n6 = "Constraint - dyn_con_hair_tail_06 (PPP)"!(ComponentPose=@n5, Alpha=1, ConstraintWeights_0=0.6591, ConstraintWeights_1:float=0.2295, ConstraintWeights_2:float=0.1114) -> (Pose) {
    alphaBoolBlend {blendOption=Linear}
    boneToModify {boneName=dyn_con_hair_tail_06}
    constraintSetup [{targetBone {boneName=spine_05}}, {targetBone {boneName=spine_04}}, {targetBone {boneName=clavicle_l}}]
    constraintWeights [1,1,1]
   }
   n7 = "Constraint - dyn_con_hair_tail_07 (PP)"!(ComponentPose=@n6, Alpha=1, ConstraintWeights_0=0.805, ConstraintWeights_1:float=0.195) -> (Pose) {
    alphaBoolBlend {blendOption=Linear}
    boneToModify {boneName=dyn_con_hair_tail_07}
    constraintSetup [{targetBone {boneName=spine_04}}, {targetBone {boneName=spine_05}}]
    constraintWeights [1,1]
   }
   n8 = "Constraint - dyn_parent_ponytail (P)"(ComponentPose=@n7, Alpha=1, ConstraintWeights_0=1) {
    alphaBoolBlend {blendOption=Linear}
    boneToModify {boneName=dyn_parent_ponytail}
    constraintSetup [{offsetOption=None, targetBone {boneName=dyn_con_hair_tail_01}}]
    constraintWeights [1]
   }
   n9 = "Constraint - dyn_ponytail_MSTR (P)"(ComponentPose=@n8, Alpha=1, ConstraintWeights_0=1) {
    alphaBoolBlend {blendOption=Linear}
    boneToModify {boneName=dyn_ponytail_MSTR}
    constraintSetup [{offsetOption=None, targetBone {boneName=dyn_con_hair_tail_02}}]
    constraintWeights [1]
   }
   n10 = "Constraint - dyn_ponytail_01 (P)"(ComponentPose=@n9, Alpha=1, ConstraintWeights_0=1) {
    alphaBoolBlend {blendOption=Linear}
    boneToModify {boneName=dyn_ponytail_01}
    constraintSetup [{offsetOption=None, targetBone {boneName=dyn_con_hair_tail_03}}]
    constraintWeights [1]
   }
   n11 = "Constraint - dyn_ponytail_02 (P)"(ComponentPose=@n10, Alpha=1, ConstraintWeights_0=1) {
    alphaBoolBlend {blendOption=Linear}
    boneToModify {boneName=dyn_ponytail_02}
    constraintSetup [{offsetOption=None, targetBone {boneName=dyn_con_hair_tail_04}}]
    constraintWeights [1]
   }
   n12 = "Constraint - dyn_ponytail_03 (P)"(ComponentPose=@n11, Alpha=1, ConstraintWeights_0=1) {
    alphaBoolBlend {blendOption=Linear}
    boneToModify {boneName=dyn_ponytail_03}
    constraintSetup [{offsetOption=None, targetBone {boneName=dyn_con_hair_tail_05}}]
    constraintWeights [1]
   }
   n13 = "Constraint - dyn_ponytail_04 (P)"(ComponentPose=@n12, Alpha=1, ConstraintWeights_0=1) {
    alphaBoolBlend {blendOption=Linear}
    boneToModify {boneName=dyn_ponytail_04}
    constraintSetup [{offsetOption=None, targetBone {boneName=dyn_con_hair_tail_06}}]
    constraintWeights [1]
   }
   n14 = "Constraint - dyn_ponytail_05 (P)"(ComponentPose=@n13, Alpha=1, ConstraintWeights_0=1) {
    alphaBoolBlend {blendOption=Linear}
    boneToModify {boneName=dyn_ponytail_05}
    constraintSetup [{offsetOption=None, targetBone {boneName=dyn_con_hair_tail_07}}]
    constraintWeights [1]
   }
   n15 = "Transform (Modify) Bone - Bone: dyn_parent_ponytail"(ComponentPose=@n14, Alpha=Get HeadDelta() -> (HeadDelta:float), Translation="0,0,0", Rotation="5,0,0", Scale="1,1,1") {
    rotationMode=BMM_Additive
    rotationSpace=BCS_BoneSpace
    translationMode=BMM_Additive
    translationSpace=BCS_BoneSpace
    alphaBoolBlend {blendOption=Linear}
    boneToModify {boneName=dyn_parent_ponytail}
    rotation {pitch=5}
   }
   n16 = Anim Dynamics - Bone: dyn_ponytail_04(ComponentPose=@n15, Alpha=1, ExternalForce="0,0,0") {
    angularDampingOverride=1
    angularSpringConstant=1500
    bAngularSpring=true
    bGravityOverrideInSimSpace=true
    bOverrideAngularDamping=true
    gravityScale=0.35
    simulationSpace=RootRelative
    alphaBoolBlend {blendOption=Linear}
    boundBone {boneName=dyn_ponytail_04}
    physicsBodyDefinitions [{
      sphereCollisionRadius=0
      boundBone {boneName=dyn_ponytail_04}
      boxExtents=<10 20 20>
      constraintSetup {angularLimitsMax=<0 15 10>, angularLimitsMin=<0 -5 0>, angularTarget=<0 0 0>}
      localJointOffset=<8 0 1>
     }]
   }
   n17 = Anim Dynamics - Bone: dyn_ponytail_05(ComponentPose=@n16, Alpha=1, ExternalForce="0,0,0") {
    angularDampingOverride=1
    angularSpringConstant=1500
    bAngularSpring=true
    bGravityOverrideInSimSpace=true
    bOverrideAngularDamping=true
    gravityScale=0.35
    simulationSpace=RootRelative
    alphaBoolBlend {blendOption=Linear}
    boundBone {boneName=dyn_ponytail_05}
    physicsBodyDefinitions [{
      sphereCollisionRadius=0
      boundBone {boneName=dyn_ponytail_05}
      boxExtents=<10 20 20>
      constraintSetup {angularLimitsMax=<0 15 15>, angularLimitsMin=<0 -15 0>, angularTarget=<0 0 0>}
      localJointOffset=<8 0 2>
     }]
   }
   n18 = RigidBody!(ComponentPose=Local To Component(LocalPose=Use cached pose 'baseCache'()), Alpha, OverrideWorldGravity:Vector="0,0,-980", ExternalForce) -> (Pose) {
    bClampLinearTranslationLimitToRefPose=true
    bOverrideWorldGravity=true
    overridePhysicsAsset=&/F_MED_Ancient_Gladiator_Hair_Dyn
    simulationSpace=BaseBoneSpace
    alphaBoolBlend {blendOption=Linear}
    baseBoneRef {boneName=spine_05}
    componentAppliedLinearAccClamp=<3000 3000 3000>
    componentLinearAccScale=<0.45 0.3 0.3>
    componentLinearVelScale=<0.45 0.3 0.3>
   }
   Output Pose(Result=Blend Poses by bool(
     BlendPose_0=Component To Local(ComponentPose=@n17)
     BlendPose_1=Component To Local(ComponentPose=@n18)
     bActiveValue=Get bIsWearingBackpack() -> (bIsWearingBackpack:bool)
    ) {blendType=Linear}) {layerGroup=None}
   n19 = "Transform (Modify) Bone - Bone: dyn_parent_ponytail"(
    ComponentPose=Local To Component(LocalPose=Copy Pose From Mesh() {bUseAttachedParent=true})
    Alpha=Get HeadDelta()
    Translation="0,0,1"
    Rotation="30,0,0"
    Scale="1,1,1"
   ) {
    rotationMode=BMM_Additive
    rotationSpace=BCS_BoneSpace
    translationMode=BMM_Additive
    translationSpace=BCS_BoneSpace
    alphaBoolBlend {blendOption=Linear}
    boneToModify {boneName=dyn_parent_ponytail}
    rotation {pitch=30}
    translation {z=1}
   }
   n20 = "Transform (Modify) Bone - Bone: dyn_parent_ponytail"(ComponentPose=@n19, Alpha=Get HeadDeltaFwd() -> (HeadDeltaFwd:float), Translation="0,0,0", Rotation="0,0,0", Scale="1,1,1") {
    rotationMode=BMM_Additive
    rotationSpace=BCS_BoneSpace
    translationMode=BMM_Additive
    translationSpace=BCS_BoneSpace
    alphaBoolBlend {blendOption=Linear}
    boneToModify {boneName=dyn_parent_ponytail}
   }
   n21 = RigidBody(ComponentPose=@n20, OverrideWorldGravity="0,0,-980") {
    bClampLinearTranslationLimitToRefPose=true
    bOverrideWorldGravity=true
    overridePhysicsAsset=&/F_MED_Ancient_Gladiator_Hair_Physics
    simulationSpace=BaseBoneSpace
    alphaBoolBlend {blendOption=Linear}
    baseBoneRef {boneName=root}
    componentAppliedLinearAccClamp=<3000 3000 2000>
    componentLinearAccScale=<0.45 0.3 0.3>
    componentLinearVelScale=<0.45 0.3 0.3>
   }
   Save cached pose 'baseCache'(Pose=Component To Local(ComponentPose=@n21)) {cacheName=baseCache}
  ]}
 EventGraph {nodes [
   Event BlueprintUpdateAnimation()
   > Get!(execute:Exec) -> (then:Exec, else:Exec, SkeletalMesh:Skeletal Mesh Component Object Reference)
   > n1 = Gravity Override State Function DEPRECATED(CharacterPartAnimInstance=Self-Reference(), GameTimeInSeconds=GetGameTimeInSeconds(), GravityOverrideParams=Get Gravity Override Params())
   > n2 = Set GravityVector(GravityVector:Vector=@n1) -> (Output_Get:Vector)
   DrawVectorFromJointFunction!(
    execute
    SkelMeshComponent:Scene Component Object Reference
    JointName:Name=head
    LineEnd:Vector=@n2
    JointOffset:Vector="0, 0, 0"
    LineColor:Linear Color Structure=(R=1,G=0,B=0,A=1)
    Thickness:double=2
    Foreground:EDrawDebugSceneDepthPriorityGroup Enum=Foreground
   ) -> (then)
  ]}
]
variables [
 bIsCharacterPack_old {type=boolean}
 bIsWearingBackpack_old {type=boolean}
 bIsWearingCape_old {type=boolean}
 Gravity Override Params {type=GravityOverrideParamsStruct}
 IsCharacterPackAlpha_old {type=number}
 IsGenericBackpackAlpha_old {type=number}
]
