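# Physics asset for the Ancient Gladiator female hair part, backbling variant, defining dynamic ponytail bodies and upper-body collision shapes.
# A capsule chain runs from head through a master bone and six ponytail bones, nearly rigid at the root with narrow swing and twist limits below.
# Capsules on head, neck, clavicles and spine form the body volume, with five on spine_05 broadening the upper torso.
# Body joints leave swing and twist free.
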PhysicsAsset /Game/Characters/Player/Female/Medium/Bodies/F_MED_Ancient_Gladiator/Meshes/Parts/F_MED_Ancient_Gladiator_Hair_Dyn_backbling
bodies {
 clavicle_l [{shapeType=Capsule, length=6.9375, radius=6, center=<4.2524 0.0001 -1.5232>, rotation=<rot 84.6895 0 0>}]
 clavicle_r [{shapeType=Capsule, length=6.2188, radius=6.7188, center=<-4.2524 -0.0001 1.5232>, rotation=<rot 84.6895 0 0>}]
 dyn_ponytail_01 [{shapeType=Capsule, length=5.7812, radius=1.9062, center=<4.3761 0 0>, rotation=<rot 90 0 0>}]
 dyn_ponytail_02 [{shapeType=Capsule, length=6.3125, radius=2.4062, center=<3.5349 0 0>, rotation=<rot 90 0 0>}]
 dyn_ponytail_03 [{shapeType=Capsule, length=4.7188, radius=3.5, center=<4.6701 0 0>, rotation=<rot 90 0 0>}]
 dyn_ponytail_04 [{shapeType=Capsule, length=3.8438, radius=4.125, center=<4.6701 0 0>, rotation=<rot 90 0 0>}]
 dyn_ponytail_05 [{shapeType=Capsule, length=4.2812, radius=3.6875, center=<4.6701 0 0>, rotation=<rot 90 0 0>}]
 dyn_ponytail_06 [{shapeType=Capsule, length=6.4375, radius=1.5312, center=<4.6701 0 0>, rotation=<rot 90 0 0>}]
 dyn_ponytail_MSTR [{shapeType=Capsule, length=5.7812, radius=1.9062, center=<4.3761 0 0>, rotation=<rot 90 0 0>}]
 head [{shapeType=Capsule, length=7.4688, radius=5.4375, center=<0.5445 3.4826 0>, rotation=<rot 90 0 0>}]
 neck_01 [{shapeType=Capsule, length=6.6875, radius=4.5, center=<2.6233 0 0.0001>, rotation=<rot 90 0 0>}]
 spine_03 [{shapeType=Capsule, length=2.7188, radius=9.4062, center=<1.3445 3.246 0>}]
 spine_04 [{shapeType=Capsule, length=5.125, radius=10.0625, center=<1.3926 2.539 0>}]
 spine_05 [
  {shapeType=Capsule, length=8.25, radius=8.9375, center=<1.3926 0.7102 0>}
  {shapeType=Capsule, length=17.375, radius=7.25, center=<9.4987 1.2237 0>}
  {shapeType=Capsule, length=13.375, radius=12.75, center=<-4.0525 -3.5069 0>}
  {shapeType=Capsule, length=17.375, radius=7.25, center=<8.6185 -10.3562 18.7008>, rotation=<rot -0.5612 3.4368 -80.7373>}
  {shapeType=Capsule, length=17.375, radius=7.25, center=<10.4508 -11.1925 -17.5661>, rotation=<rot 0.3467 3.7372 -95.293>}
 ]
}
constraints [
 {bone1Name=dyn_ponytail_MSTR, bone2Name=head, swing1LimitDegrees=1, swing2LimitDegrees=1, twistLimitDegrees=1}
 {bone1Name=dyn_ponytail_01, bone2Name=dyn_ponytail_MSTR, swing1LimitDegrees=1, swing2LimitDegrees=1, twistLimitDegrees=1}
 {bone1Name=dyn_ponytail_02, bone2Name=dyn_ponytail_01, swing1LimitDegrees=15, swing2LimitDegrees=15, twistLimitDegrees=1}
 {bone1Name=dyn_ponytail_03, bone2Name=dyn_ponytail_02, swing1LimitDegrees=8, swing2LimitDegrees=8, twistLimitDegrees=2}
 {bone1Name=dyn_ponytail_04, bone2Name=dyn_ponytail_03, swing1LimitDegrees=5, swing2LimitDegrees=15, twistLimitDegrees=5}
 {bone1Name=dyn_ponytail_05, bone2Name=dyn_ponytail_04, swing1LimitDegrees=8, swing2LimitDegrees=15, twistLimitDegrees=0.3}
 {bone1Name=dyn_ponytail_06, bone2Name=dyn_ponytail_05, swing1LimitDegrees=15, swing2LimitDegrees=15, twistLimitDegrees=3}
 {
  bone1Name=head
  bone2Name=neck_01
  swing1LimitDegrees=45
  swing1Motion=Free
  swing2LimitDegrees=45
  swing2Motion=Free
  twistLimitDegrees=45
  twistMotion=Free
 }
 {
  bone1Name=clavicle_l
  bone2Name=spine_05
  swing1LimitDegrees=45
  swing1Motion=Free
  swing2LimitDegrees=45
  swing2Motion=Free
  twistLimitDegrees=45
  twistMotion=Free
 }
 {
  bone1Name=clavicle_r
  bone2Name=spine_05
  swing1LimitDegrees=45
  swing1Motion=Free
  swing2LimitDegrees=45
  swing2Motion=Free
  twistLimitDegrees=45
  twistMotion=Free
 }
 {
  bone1Name=neck_01
  bone2Name=spine_05
  swing1LimitDegrees=45
  swing1Motion=Free
  swing2LimitDegrees=45
  swing2Motion=Free
  twistLimitDegrees=45
  twistMotion=Free
 }
 {
  bone1Name=spine_05
  bone2Name=spine_04
  swing1LimitDegrees=45
  swing1Motion=Free
  swing2LimitDegrees=45
  swing2Motion=Free
  twistLimitDegrees=45
  twistMotion=Free
 }
 {
  bone1Name=spine_04
  bone2Name=spine_03
  swing1LimitDegrees=45
  swing1Motion=Free
  swing2LimitDegrees=45
  swing2Motion=Free
  twistLimitDegrees=45
  twistMotion=Free
 }
]
